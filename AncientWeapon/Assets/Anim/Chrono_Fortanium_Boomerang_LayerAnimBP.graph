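# Player item-layer Anim Blueprint that animates throwing the AncientWeapon boomerang via the ItemFullBodyOverride layer.
# Each update derives state flags from gameplay tags, movement and anim curves, driving a NonFire, ThrowBoomerang and BoomerangInAir state machine smoothed by inertialization and dead blending.
# The throw picks directional full-body animations, or upper-body ones layered over the cached lower body when sliding, crouching, riding or similarly constrained, then adds an aim offset.
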AnimBlueprint /AncientWeapon/Assets/Anim/Chrono_Fortanium_Boomerang_LayerAnimBP
target_skeleton=&Fortnite_M_Avg_Player_Skeleton
graphs [
 AnimGraph {nodes [Output Pose() {layerGroup=None}]}
 EventGraph {nodes [
   Event BlueprintUpdateAnimation()
   n1 = Get FortPlayerPawn() -> (FortPlayerPawn:Fort Player Pawn Object Reference)
   > Set IsThrown(IsThrown:bool=HasMatchingGameplayTag!(self:Gameplay Tag Asset Interface Interface=@n1, TagToCheck:Gameplay Tag Structure=(TagName="Granted.Chrono.Item.AncientWeapon.ThrowBlock")) -> (ReturnValue:bool)) -> (Output_Get:bool)
   > Set bAllowNonFireToThrowTimedTagCheck(bAllowNonFireToThrowTimedTagCheck:bool=HasMatchingGameplayTag(self=@n1, TagToCheck=(TagName="Granted.Chrono.Item.AncientWeapon.ThrowAnimTracker"))) -> (Output_Get:bool)
   > n2 = IsZiplining!(execute, self:Fort Player Pawn Object Reference=@n1) -> (then, ReturnValue:bool)
   > Set IsUpperlBody(IsUpperlBody:bool=OR Boolean!(
     A=Get bIsSliding()
     B=@n2
     C:bool=Get bIsRiding() -> (bIsRiding:bool)
     D:bool=Get bIsSurfaceSwimming()
     E:bool=Get IsUpperBodyState() -> (IsUpperBodyState:bool)
     F:bool=Get bIsWaterJump() -> (bIsWaterJump:bool)
     G:bool=HasMatchingGameplayTag(self=@n1, TagToCheck=(TagName="Gameplay.Action.Player.AthenaInVehicle"))
     H:bool=Get bIsCrouching()
    ) -> (ReturnValue)) -> (Output_Get:bool)
   > n3 = Sequence!(execute) -> (then_0, then_1, then_2:Exec, then_3:Exec, then_4:Exec)
   > n4 = Branch(Condition=Get bIsAccelerating2D())
   > Set Direction(Direction:double=Get PawnMovementDirectionAzimuth() -> (PawnMovementDirectionAzimuth:float)) -> (Output_Get:double)
   Set Direction(execute=@n4.else, Direction=0)
   n5 = AND Boolean(A="float > float"(A=GetCurveValue(CurveName=Interruptible), B=0.4), B=Get bIsCrouching())
   n6 = AND Boolean(A="float > float"(A=GetCurveValue(CurveName=Interruptible), B=0.8), B=NOT Boolean(A=Get bIsCrouching()))
   n7 = GetCurveValue(CurveName=TurnEnd)
   Set IsFinishedThrowAnimation(execute=@n3.then_1, IsFinishedThrowAnimation:bool=AND Boolean(A="float > float"(A=@n7, B=0.8), B=OR Boolean!(A=@n6, B=@n5) -> (ReturnValue))) -> (Output_Get:bool)
   Set IsCanThrow(execute=@n3.then_3, IsCanThrow:bool=NOT Boolean(A=Get IsThrown() -> (IsThrown:bool))) -> (Output_Get:bool)
   AnimNotify_EnterUpperBody() {is=title K2Node_Event}
   > Set IsUpperBodyState(IsUpperBodyState:bool=true) -> (Output_Get:bool)
   AnimNotify_ExitUpperBody() {is=title K2Node_Event}
   > Set IsUpperBodyState(IsUpperBodyState=false)
   AnimNotify_ExitUpperBodySliding() {is=title K2Node_Event}
   > Set IsUpperBodyState(IsUpperBodyState=false)
   AnimNotify_EnterUpperBodySliding() {is=title K2Node_Event}
   > Set IsUpperBodyState(IsUpperBodyState=true)
   AnimNotify_EnterNonFire() {is=title K2Node_Event}
   > Set IsUpperBodyState(IsUpperBodyState=false)
   TryGetPawnOwner()
   Event BlueprintInitializeAnimation()
  ]}
 ItemAimOffset {nodes [
   Input Pose!() -> (Pose, AimOffsetAlpha:double, Yaw:double, Pitch:double) {name=InPose_AimOffset}
   Input Pose!() -> (Pose) {name=InPose_PreAimOffsets}
   Output Pose() {layerGroup=Item Override}
  ]}
 ItemFallAdditive {nodes [
   Input Pose() {name=InPoseItemFallAdditive}
   Output Pose() {layerGroup=Item Override}
  ]}
 ItemFallLandAdditive {nodes [
   Input Pose() {name=InPoseItemFallLandAdditive}
   Output Pose() {layerGroup=Item Override}
  ]}
 ItemFallLevel2Override {nodes [
   Input Pose() {name=InPose_FallLevel2}
   Output Pose() {layerGroup=Item Override}
  ]}
 ItemFinalPoseOverride {nodes [
   Input Pose() {name=InFinalPose}
   Output Pose() {layerGroup=Item Override}
  ]}
 ItemFullBodyAimOffsets {nodes [
   Input Pose!() -> (Pose, FullBody_AimOffsetAlpha:double, Aim_Yaw:double, Aim_Pitch:double) {name=InPoseFullBodyAimOffsets}
   Input Pose!() -> (Pose) {name=InPosePreFullBodyAimOffsets}
   Output Pose() {layerGroup=Item Override}
  ]}
 ItemFullBodyOverride {nodes [
   AnimGraphNode_StateMachine_0 = FireState() {is=title AnimGraphNode_StateMachine, bAllowConduitEntryStates=true}
   Output Pose(Result=Dead Blending!(Source:Pose Link Structure=@AnimGraphNode_StateMachine_0) -> (Pose:Pose Link Structure) {is=title AnimGraphNode_DeadBlending, defaultBlendMode=Linear}) {layerGroup=Item Override}
   Save cached pose 'FullBodySavedPose'(Pose=Input Pose() {name=InPoseFullBody}) {cacheName=FullBodySavedPose}
  ]}
 FireState {
  entry=Conduit
  path=ItemFullBodyOverride.AnimGraphNode_StateMachine_0.FireState
  ids=id
  nodes [
   AnimStateConduitNode_0 = Conduit()
   AnimStateNode_0 = NonFire() {
    is=title AnimStateNode
    stateEntered {guid=#1, notifyName=EnterNonFire}
    stateFullyBlended {guid=#2}
    stateLeft {guid=#3}
   }
   AnimStateNode_1 = ThrowBoomerang() {
    is=title AnimStateNode
    stateEntered {guid=#4, notifyName=EnterFire}
    stateFullyBlended {guid=#5}
    stateLeft {guid=#6}
   }
   AnimStateNode_2 = BoomerangInAir() {is=title AnimStateNode}
   AnimStateTransitionNode_0 = AnimStateTransitionNode() {crossfadeDuration=0.05, logicType=TLT_Inertialization}
   AnimStateTransitionNode_1 = AnimStateTransitionNode() {logicType=TLT_Inertialization, priorityOrder=100}
   AnimStateTransitionNode_2 = AnimStateTransitionNode() {automaticRuleTriggerTime=0, logicType=TLT_Inertialization}
   AnimStateTransitionNode_3 = AnimStateTransitionNode() {logicType=TLT_Inertialization}
   AnimStateTransitionNode_4 = AnimStateTransitionNode() {logicType=TLT_Inertialization}
  ]
  links [
   {
    dst=NonFire
    rule=ItemFullBodyOverride.AnimGraphNode_StateMachine_0.FireState.AnimStateTransitionNode_3.Transition
    src=BoomerangInAir
    via=AnimStateTransitionNode_3
   }
   {
    dst=BoomerangInAir
    rule=ItemFullBodyOverride.AnimGraphNode_StateMachine_0.FireState.AnimStateTransitionNode_4.Transition
    src=Conduit
    via=AnimStateTransitionNode_4
   }
   {
    dst=NonFire
    rule=ItemFullBodyOverride.AnimGraphNode_StateMachine_0.FireState.AnimStateTransitionNode_1.Transition
    src=Conduit
    via=AnimStateTransitionNode_1
   }
   {
    dst=ThrowBoomerang
    rule=ItemFullBodyOverride.AnimGraphNode_StateMachine_0.FireState.AnimStateTransitionNode_0.Transition
    src=NonFire
    via=AnimStateTransitionNode_0
   }
   {
    dst=BoomerangInAir
    rule=ItemFullBodyOverride.AnimGraphNode_StateMachine_0.FireState.AnimStateTransitionNode_2.Transition
    src=ThrowBoomerang
    via=AnimStateTransitionNode_2
   }
  ]
 }
 Conduit {path=ItemFullBodyOverride.AnimGraphNode_StateMachine_0.FireState.AnimStateConduitNode_0.Conduit, nodes [Result(bCanEnterTransition=true)]}
 NonFire {
  path=ItemFullBodyOverride.AnimGraphNode_StateMachine_0.FireState.AnimStateNode_0.NonFire
  nodes [Output Animation Pose(Result=Use cached pose 'FullBodySavedPose'())]
 }
 ThrowBoomerang {
  path=ItemFullBodyOverride.AnimGraphNode_StateMachine_0.FireState.AnimStateNode_1.ThrowBoomerang
  nodes [
   AnimGraphNode_StateMachine_1 = UpperBodyStatemachine() {is=title AnimGraphNode_StateMachine, bAllowConduitEntryStates=true}
   n1 = Layered blend per bone(BasePose=@AnimGraphNode_StateMachine_1, BlendPoses_0=Use cached pose 'FullBodySavedPose'()) {blendMode=BlendMask, blendMasks [&Fortnite_M_Avg_Player_Skeleton:LowerBody], layerSetup []}
   AnimGraphNode_StateMachine_0 = FullBodyStateMachine() {is=title AnimGraphNode_StateMachine, bAllowConduitEntryStates=true}
   n2 = Blend Poses by bool(BlendPose_0=@n1, BlendPose_1=@AnimGraphNode_StateMachine_0, bActiveValue=Get IsUpperlBody() -> (IsUpperlBody:bool)) {blendType=Linear}
   n3 = Get AimYaw() -> (AimYaw:float)
   n4 = Get AimPitch()
   Output Animation Pose(Result=Modify Curve(
     SourcePose=AimOffset Player 'AncientWeapon_AO'!(X:float=@n3, Y:float=@n4, BasePose:Pose Link Structure=@n2, Alpha:float=1) -> (Pose:Pose Link Structure) {blendSpace=&/AncientWeapon_AO, alphaBoolBlend {blendOption=Linear}}
     CurveValues_0=1
     Alpha=1
    ) {curveNames [TurnEnd], curveValues [0]})
  ]
 }
 FullBodyStateMachine {
  entry=Conduit
  path=ItemFullBodyOverride.AnimGraphNode_StateMachine_0.FireState.AnimStateNode_1.ThrowBoomerang.AnimGraphNode_StateMachine_0.FullBodyStateMachine
  ids=id
  nodes [
   AnimStateConduitNode_0 = Conduit()
   AnimStateNode_0 = North() {is=title AnimStateNode}
   AnimStateNode_1 = West() {is=title AnimStateNode}
   AnimStateNode_2 = East() {is=title AnimStateNode}
   AnimStateNode_3 = South() {is=title AnimStateNode}
   AnimStateNode_4 = Idle()
   AnimStateTransitionNode_0 = AnimStateTransitionNode() {logicType=TLT_Inertialization, priorityOrder=50}
   AnimStateTransitionNode_1 = AnimStateTransitionNode() {logicType=TLT_Inertialization}
   AnimStateTransitionNode_2 = AnimStateTransitionNode() {logicType=TLT_Inertialization, priorityOrder=2}
   AnimStateTransitionNode_3 = AnimStateTransitionNode() {logicType=TLT_Inertialization, priorityOrder=3}
   AnimStateTransitionNode_4 = AnimStateTransitionNode() {priorityOrder=100}
   AnimStateTransitionNode_5 = AnimStateTransitionNode()
   AnimStateTransitionNode_6 = AnimStateTransitionNode()
  ]
  links [
   {
    dst=East
    rule=ItemFullBodyOverride.AnimGraphNode_StateMachine_0.FireState.AnimStateNode_1.ThrowBoomerang.AnimGraphNode_StateMachine_0.FullBodyStateMachine.AnimStateTransitionNode_2.Transition
    src=Conduit
    via=AnimStateTransitionNode_2
   }
   {
    dst=Idle
    rule=ItemFullBodyOverride.AnimGraphNode_StateMachine_0.FireState.AnimStateNode_1.ThrowBoomerang.AnimGraphNode_StateMachine_0.FullBodyStateMachine.AnimStateTransitionNode_4.Transition
    src=Conduit
    via=AnimStateTransitionNode_4
   }
   {
    dst=North
    rule=ItemFullBodyOverride.AnimGraphNode_StateMachine_0.FireState.AnimStateNode_1.ThrowBoomerang.AnimGraphNode_StateMachine_0.FullBodyStateMachine.AnimStateTransitionNode_0.Transition
    src=Conduit
    via=AnimStateTransitionNode_0
   }
   {
    dst=South
    rule=ItemFullBodyOverride.AnimGraphNode_StateMachine_0.FireState.AnimStateNode_1.ThrowBoomerang.AnimGraphNode_StateMachine_0.FullBodyStateMachine.AnimStateTransitionNode_3.Transition
    src=Conduit
    via=AnimStateTransitionNode_3
   }
   {
    dst=West
    rule=ItemFullBodyOverride.AnimGraphNode_StateMachine_0.FireState.AnimStateNode_1.ThrowBoomerang.AnimGraphNode_StateMachine_0.FullBodyStateMachine.AnimStateTransitionNode_1.Transition
    src=Conduit
    via=AnimStateTransitionNode_1
   }
   {
    dst=Conduit
    rule=ItemFullBodyOverride.AnimGraphNode_StateMachine_0.FireState.AnimStateNode_1.ThrowBoomerang.AnimGraphNode_StateMachine_0.FullBodyStateMachine.AnimStateTransitionNode_6.Transition
    src=Idle
    via=AnimStateTransitionNode_6
   }
   {
    dst=Conduit
    rule=ItemFullBodyOverride.AnimGraphNode_StateMachine_0.FireState.AnimStateNode_1.ThrowBoomerang.AnimGraphNode_StateMachine_0.FullBodyStateMachine.AnimStateTransitionNode_5.Transition
    src=North
    via=AnimStateTransitionNode_5
   }
  ]
 }
 Conduit {
  path=ItemFullBodyOverride.AnimGraphNode_StateMachine_0.FireState.AnimStateNode_1.ThrowBoomerang.AnimGraphNode_StateMachine_0.FullBodyStateMachine.AnimStateConduitNode_0.Conduit
  nodes [Result(bCanEnterTransition=true)]
 }
 North {
  path=ItemFullBodyOverride.AnimGraphNode_StateMachine_0.FireState.AnimStateNode_1.ThrowBoomerang.AnimGraphNode_StateMachine_0.FullBodyStateMachine.AnimStateNode_0.North
  nodes [
   n1 = Sequence Player 'Fortanium_Boomerang_SprintThrow_Directional_N'() {bLoopAnimation=false, groupName=Throw, method=SyncGroup, sequence=&/Fortanium_Boomerang_SprintThrow_Directional_N}
   Output Animation Pose(Result=@n1)
  ]
 }
 West {
  path=ItemFullBodyOverride.AnimGraphNode_StateMachine_0.FireState.AnimStateNode_1.ThrowBoomerang.AnimGraphNode_StateMachine_0.FullBodyStateMachine.AnimStateNode_1.West
  nodes [
   n1 = Sequence Player 'Fortanium_Boomerang_SprintThrow_Directional_W'() {bLoopAnimation=false, groupName=Throw, method=SyncGroup, sequence=&/Fortanium_Boomerang_SprintThrow_Directional_W}
   Output Animation Pose(Result=@n1)
  ]
 }
 East {
  path=ItemFullBodyOverride.AnimGraphNode_StateMachine_0.FireState.AnimStateNode_1.ThrowBoomerang.AnimGraphNode_StateMachine_0.FullBodyStateMachine.AnimStateNode_2.East
  nodes [
   n1 = Sequence Player 'Fortanium_Boomerang_SprintThrow_Directional_E'() {bLoopAnimation=false, groupName=Throw, method=SyncGroup, sequence=&/Fortanium_Boomerang_SprintThrow_Directional_E}
   Output Animation Pose(Result=@n1)
  ]
 }
 South {
  path=ItemFullBodyOverride.AnimGraphNode_StateMachine_0.FireState.AnimStateNode_1.ThrowBoomerang.AnimGraphNode_StateMachine_0.FullBodyStateMachine.AnimStateNode_3.South
  nodes [
   n1 = Sequence Player 'Fortanium_Boomerang_SprintThrow_Directional_S'() {bLoopAnimation=false, groupName=Throw, method=SyncGroup, sequence=&/Fortanium_Boomerang_SprintThrow_Directional_S}
   Output Animation Pose(Result=@n1)
  ]
 }
 Idle {
  path=ItemFullBodyOverride.AnimGraphNode_StateMachine_0.FireState.AnimStateNode_1.ThrowBoomerang.AnimGraphNode_StateMachine_0.FullBodyStateMachine.AnimStateNode_4.Idle
  nodes [
   n1 = Sequence Player 'Fortanium_Boomerang_Throw_InPlace'() {bLoopAnimation=false, groupName=Throw, method=SyncGroup, sequence=&/Fortanium_Boomerang_Throw_InPlace}
   Output Animation Pose(Result=@n1)
  ]
 }
 Transition {
  path=ItemFullBodyOverride.AnimGraphNode_StateMachine_0.FireState.AnimStateNode_1.ThrowBoomerang.AnimGraphNode_StateMachine_0.FullBodyStateMachine.AnimStateTransitionNode_0.Transition
  nodes [
   Result(bCanEnterTransition=Get bIsAccelerating2D())
   Get LocomotionCardinalDirection() -> (LocomotionCardinalDirection:EFortCardinalDirection Enum=North)
  ]
 }
 Transition {
  path=ItemFullBodyOverride.AnimGraphNode_StateMachine_0.FireState.AnimStateNode_1.ThrowBoomerang.AnimGraphNode_StateMachine_0.FullBodyStateMachine.AnimStateTransitionNode_1.Transition
  nodes [
   n1 = Get LocomotionCardinalDirection() -> (LocomotionCardinalDirection=North)
   Result(bCanEnterTransition=AND Boolean(A="Equal (Enum)"(A:EFortCardinalDirection Enum=@n1, B:EFortCardinalDirection Enum=West), B=Get bIsAccelerating2D()))
  ]
 }
 Transition {
  path=ItemFullBodyOverride.AnimGraphNode_StateMachine_0.FireState.AnimStateNode_1.ThrowBoomerang.AnimGraphNode_StateMachine_0.FullBodyStateMachine.AnimStateTransitionNode_2.Transition
  nodes [
   n1 = Get LocomotionCardinalDirection() -> (LocomotionCardinalDirection=North)
   Result(bCanEnterTransition=AND Boolean(A="Equal (Enum)"(A=@n1, B=East), B=Get bIsAccelerating2D()))
  ]
 }
 Transition {
  path=ItemFullBodyOverride.AnimGraphNode_StateMachine_0.FireState.AnimStateNode_1.ThrowBoomerang.AnimGraphNode_StateMachine_0.FullBodyStateMachine.AnimStateTransitionNode_3.Transition
  nodes [
   n1 = Get LocomotionCardinalDirection() -> (LocomotionCardinalDirection=North)
   Result(bCanEnterTransition=AND Boolean(A="Equal (Enum)"(A=@n1, B=South), B=Get bIsAccelerating2D()))
  ]
 }
 Transition {
  path=ItemFullBodyOverride.AnimGraphNode_StateMachine_0.FireState.AnimStateNode_1.ThrowBoomerang.AnimGraphNode_StateMachine_0.FullBodyStateMachine.AnimStateTransitionNode_4.Transition
  nodes [Result(bCanEnterTransition=true)]
 }
 Transition {
  path=ItemFullBodyOverride.AnimGraphNode_StateMachine_0.FireState.AnimStateNode_1.ThrowBoomerang.AnimGraphNode_StateMachine_0.FullBodyStateMachine.AnimStateTransitionNode_5.Transition
  nodes [Result(bCanEnterTransition=true)]
 }
 Transition {
  path=ItemFullBodyOverride.AnimGraphNode_StateMachine_0.FireState.AnimStateNode_1.ThrowBoomerang.AnimGraphNode_StateMachine_0.FullBodyStateMachine.AnimStateTransitionNode_6.Transition
  nodes [Result(bCanEnterTransition=Get bIsAccelerating2D())]
 }
 UpperBodyStatemachine {
  entry=Conduit
  path=ItemFullBodyOverride.AnimGraphNode_StateMachine_0.FireState.AnimStateNode_1.ThrowBoomerang.AnimGraphNode_StateMachine_1.UpperBodyStatemachine
  ids=id
  nodes [
   AnimStateConduitNode_0 = Conduit()
   AnimStateNode_0 = UpperBodySliding() {
    is=title AnimStateNode
    stateEntered {guid=#7, notifyName=EnterUpperBodySliding}
    stateLeft {guid=#8, notifyName=ExitUpperBodySliding}
   }
   AnimStateNode_2 = UpperBody() {
    is=title AnimStateNode
    stateEntered {guid=#9, notifyName=EnterUpperBody}
    stateLeft {guid=#10, notifyName=ExitUpperBody}
   }
   AnimStateTransitionNode_0 = AnimStateTransitionNode() {logicType=TLT_Inertialization}
   AnimStateTransitionNode_2 = AnimStateTransitionNode() {logicType=TLT_Inertialization, priorityOrder=100}
  ]
  links [
   {
    dst=UpperBody
    rule=ItemFullBodyOverride.AnimGraphNode_StateMachine_0.FireState.AnimStateNode_1.ThrowBoomerang.AnimGraphNode_StateMachine_1.UpperBodyStatemachine.AnimStateTransitionNode_2.Transition
    src=Conduit
    via=AnimStateTransitionNode_2
   }
   {
    dst=UpperBodySliding
    rule=ItemFullBodyOverride.AnimGraphNode_StateMachine_0.FireState.AnimStateNode_1.ThrowBoomerang.AnimGraphNode_StateMachine_1.UpperBodyStatemachine.AnimStateTransitionNode_0.Transition
    src=Conduit
    via=AnimStateTransitionNode_0
   }
  ]
 }
 Conduit {
  path=ItemFullBodyOverride.AnimGraphNode_StateMachine_0.FireState.AnimStateNode_1.ThrowBoomerang.AnimGraphNode_StateMachine_1.UpperBodyStatemachine.AnimStateConduitNode_0.Conduit
  nodes [Result(bCanEnterTransition=true)]
 }
 UpperBodySliding {
  path=ItemFullBodyOverride.AnimGraphNode_StateMachine_0.FireState.AnimStateNode_1.ThrowBoomerang.AnimGraphNode_StateMachine_1.UpperBodyStatemachine.AnimStateNode_0.UpperBodySliding
  nodes [Output Animation Pose(Result=Sequence Player 'Boomerang_ThrowUpperBodySlide_CMM'() {bLoopAnimation=false, sequence=&/Boomerang_ThrowUpperBodySlide_CMM})]
 }
 UpperBody {
  path=ItemFullBodyOverride.AnimGraphNode_StateMachine_0.FireState.AnimStateNode_1.ThrowBoomerang.AnimGraphNode_StateMachine_1.UpperBodyStatemachine.AnimStateNode_2.UpperBody
  nodes [Output Animation Pose(Result=Sequence Player 'Boomerang_ThrowUpperBody_CMM'() {bLoopAnimation=false, sequence=&/Boomerang_ThrowUpperBody_CMM})]
 }
 Transition {
  path=ItemFullBodyOverride.AnimGraphNode_StateMachine_0.FireState.AnimStateNode_1.ThrowBoomerang.AnimGraphNode_StateMachine_1.UpperBodyStatemachine.AnimStateTransitionNode_0.Transition
  nodes [Result(bCanEnterTransition=AND Boolean(A=Get IsUpperlBody(), B=Get bIsSliding()))]
 }
 Transition {
  path=ItemFullBodyOverride.AnimGraphNode_StateMachine_0.FireState.AnimStateNode_1.ThrowBoomerang.AnimGraphNode_StateMachine_1.UpperBodyStatemachine.AnimStateTransitionNode_2.Transition
  nodes [Result(bCanEnterTransition=true)]
 }
 BoomerangInAir {
  path=ItemFullBodyOverride.AnimGraphNode_StateMachine_0.FireState.AnimStateNode_2.BoomerangInAir
  nodes [Output Animation Pose(Result=Use cached pose 'FullBodySavedPose'())]
 }
 Transition {
  path=ItemFullBodyOverride.AnimGraphNode_StateMachine_0.FireState.AnimStateTransitionNode_0.Transition
  nodes [Result(bCanEnterTransition=Get bAllowNonFireToThrowTimedTagCheck() -> (bAllowNonFireToThrowTimedTagCheck:bool))]
 }
 Transition {
  path=ItemFullBodyOverride.AnimGraphNode_StateMachine_0.FireState.AnimStateTransitionNode_1.Transition
  nodes [Result(bCanEnterTransition=true)]
 }
 Transition {
  path=ItemFullBodyOverride.AnimGraphNode_StateMachine_0.FireState.AnimStateTransitionNode_2.Transition
  nodes [Result(bCanEnterTransition=Get IsFinishedThrowAnimation() -> (IsFinishedThrowAnimation:bool))]
 }
 Transition {
  path=ItemFullBodyOverride.AnimGraphNode_StateMachine_0.FireState.AnimStateTransitionNode_3.Transition
  nodes [Result(bCanEnterTransition=NOT Boolean(A=Get IsThrown()))]
 }
 Transition {
  path=ItemFullBodyOverride.AnimGraphNode_StateMachine_0.FireState.AnimStateTransitionNode_4.Transition
  nodes [Result(bCanEnterTransition=Get IsCanThrow() -> (IsCanThrow:bool))]
 }
 ItemIdleOverride {nodes [
   Input Pose() {name=InPose_Idle}
   Output Pose() {layerGroup=Item Override}
  ]}
 ItemJetPackJumpAdditive {nodes [Output Pose() {layerGroup=Item Override}]}
 ItemJetPackStartAdditive {nodes [Output Pose() {layerGroup=Item Override}]}
 ItemJumpLoopAdditive {nodes [
   Input Pose() {name=InPoseItemJumpLoopAdditive}
   Output Pose() {layerGroup=Item Override}
  ]}
 ItemJumpOverride {nodes [
   Input Pose() {name=InPoseJump}
   Output Pose() {layerGroup=Item Override}
  ]}
 ItemJumpUpAdditive {nodes [
   Input Pose() {name=InPoseItemJumpUpAdditive}
   Output Pose() {layerGroup=Item Override}
  ]}
 ItemPreAimOffsets {nodes [
   Input Pose() {name=InPosePreAimOffsets}
   Output Pose() {layerGroup=Item Override}
  ]}
 ItemPreIK {nodes [
   Input Pose() {name=InPose_PreIK}
   Output Pose() {layerGroup=Item Override}
  ]}
 ItemRightHandIK {nodes [
   Input Pose() {name=InPosePreRightHandIK}
   Input Pose() {name=InPosePostRightHandIK}
   Output Pose() {layerGroup=Item Override}
  ]}
 ItemSkeletalControl {nodes [
   Input Pose()
   Output Pose() {layerGroup=Item Override}
  ]}
 ItemSwimJumpFallLoop {nodes [Output Pose() {layerGroup=Item Override}]}
 ItemSwimJumpSurfaceEnd {nodes [Output Pose() {layerGroup=Item Override}]}
 ItemSwimJumpSurfaceLoop {nodes [Output Pose() {layerGroup=Item Override}]}
 ItemUpperBody {nodes [
   Input Pose!() -> (Pose, InputParam:Fort Anim Input Adjusted Aim Structure) {name=InPoseUpperBody}
   Output Pose() {layerGroup=Item Override}
  ]}
 ItemZipLineStartAdditive {nodes [Output Pose() {layerGroup=Item Override}]}
]
variables [
 bAllowNonFireToThrowTimedTagCheck {type=boolean}
 Direction {type=number}
 EnterFire {type=boolean}
 Fort Anim Input Zipline {type=FortAnimInput_Zipline, default {}}
 IsAlreadyRecalled {type=boolean}
 IsCanThrow {type=boolean}
 IsFinishedThrowAnimation {type=boolean}
 IsThrown {type=boolean}
 IsUpperBodyState {type=boolean}
 IsUpperlBody {type=boolean}
]
