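# Sound cue for the Ancient Weapon's aim-down-sights-in audio, routed to the melee weapon sound class with equip attenuation.
# A mixer layers a main ADS-in wave, a random whoosh, a random zap and a container ambient zap loop, each through its own modulator.
# The main wave and random zaps are pitched up with slight variation.
# The ambient layer loops, so the cue has no natural end.
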SoundCue /AncientWeapon/Assets/Audio/AncientWeapon_ADS_In_Cue
root=Mixer_0
nodes [
 {id=Mixer_0, children [Modulator_3,Modulator_8,Modulator_9,Modulator_0], properties {inputVolume [1,0.2,0.8,0.8]}}
 {id=Modulator_0, children [WavePlayer_0]}
 {id=Modulator_3, children [WavePlayer_15], properties {pitchMax=1.3, pitchMin=1.2, volumeMax=1, volumeMin=1}}
 {id=Modulator_8, children [Random_6], properties {pitchMax=1, pitchMin=1, volumeMax=1, volumeMin=1}}
 {id=Modulator_9, children [Random_7], properties {pitchMax=1.3, pitchMin=1.2, volumeMax=1, volumeMin=1}}
 {id=Random_6, children [WavePlayer_16, WavePlayer_17, WavePlayer_18], properties {weights [1,1,1]}}
 {id=Random_7, children [WavePlayer_21, WavePlayer_23, WavePlayer_24], properties {weights [1,1,1]}}
 {id=WavePlayer_0, properties {bLooping=true, soundWaveAssetPtr=/AncientWeapon/Environmental/Sound/Ancient_Weapon_Container_Ambient_Loop_Zaps}}
 {id=WavePlayer_15, properties {soundWaveAssetPtr=&/Ancient_Weapon_ADS_In}}
 {id=WavePlayer_16, properties {soundWaveAssetPtr=&/Ancient_Weapon_ADS_In_Whoosh_03}}
 {id=WavePlayer_17, properties {soundWaveAssetPtr=&/Ancient_Weapon_ADS_In_Whoosh_01}}
 {id=WavePlayer_18, properties {soundWaveAssetPtr=&/Ancient_Weapon_ADS_In_Whoosh_02}}
 {id=WavePlayer_21, properties {soundWaveAssetPtr=&/Ancient_Weapon_ADS_In_Zaps_01}}
 {id=WavePlayer_23, properties {soundWaveAssetPtr=&/Ancient_Weapon_ADS_In_Zaps_02}}
 {id=WavePlayer_24, properties {soundWaveAssetPtr=&/Ancient_Weapon_ADS_In_Zaps_03}}
]
properties {
 attenuationSettings=/AncientWeapon/Assets/Audio/Audio_Settings/Ancient_Weapon_Equip_Att
 maxDistance=1800
 soundClassObject=&Fort_WeaponMelee
 volumeMultiplier=0.6
}
