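# Sound cue for the Ancient Weapon's aim-down-sights exit sound.
# A single wave player feeds a modulator that slightly randomizes pitch downward for variation.
# The cue plays at reduced volume in the melee weapon sound class and borrows the weapon's equip attenuation settings.
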SoundCue /AncientWeapon/Assets/Audio/AncientWeapon_ADS_Out_Cue
root=Modulator_8
nodes [
 {id=Modulator_8, children [WavePlayer_18], properties {pitchMax=1, pitchMin=0.9}}
 {id=WavePlayer_18, properties {soundWaveAssetPtr=&/Ancient_Weapon_ADS_Out}}
]
properties {
 attenuationSettings=/AncientWeapon/Assets/Audio/Audio_Settings/Ancient_Weapon_Equip_Att
 maxDistance=1800
 soundClassObject=&Fort_WeaponMelee
 volumeMultiplier=0.2
}
