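# Sound cue for the Ancient Weapon's catch sound, classed as melee weapon audio.
# A mixer layers a random pick of six catch waves with a random pick of four electric zap waves, each through its own modulator, with the zaps pitched up.
# It uses a dedicated attenuation asset and a concurrency override allowing one instance per owner, blocking new plays while one is active.
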SoundCue /AncientWeapon/Assets/Audio/AncientWeapon_Catch
root=Mixer_0
nodes [
 {id=Mixer_0, children [Modulator_0,Modulator_1], properties {inputVolume [1,1]}}
 {id=Modulator_0, children [Random_1]}
 {id=Modulator_1, children [Random_0], properties {pitchMax=1.2, pitchMin=1.2}}
 {id=Random_0, children [WavePlayer_0,WavePlayer_7,WavePlayer_8,WavePlayer_9], properties {weights [1,1,1,1]}}
 {id=Random_1, children [WavePlayer_1,WavePlayer_2,WavePlayer_3,WavePlayer_4,WavePlayer_5,WavePlayer_6], properties {weights [1,1,1,1,1,1]}}
 {id=WavePlayer_0, properties {soundWaveAssetPtr=&/Ancient_Weapon_Catch_Zaps_01}}
 {id=WavePlayer_1, properties {soundWaveAssetPtr=&/AncientWeapon_Catch_01}}
 {id=WavePlayer_2, properties {soundWaveAssetPtr=&/AncientWeapon_Catch_02}}
 {id=WavePlayer_3, properties {soundWaveAssetPtr=&/AncientWeapon_Catch_03}}
 {id=WavePlayer_4, properties {soundWaveAssetPtr=&/AncientWeapon_Catch_04}}
 {id=WavePlayer_5, properties {soundWaveAssetPtr=&/AncientWeapon_Catch_05}}
 {id=WavePlayer_6, properties {soundWaveAssetPtr=&/AncientWeapon_Catch_06}}
 {id=WavePlayer_7, properties {soundWaveAssetPtr=&/Ancient_Weapon_Catch_Zaps_02}}
 {id=WavePlayer_8, properties {soundWaveAssetPtr=&/Ancient_Weapon_Catch_Zaps_03}}
 {id=WavePlayer_9, properties {soundWaveAssetPtr=&/Ancient_Weapon_Catch_Zaps_04}}
]
properties {
 attenuationSettings=/AncientWeapon/Assets/Audio/Audio_Settings/Ancient_Weapon_Catch_Att
 bOverrideConcurrency=true
 maxDistance=12000
 pitchMultiplier=0.95
 soundClassObject=&Fort_WeaponMelee
 volumeMultiplier=0.5
 concurrencyOverrides {bLimitToOwner=true, maxCount=1, resolutionRule=PreventNew}
}
