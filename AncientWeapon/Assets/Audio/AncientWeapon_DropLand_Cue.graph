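# Sound cue for the landing sound of a dropped Ancient Weapon.
# A Random node picks one of three equally weighted prototype landing waves, and a Modulator at the root adds variation.
# It uses the shared weapon-drop attenuation and loot-collect sound class, with raised pitch and volume multipliers.
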SoundCue /AncientWeapon/Assets/Audio/AncientWeapon_DropLand_Cue
root=Modulator_0
nodes [
 {id=Modulator_0, children [Random_1]}
 {id=Random_1, children [WavePlayer_3,WavePlayer_4,WavePlayer_5], properties {weights [1,1,1]}}
 {id=WavePlayer_3, properties {soundWaveAssetPtr=/AncientWeapon/Assets/Audio/Proto/AncientWeapon_DropLand_01}}
 {id=WavePlayer_4, properties {soundWaveAssetPtr=/AncientWeapon/Assets/Audio/Proto/AncientWeapon_DropLand_02}}
 {id=WavePlayer_5, properties {soundWaveAssetPtr=/AncientWeapon/Assets/Audio/Proto/AncientWeapon_DropLand_03}}
]
properties {
 attenuationSettings=&Fort_Weapon_Drop_Att
 maxDistance=3000
 pitchMultiplier=1.5
 soundClassObject=&Fort_Loot_Collect
 volumeMultiplier=1.1631
}
