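# Sound cue that plays the AncientWeapon drop sound.
# A single wave player feeds a modulator that randomizes pitch upward on each play, while the cue itself lowers the base pitch and boosts volume.
# It uses the shared Fort weapon drop attenuation and routes through the loot collect sound class.
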SoundCue /AncientWeapon/Assets/Audio/AncientWeapon_Drop_Cue
root=Modulator_0
nodes [
 {id=Modulator_0, children [WavePlayer_1], properties {pitchMax=1.2, pitchMin=1}}
 {id=WavePlayer_1, properties {soundWaveAssetPtr=/AncientWeapon/Assets/Audio/Proto/AncientWeapon_Drop_01}}
]
properties {
 attenuationSettings=&Fort_Weapon_Drop_Att
 maxDistance=3000
 pitchMultiplier=0.8
 soundClassObject=&Fort_Loot_Collect
 volumeMultiplier=2.0935
}
