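# Sound cue for equipping the Ancient Weapon, routed to the Fort_WeaponMelee sound class.
# A Random node picks one of three equip wave variations with equal weight and feeds a Modulator at its default settings for pitch and volume variation.
# The cue raises its volume multiplier and uses a dedicated equip attenuation asset for distance falloff.
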SoundCue /AncientWeapon/Assets/Audio/AncientWeapon_Equip
root=Modulator_8
nodes [
 {id=Modulator_8, children [Random_6]}
 {id=Random_6, children [WavePlayer_15, WavePlayer_16, WavePlayer_17], properties {weights [1,1,1]}}
 {id=WavePlayer_15, properties {soundWaveAssetPtr=&/Ancient_Weapon_Equip_01}}
 {id=WavePlayer_16, properties {soundWaveAssetPtr=&/Ancient_Weapon_Equip_02}}
 {id=WavePlayer_17, properties {soundWaveAssetPtr=&/Ancient_Weapon_Equip_03}}
]
properties {
 attenuationSettings=/AncientWeapon/Assets/Audio/Audio_Settings/Ancient_Weapon_Equip_Att
 maxDistance=1800
 soundClassObject=&Fort_WeaponMelee
 volumeMultiplier=2.5
}
