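# Looping idle sound cue for the Ancient Weapon, layering an energy hum with electrical zaps under the gadgets/items sound class.
# A Mixer blends the two looping waves, weighting the energy layer well above the zaps.
# Each layer passes through its own Modulator, with the energy layer pinned to a lowered pitch.
# The cue uses the weapon's equip attenuation asset and overrides concurrency to cap simultaneous instances.
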SoundCue /AncientWeapon/Assets/Audio/AncientWeapon_Idle_Loop_Cue
root=Mixer_0
nodes [
 {id=Mixer_0, children [Modulator_0,Modulator_1], properties {inputVolume [1.5,0.7]}}
 {id=Modulator_0, children [WavePlayer_3], properties {pitchMax=0.9, pitchMin=0.9}}
 {id=Modulator_1, children [WavePlayer_2], properties {pitchMin=1}}
 {id=WavePlayer_2, properties {bLooping=true, soundWaveAssetPtr=&/Ancient_Weapon_Idle_Zaps_Loop}}
 {id=WavePlayer_3, properties {bLooping=true, soundWaveAssetPtr=&/Ancient_Weapon_Idle_Energy_Loop}}
]
properties {
 attenuationSettings=/AncientWeapon/Assets/Audio/Audio_Settings/Ancient_Weapon_Equip_Att
 bOverrideConcurrency=true
 maxDistance=1800
 soundClassObject=&Fort_Gadgets_Items
 volumeMultiplier=0.8
 attenuationOverrides {distanceAlgorithm=NaturalSound, falloffDistance=4000, attenuationShapeExtents {x=1000}}
 concurrencyOverrides {maxCount=4}
}
