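# Sound cue for the Ancient Weapon's environment impact sound, in the melee weapon sound class, limited to two concurrent instances per owner.
# IsLocal and IsEnemy branches select local, enemy or non-enemy mixes, with enemy as the fallback when IsEnemy is unset.
# Each mix layers random close and slightly quieter distant waves through modulators and separate attenuations.
# All three mixes currently use identical waves and settings, so the branches sound alike.
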SoundCue /AncientWeapon/Assets/Audio/AncientWeapon_Impact_Env_Cue
root=Branch_1
nodes [
 {
  id=Attenuation_0
  children [Modulator_2]
  properties {attenuationSettings=/AncientWeapon/Assets/Audio/Audio_Settings/Ancient_Weapon_Impact_World_Distant_Att}
 }
 {
  id=Attenuation_1
  children [Modulator_3]
  properties {attenuationSettings=/AncientWeapon/Assets/Audio/Audio_Settings/Ancient_Weapon_Impact_World_Close_Att}
 }
 {
  id=Attenuation_2
  children [Modulator_0]
  properties {attenuationSettings=&Ancient_Weapon_Impact_World_Close_Att}
 }
 {
  id=Attenuation_3
  children [Modulator_4]
  properties {attenuationSettings=&Ancient_Weapon_Impact_World_Distant_Att}
 }
 {
  id=Attenuation_4
  children [Modulator_5]
  properties {attenuationSettings=&Ancient_Weapon_Impact_World_Close_Att}
 }
 {
  id=Attenuation_5
  children [Modulator_6]
  properties {attenuationSettings=&Ancient_Weapon_Impact_World_Distant_Att}
 }
 {id=Branch_1, children [Mixer_0,Branch_2,Branch_2], properties {boolParameterName=IsLocal}}
 {id=Branch_2, children [Mixer_1,Mixer_2,Mixer_1], properties {boolParameterName=IsEnemy}}
 {id=Mixer_0, children [Attenuation_2, Attenuation_0], properties {inputVolume [1,0.9]}}
 {id=Mixer_1, children [Attenuation_1, Attenuation_3], properties {inputVolume [1,0.9]}}
 {id=Mixer_2, children [Attenuation_4, Attenuation_5], properties {inputVolume [1,0.9]}}
 {id=Modulator_0, children [Random_2], properties {volumeMax=1, volumeMin=1}}
 {id=Modulator_2, children [Random_3], properties {volumeMax=1, volumeMin=1}}
 {id=Modulator_3, children [Random_4], properties {volumeMax=1, volumeMin=1}}
 {id=Modulator_4, children [Random_5], properties {volumeMax=1, volumeMin=1}}
 {id=Modulator_5, children [Random_6], properties {volumeMax=1, volumeMin=1}}
 {id=Modulator_6, children [Random_7], properties {volumeMax=1, volumeMin=1}}
 {
  id=Random_2
  children [WavePlayer_6, WavePlayer_9, WavePlayer_10, WavePlayer_11]
  properties {bShouldExcludeFromBranchCulling=true, weights [1,1,1,1]}
 }
 {
  id=Random_3
  children [WavePlayer_12, WavePlayer_13, WavePlayer_14, WavePlayer_15, WavePlayer_16, WavePlayer_17]
  properties {bShouldExcludeFromBranchCulling=true, weights [1,1,1,1,1,1]}
 }
 {
  id=Random_4
  children [WavePlayer_18, WavePlayer_19, WavePlayer_20, WavePlayer_21]
  properties {bShouldExcludeFromBranchCulling=true, weights [1,1,1,1]}
 }
 {
  id=Random_5
  children [WavePlayer_22, WavePlayer_23, WavePlayer_24, WavePlayer_25, WavePlayer_26, WavePlayer_27]
  properties {bShouldExcludeFromBranchCulling=true, weights [1,1,1,1,1,1]}
 }
 {
  id=Random_6
  children [WavePlayer_28, WavePlayer_29, WavePlayer_30, WavePlayer_31]
  properties {bShouldExcludeFromBranchCulling=true, weights [1,1,1,1]}
 }
 {
  id=Random_7
  children [WavePlayer_32, WavePlayer_33, WavePlayer_34, WavePlayer_35, WavePlayer_36, WavePlayer_37]
  properties {bShouldExcludeFromBranchCulling=true, weights [1,1,1,1,1,1]}
 }
 {id=WavePlayer_10, properties {soundWaveAssetPtr=&/AncientWeapon_Impact_Close_03}}
 {id=WavePlayer_11, properties {soundWaveAssetPtr=&/AncientWeapon_Impact_Close_04}}
 {id=WavePlayer_12, properties {soundWaveAssetPtr=&/AncientWeapon_Impact_Distant_01}}
 {id=WavePlayer_13, properties {soundWaveAssetPtr=&/AncientWeapon_Impact_Distant_02}}
 {id=WavePlayer_14, properties {soundWaveAssetPtr=&/AncientWeapon_Impact_Distant_03}}
 {id=WavePlayer_15, properties {soundWaveAssetPtr=&/AncientWeapon_Impact_Distant_04}}
 {id=WavePlayer_16, properties {soundWaveAssetPtr=&/AncientWeapon_Impact_Distant_05}}
 {id=WavePlayer_17, properties {soundWaveAssetPtr=&/AncientWeapon_Impact_Distant_06}}
 {id=WavePlayer_18, properties {soundWaveAssetPtr=&/AncientWeapon_Impact_Close_01}}
 {id=WavePlayer_19, properties {soundWaveAssetPtr=&/AncientWeapon_Impact_Close_02}}
 {id=WavePlayer_20, properties {soundWaveAssetPtr=&AncientWeapon_Impact_Close_03}}
 {id=WavePlayer_21, properties {soundWaveAssetPtr=&AncientWeapon_Impact_Close_04}}
 {id=WavePlayer_22, properties {soundWaveAssetPtr=&AncientWeapon_Impact_Distant_01}}
 {id=WavePlayer_23, properties {soundWaveAssetPtr=&AncientWeapon_Impact_Distant_02}}
 {id=WavePlayer_24, properties {soundWaveAssetPtr=&AncientWeapon_Impact_Distant_03}}
 {id=WavePlayer_25, properties {soundWaveAssetPtr=&AncientWeapon_Impact_Distant_04}}
 {id=WavePlayer_26, properties {soundWaveAssetPtr=&AncientWeapon_Impact_Distant_05}}
 {id=WavePlayer_27, properties {soundWaveAssetPtr=&AncientWeapon_Impact_Distant_06}}
 {id=WavePlayer_28, properties {soundWaveAssetPtr=&AncientWeapon_Impact_Close_01}}
 {id=WavePlayer_29, properties {soundWaveAssetPtr=&AncientWeapon_Impact_Close_02}}
 {id=WavePlayer_30, properties {soundWaveAssetPtr=&AncientWeapon_Impact_Close_03}}
 {id=WavePlayer_31, properties {soundWaveAssetPtr=&AncientWeapon_Impact_Close_04}}
 {id=WavePlayer_32, properties {soundWaveAssetPtr=&AncientWeapon_Impact_Distant_01}}
 {id=WavePlayer_33, properties {soundWaveAssetPtr=&AncientWeapon_Impact_Distant_02}}
 {id=WavePlayer_34, properties {soundWaveAssetPtr=&AncientWeapon_Impact_Distant_03}}
 {id=WavePlayer_35, properties {soundWaveAssetPtr=&AncientWeapon_Impact_Distant_04}}
 {id=WavePlayer_36, properties {soundWaveAssetPtr=&AncientWeapon_Impact_Distant_05}}
 {id=WavePlayer_37, properties {soundWaveAssetPtr=&AncientWeapon_Impact_Distant_06}}
 {id=WavePlayer_6, properties {soundWaveAssetPtr=&AncientWeapon_Impact_Close_01}}
 {id=WavePlayer_9, properties {soundWaveAssetPtr=&AncientWeapon_Impact_Close_02}}
]
properties {
 bOverrideConcurrency=true
 maxDistance=14000
 soundClassObject=&Fort_WeaponMelee
 volumeMultiplier=1
 concurrencyOverrides {bLimitToOwner=true, maxCount=2}
}
