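# Sound cue for the Ancient Weapon player impact, playing a prototype Fortanium hit wave in the melee weapon sound class.
# An IsLocal branch plays the hit without an attenuation node for the local player, and through the borrowed Turn_InAir attenuation when false or unset.
# Both paths share a pitch-randomizing modulator, with another modulator at the root for extra variation.
# A concurrency override caps simultaneous instances per owner.
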SoundCue /AncientWeapon/Assets/Audio/AncientWeapon_Impact_Player_Cue
root=Modulator_0
nodes [
 {
  id=Attenuation_2
  children [Modulator_1]
  properties {attenuationSettings=/AncientWeapon/Assets/Audio/Audio_Settings/Ancient_Weapon_Turn_InAir_Att}
 }
 {id=Branch_0, children [Modulator_1, Attenuation_2, Attenuation_2], properties {boolParameterName=IsLocal}}
 {id=Modulator_0, children [Branch_0]}
 {id=Modulator_1, children [WavePlayer_6], properties {pitchMax=1.1, pitchMin=0.9}}
 {id=WavePlayer_6, properties {soundWaveAssetPtr=/AncientWeapon/Assets/Audio/Proto/Proto_AncientWeapon_Fortanium_Hit_01}}
]
properties {
 bOverrideConcurrency=true
 maxDistance=15000
 soundClassObject=&Fort_WeaponMelee
 volumeMultiplier=0.33
 concurrencyOverrides {bLimitToOwner=true, maxCount=2}
}
