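# Sound cue for the Ancient Weapon's close-range in-air projectile loop, with Doppler shift, an attenuation asset and a concurrency override.
# A Mixer blends a looping wave, faded in by an Enveloper, with a quieter, pitch-lowered layer of endlessly repeated random one-shots separated by short random delays.
# A quality-level switch draws those one-shots from ten variations on the first two levels and only three on the last two.
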SoundCue /AncientWeapon/Assets/Audio/AncientWeapon_InAir_Close_Loop
root=Doppler_0
nodes [
 {id=Delay_0, children [QualityLevel_0], properties {delayMax=0.5}}
 {id=Doppler_0, children [Mixer_0], properties {dopplerIntensity=1.02}}
 {
  id=Enveloper_2
  children [Modulator_0]
  properties {pitchCurve {editorCurveData {keys [{time=0}]}}, volumeCurve {editorCurveData {keys [<at 0 0 RCIM_Cubic>,<at 0.8 1>]}}}
 }
 {id=Looping_0, children [Delay_0]}
 {id=Mixer_0, children [Enveloper_2,Modulator_1], properties {inputVolume [1,0.3]}}
 {id=Modulator_0, children [WavePlayer_4], properties {volumeMax=1, volumeMin=1}}
 {id=Modulator_1, children [Looping_0], properties {pitchMax=0.5, pitchMin=0.5, volumeMax=1, volumeMin=1}}
 {id=QualityLevel_0, children [Random_0,Random_0,Random_2,Random_2]}
 {
  id=Random_0
  children [
   WavePlayer_11
   WavePlayer_12
   WavePlayer_13
   WavePlayer_14
   WavePlayer_15
   WavePlayer_16
   WavePlayer_17
   WavePlayer_18
   WavePlayer_19
   WavePlayer_20
  ]
  properties {weights [1,1,1,1,1,1,1,1,1,1]}
 }
 {id=Random_2, children [WavePlayer_18, WavePlayer_19, WavePlayer_20], properties {bShouldExcludeFromBranchCulling=true, weights [1,1,1]}}
 {id=WavePlayer_11, properties {soundWaveAssetPtr=&/Ancient_Weapon_InAir_Oneshot_10}}
 {id=WavePlayer_12, properties {soundWaveAssetPtr=&/Ancient_Weapon_InAir_Oneshot_09}}
 {id=WavePlayer_13, properties {soundWaveAssetPtr=&/Ancient_Weapon_InAir_Oneshot_08}}
 {id=WavePlayer_14, properties {soundWaveAssetPtr=&/Ancient_Weapon_InAir_Oneshot_07}}
 {id=WavePlayer_15, properties {soundWaveAssetPtr=&/Ancient_Weapon_InAir_Oneshot_06}}
 {id=WavePlayer_16, properties {soundWaveAssetPtr=&/Ancient_Weapon_InAir_Oneshot_05}}
 {id=WavePlayer_17, properties {soundWaveAssetPtr=&/Ancient_Weapon_InAir_Oneshot_04}}
 {id=WavePlayer_18, properties {soundWaveAssetPtr=&/Ancient_Weapon_InAir_Oneshot_03}}
 {id=WavePlayer_19, properties {soundWaveAssetPtr=&/Ancient_Weapon_InAir_Oneshot_02}}
 {id=WavePlayer_20, properties {soundWaveAssetPtr=&/Ancient_Weapon_InAir_Oneshot_01}}
 {id=WavePlayer_4, properties {bLooping=true, soundWaveAssetPtr=&/AncientWeapon_InAirLoop_03}}
]
properties {
 attenuationSettings=/AncientWeapon/Assets/Audio/Audio_Settings/Ancient_Weapon_InAir_Close_Loop_Att
 bOverrideConcurrency=true
 maxDistance=10000
 pitchMultiplier=1.5
 soundClassObject=&Fort_WeaponRanged_Projectile_Effects
 volumeMultiplier=1.5
 attenuationOverrides {falloffDistance=5000, attenuationShapeExtents {x=1000}}
 concurrencyOverrides {maxCount=4}
}
