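# Sound Cue for the looping in-air sound of Ancient Weapon projectiles, routed to the ranged weapon projectile effects sound class.
# A Random node picks one of three looping wave variations with equal weight and feeds a Doppler node at the root for pitch shift with relative motion.
# It uses a dedicated in-air loop attenuation asset and overrides concurrency to cap simultaneous instances.
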SoundCue /AncientWeapon/Assets/Audio/AncientWeapon_InAir_Loop
root=Doppler_0
nodes [
 {id=Doppler_0, children [Random_0], properties {dopplerIntensity=1.02}}
 {id=Random_0, children [WavePlayer_2,WavePlayer_3,WavePlayer_4], properties {weights [1,1,1]}}
 {id=WavePlayer_2, properties {bLooping=true, soundWaveAssetPtr=&/AncientWeapon_InAirLoop_01}}
 {id=WavePlayer_3, properties {bLooping=true, soundWaveAssetPtr=&/AncientWeapon_InAirLoop_02}}
 {id=WavePlayer_4, properties {bLooping=true, soundWaveAssetPtr=&/AncientWeapon_InAirLoop_03}}
]
properties {
 attenuationSettings=/AncientWeapon/Assets/Audio/Audio_Settings/Ancient_Weapon_InAir_Loop_Att
 bOverrideConcurrency=true
 maxDistance=20000
 soundClassObject=&Fort_WeaponRanged_Projectile_Effects
 volumeMultiplier=1.5
 attenuationOverrides {falloffDistance=5000, attenuationShapeExtents {x=1000}}
 concurrencyOverrides {maxCount=4}
}
