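# Looping stereo MetaSound for the thrown Ancient Weapon in flight and returning.
# A random thrown loop fades into the return loop on recall or turn, using chained ADSR envelopes.
# Distance drives a blended pitch shift, elevation drives filter gain, and turn or explosion triggers duck the loops.
# Random-pitched one-shots mark turns and, for the local player during return, recall or distance-threshold crossings, with incoming volume scaled by distance.
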MetaSoundSource /AncientWeapon/Assets/Audio/AncientWeapon_InAir_Loop_MSS
inputs [
 IsLocal {type=Bool}
 LocalCharges {type=Int32}
 OnExplosion {type=Trigger}
 OnRecall {type=Trigger}
 OnStop {type=Trigger}
 OnTurnInAir {type=Trigger}
 ReturnLoop {type=WaveAsset}
 ReturnOneShot {type=WaveAsset:Array}
 ThrownLoop {type=WaveAsset:Array}
 TurnInAirOneShot {type=WaveAsset:Array}
 UE.Attenuation.Distance {type=Float}
 UE.Source.OnPlay {type=Trigger}
 UE.Spatialization.Azimuth {type=Float}
 UE.Spatialization.Elevation {type=Float}
]
interfaces [UE.Attenuation:1.0, UE.OutputFormat.Stereo:1.0, UE.Source:1.1, UE.Spatialization:1.0]
outputs [
 "UE.OutputFormat.Stereo.Audio:0" {type=Audio}
 "UE.OutputFormat.Stereo.Audio:1" {type=Audio}
]
pages [Default {
  input_defaults {
   IsLocal=null
   LocalCharges=3
   OnExplosion=null
   OnRecall=null
   OnStop=null
   OnTurnInAir=null
   ReturnLoop=&/AncientWeapon_InAirLoop_03
   UE.Attenuation.Distance=10000
   UE.Source.OnPlay=null
   UE.Spatialization.Azimuth=0.464
   UE.Spatialization.Elevation=1
   ReturnOneShot [
    &/Ancient_Weapon_Incoming_01
    &/Ancient_Weapon_Incoming_02
    &/Ancient_Weapon_Incoming_03
   ]
   ThrownLoop [&/AncientWeapon_InAirLoop_02, &/AncientWeapon_InAirLoop_01]
   TurnInAirOneShot [&/AncientWeapon_TurnInAir_01]
  }
  nodes [
   TriggerAnyTriggerAny2_0 = "TriggerAny.Trigger Any (2)"(In 0=^OnRecall, In 1=^OnTurnInAir)
   ADSREnvelopeADSREnvelope_1 = ADSR Envelope.ADSR Envelope:Audio(Trigger Attack=^UE.Source.OnPlay, Trigger Release=@TriggerAnyTriggerAny2_0.Out) {also {defaults {Attack Time=0.5, Hard Reset="<MetaSoundBuilderResult.SUCCEEDED: 0>", Release Curve=0.5, Release Time=0.5, Sustain Level=1}}}
   n1 = InitVariable.Trigger()
   ADSREnvelopeADSREnvelope_0 = ADSR Envelope.ADSR Envelope:Audio(
    Trigger Release=^OnStop
    Trigger Attack=VariableAccessor.Trigger(Variable=VariableMutator.Trigger(Value=@ADSREnvelopeADSREnvelope_1.On Release Triggered, Variable=@n1.Variable).Variable).Value
   ) {also {defaults {Attack Curve=2, Attack Time=0.5, Hard Reset="<MetaSoundBuilderResult.SUCCEEDED: 0>", Release Time=0.3, Sustain Level=1}}}
   TriggerAnyTriggerAny2_2 = "TriggerAny.Trigger Any (2)"(In 0=^OnRecall, In 1=^OnTurnInAir)
   WavePlayer_1 = UE.Wave Player:Stereo(
    Wave Asset=^ReturnLoop
    Stop=@ADSREnvelopeADSREnvelope_0.On Done
    Start Time=&MSP_WaveGetRandomStartTime(Sound=^ReturnLoop).Value
    Play=@TriggerAnyTriggerAny2_2.Out
   ) {also {defaults {Loop="<MetaSoundBuilderResult.SUCCEEDED: 0>"}}}
   ArrayRandomGet_2 = Array.Random Get:WaveAsset:Array(In Array=^ThrownLoop, Next=^UE.Source.OnPlay)
   WavePlayer_0 = UE.Wave Player:Stereo(
    Stop=@ADSREnvelopeADSREnvelope_1.On Done
    Play=@ArrayRandomGet_2.On Next
    Wave Asset=@ArrayRandomGet_2.Value
    Start Time=&MSP_WaveGetRandomStartTime(Sound=@ArrayRandomGet_2.Value).Value
   ) {also {defaults {Loop="<MetaSoundBuilderResult.SUCCEEDED: 0>"}}}
   AudioMixerAudioMixerStereo3_1 = "AudioMixer.Audio Mixer (Stereo, 3)"(
    In 1 R=UE.Multiply:Audio(PrimaryOperand=@ADSREnvelopeADSREnvelope_0.Out Envelope, AdditionalOperands=@WavePlayer_1.Out Right).Out
    In 1 L=UE.Multiply:Audio(PrimaryOperand=@ADSREnvelopeADSREnvelope_0.Out Envelope, AdditionalOperands=@WavePlayer_1.Out Left).Out
    In 0 R=UE.Multiply:Audio(PrimaryOperand=@ADSREnvelopeADSREnvelope_1.Out Envelope, AdditionalOperands=@WavePlayer_0.Out Right).Out
    In 0 L=UE.Multiply:Audio(PrimaryOperand=@ADSREnvelopeADSREnvelope_1.Out Envelope, AdditionalOperands=@WavePlayer_0.Out Left).Out
   )
   FrequencyMultipliertoSemitone_0 = UE.Frequency Multiplier to Semitone:Float(Frequency Multiplier=MapRange.MapRange:Float(In=^UE.Attenuation.Distance) {also {defaults {In Range B=8000, Out Range A=1.2}}}.Out Value)
   n2 = VariableAccessor.Float(Variable=VariableMutator.Float(Value=@FrequencyMultipliertoSemitone_0.Semitones, Variable=InitVariable.Float().Variable).Variable)
   DelayPitchShift_0 = UE.Delay Pitch Shift:Audio(In=@AudioMixerAudioMixerStereo3_1.Out R, Pitch Shift=@n2.Value)
   DelayPitchShift_1 = UE.Delay Pitch Shift:Audio(In=@AudioMixerAudioMixerStereo3_1.Out L, Pitch Shift=@n2.Value)
   AudioMixerAudioMixerStereo2_0 = "AudioMixer.Audio Mixer (Stereo, 2)"(
    In 0 L=@AudioMixerAudioMixerStereo3_1.Out L
    In 0 R=@AudioMixerAudioMixerStereo3_1.Out R
    In 1 R=@DelayPitchShift_0.Out
    In 1 L=@DelayPitchShift_1.Out
   ) {also {defaults {Gain 0=0.3, Gain 1=0.7}}}
   LinearGaintoDecibels_0 = UE.Linear Gain to Decibels:Float(Linear Gain=MapRange.MapRange:Float(In=^UE.Spatialization.Elevation) {also {defaults {Out Range A=1, Out Range B=0.25}}}.Out Value)
   n3 = VariableAccessor.Float(Variable=VariableMutator.Float(Variable=InitVariable.Float().Variable, Value=@LinearGaintoDecibels_0.Decibels).Variable)
   BiquadFilter_0 = UE.Biquad Filter:Audio(In=@AudioMixerAudioMixerStereo2_0.Out L, Gain=@n3.Value) {also {defaults {Cutoff Frequency=100, Type=5}}}
   n4 = InitVariable.Trigger()
   ArrayRandomGet_1 = Array.Random Get:WaveAsset:Array(Next=^OnTurnInAir, In Array=^TurnInAirOneShot) {also {defaults {Weights []}}}
   n5 = UE.RandomFloat(Next=@ArrayRandomGet_1.On Next) {also {defaults {Max=1.05, Min=0.95}}}
   FrequencyMultipliertoSemitone_2 = UE.Frequency Multiplier to Semitone:Float(Frequency Multiplier=@n5.Value)
   WavePlayer_2 = UE.Wave Player:Stereo(
    Pitch Shift=@FrequencyMultipliertoSemitone_2.Semitones
    Play=@n5.On Next
    Wave Asset=UE.Reroute(Value=UE.Reroute(Value=@ArrayRandomGet_1.Value).Value).Value
   )
   TriggerAnyTriggerAny2_1 = "TriggerAny.Trigger Any (2)"(
    In 0=^OnExplosion
    In 1=VariableAccessor.Trigger(Variable=VariableMutator.Trigger(Variable=@n4.Variable, Value=@WavePlayer_2.On Play).Variable).Value
   )
   ADEnvelopeADEnvelope_0 = AD Envelope.AD Envelope:Float(Trigger=@TriggerAnyTriggerAny2_1.Out) {also {defaults {Attack Time=0.2, Hard Reset="<MetaSoundBuilderResult.SUCCEEDED: 0>"}}}
   n6 = UE.Multiply:Float(PrimaryOperand=@ADEnvelopeADEnvelope_0.Out Envelope) {also {defaults {AdditionalOperands=0.5}}}
   n7 = VariableAccessor.Float(Variable=VariableMutator.Float(Variable=InitVariable.Float().Variable, Value=UE.Subtract:Float(AdditionalOperands=@n6.Out) {also {defaults {PrimaryOperand=1}}}.Out).Variable)
   n8 = MapRange.MapRange:Float(In=^UE.Attenuation.Distance) {also {defaults {In Range B=7000, Out Range A=0.65, Out Range B=0}}}
   TriggerAnyTriggerAny2_3 = "TriggerAny.Trigger Any (2)"(In 0=^OnRecall, In 1=UE.TriggerOnThreshold:Float(In=^UE.Attenuation.Distance) {also {defaults {Threshold=6000, Type=1}}}.Out)
   n9 = VariableMutator.Bool(Variable=InitVariable.Bool().Variable, Value=Value.Value:Bool(Reset=@WavePlayer_0.On Play, Set=@WavePlayer_1.On Play).Output Value)
   ArrayRandomGet_0 = Array.Random Get:WaveAsset:Array(
    In Array=^ReturnOneShot
    Next=TriggerCompare.Clamp:Bool(
     Compare=TriggerCompare.Clamp:Bool(A=^IsLocal, Compare=@TriggerAnyTriggerAny2_3.Out) {also {defaults {B="<MetaSoundBuilderResult.SUCCEEDED: 0>"}}}.TRUE
     A=VariableAccessor.Bool(Variable=@n9.Variable).Value
    ) {also {defaults {B="<MetaSoundBuilderResult.SUCCEEDED: 0>"}}}.TRUE
   )
   n10 = UE.RandomFloat(Next=@ArrayRandomGet_0.On Next) {also {defaults {Max=1.1, Min=0.9}}}
   FrequencyMultipliertoSemitone_1 = UE.Frequency Multiplier to Semitone:Float(Frequency Multiplier=@n10.Value)
   WavePlayer_3 = UE.Wave Player:Stereo(
    Pitch Shift=@FrequencyMultipliertoSemitone_1.Semitones
    Play=@n10.On Next
    Wave Asset=UE.Reroute(Value=UE.Reroute(Value=@ArrayRandomGet_0.Value).Value).Value
   )
   BiquadFilter_1 = UE.Biquad Filter:Audio(In=@AudioMixerAudioMixerStereo2_0.Out R, Gain=@n3.Value) {also {defaults {Cutoff Frequency=100, Type=5}}}
   AudioMixerAudioMixerStereo3_0 = "AudioMixer.Audio Mixer (Stereo, 3)"(
    In 0 L=UE.Multiply:Audio by Float(PrimaryOperand=@BiquadFilter_0.Out, AdditionalOperands=@n7.Value).Out
    In 2 R=UE.Multiply:Audio by Float(AdditionalOperands=@n8.Out Value, PrimaryOperand=@WavePlayer_3.Out Right).Out
    In 0 R=UE.Multiply:Audio by Float(PrimaryOperand=@BiquadFilter_1.Out, AdditionalOperands=@n7.Value).Out
    In 2 L=UE.Multiply:Audio by Float(AdditionalOperands=@n8.Out Value, PrimaryOperand=@WavePlayer_3.Out Left).Out
    In 1 L=@WavePlayer_2.Out Left
    In 1 R=@WavePlayer_2.Out Right
   ) {also {defaults {Gain 0=1.3, Gain 1=0.5, Gain 2=1.4}}}
   InitVariable.Float()
  ]
  links [
   {dst="@UE.OutputFormat.Stereo.Audio:0", src=AudioMixerAudioMixerStereo3_0.Out L}
   {dst="@UE.OutputFormat.Stereo.Audio:1", src=AudioMixerAudioMixerStereo3_0.Out R}
  ]
 }]
properties {
 assetImportData=&self:AssetImportData
 attenuationSettings=/AncientWeapon/Assets/Audio/Audio_Settings/Ancient_Weapon_InAir_Loop_Att
 bLooping=true
 bOverrideConcurrency=true
 duration=10000
 maxDistance=20000
 numChannels=2
 outputFormat=Stereo
 soundClassObject=&Fort_WeaponRanged_Projectile_Effects
 volume=2.35
 concurrencyOverrides {maxCount=4}
 concurrencySet [null]
}
