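# Sound cue for picking up the Ancient Weapon, playing one of three prototype pickup sounds.
# An equal-weight Random node picks the wave and feeds a Modulator at the root.
# The cue plays louder and higher-pitched than the source waves, using shared weapon-pickup attenuation, the loot-collect sound class and a pickup concurrency limit.
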SoundCue /AncientWeapon/Assets/Audio/AncientWeapon_Pickup_Cue
root=Modulator_0
nodes [
 {id=Modulator_0, children [Random_1]}
 {id=Random_1, children [WavePlayer_0,WavePlayer_4,WavePlayer_5], properties {weights [1,1,1]}}
 {id=WavePlayer_0, properties {soundWaveAssetPtr=/AncientWeapon/Assets/Audio/Proto/AncientWeapon_PickUp_01}}
 {id=WavePlayer_4, properties {soundWaveAssetPtr=/AncientWeapon/Assets/Audio/Proto/AncientWeapon_PickUp_02}}
 {id=WavePlayer_5, properties {soundWaveAssetPtr=/AncientWeapon/Assets/Audio/Proto/AncientWeapon_PickUp_03}}
]
properties {
 attenuationSettings=&Fort_Weapon_Pickup_Att
 maxDistance=3500
 pitchMultiplier=1.3
 soundClassObject=&Fort_Loot_Collect
 volumeMultiplier=1.6
 concurrencySet [&PickupMaximum_Oldest]
}
