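# Sound cue for the close-range projectile whip-by of the Ancient Weapon, routed to the ranged weapon projectile effects sound class.
# A Modulator feeds a quality-level switch: the first two levels randomly pick from eight equally weighted waves, while the other two draw only from the last three.
# It uses the shared Athena bullet whip attenuation, a reduced volume, and a concurrency override capping simultaneous instances.
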SoundCue /AncientWeapon/Assets/Audio/AncientWeapon_ProjectileWhip_Close_Cue
root=Modulator_2
nodes [
 {id=Modulator_2, children [QualityLevel_2]}
 {id=QualityLevel_2, children [Random_4,Random_4,Random_5,Random_5]}
 {
  id=Random_4
  children [WavePlayer_0,WavePlayer_1,WavePlayer_2,WavePlayer_3,WavePlayer_4,WavePlayer_5,WavePlayer_6,WavePlayer_7]
  properties {weights [1,1,1,1,1,1,1,1]}
 }
 {id=Random_5, children [WavePlayer_5,WavePlayer_6,WavePlayer_7], properties {weights [1,1,1]}}
 {id=WavePlayer_0, properties {soundWaveAssetPtr=&/AncientWeapon_ProjectileWhip_Close_01}}
 {id=WavePlayer_1, properties {soundWaveAssetPtr=&/AncientWeapon_ProjectileWhip_Close_02}}
 {id=WavePlayer_2, properties {soundWaveAssetPtr=&/AncientWeapon_ProjectileWhip_Close_03}}
 {id=WavePlayer_3, properties {soundWaveAssetPtr=&/AncientWeapon_ProjectileWhip_Close_04}}
 {id=WavePlayer_4, properties {soundWaveAssetPtr=&/AncientWeapon_ProjectileWhip_Close_05}}
 {id=WavePlayer_5, properties {soundWaveAssetPtr=&/AncientWeapon_ProjectileWhip_Close_06}}
 {id=WavePlayer_6, properties {soundWaveAssetPtr=&/AncientWeapon_ProjectileWhip_Close_07}}
 {id=WavePlayer_7, properties {soundWaveAssetPtr=&/AncientWeapon_ProjectileWhip_Close_08}}
]
properties {
 attenuationSettings=/Game/Athena/Sounds/Weapons/BulletWhips/bullet_whip_Att.Bullet_Whip_Att
 bOverrideConcurrency=true
 maxDistance=2097152
 soundClassObject=&Fort_WeaponRanged_Projectile_Effects
 volumeMultiplier=0.4
 concurrencyOverrides {maxCount=4}
}
