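# Sound cue for the distant projectile whip-by of the Ancient Weapon, using shared bullet whip attenuation and the ranged projectile effects sound class.
# A root Modulator varies playback over a quality switch that feeds equally weighted random pickers.
# The first two quality slots draw from all eight waves, while the last two reuse only waves 06 to 08.
# A concurrency override caps simultaneous instances at four.
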SoundCue /AncientWeapon/Assets/Audio/AncientWeapon_ProjectileWhip_Distant_Cue
root=Modulator_2
nodes [
 {id=Modulator_2, children [QualityLevel_2]}
 {id=QualityLevel_2, children [Random_4,Random_4,Random_5,Random_5]}
 {
  id=Random_4
  children [WavePlayer_8, WavePlayer_9, WavePlayer_10, WavePlayer_11, WavePlayer_12, WavePlayer_13, WavePlayer_14, WavePlayer_15]
  properties {weights [1,1,1,1,1,1,1,1]}
 }
 {id=Random_5, children [WavePlayer_13, WavePlayer_14, WavePlayer_15], properties {weights [1,1,1]}}
 {id=WavePlayer_10, properties {soundWaveAssetPtr=&/AncientWeapon_ProjectileWhip_Distant_03}}
 {id=WavePlayer_11, properties {soundWaveAssetPtr=&/AncientWeapon_ProjectileWhip_Distant_04}}
 {id=WavePlayer_12, properties {soundWaveAssetPtr=&/AncientWeapon_ProjectileWhip_Distant_05}}
 {id=WavePlayer_13, properties {soundWaveAssetPtr=&/AncientWeapon_ProjectileWhip_Distant_06}}
 {id=WavePlayer_14, properties {soundWaveAssetPtr=&/AncientWeapon_ProjectileWhip_Distant_07}}
 {id=WavePlayer_15, properties {soundWaveAssetPtr=&/AncientWeapon_ProjectileWhip_Distant_08}}
 {id=WavePlayer_8, properties {soundWaveAssetPtr=&/AncientWeapon_ProjectileWhip_Distant_01}}
 {id=WavePlayer_9, properties {soundWaveAssetPtr=&/AncientWeapon_ProjectileWhip_Distant_02}}
]
properties {
 attenuationSettings=/Game/Athena/Sounds/Weapons/BulletWhips/bullet_whip_Att.Bullet_Whip_Att
 bOverrideConcurrency=true
 maxDistance=2097152
 soundClassObject=&Fort_WeaponRanged_Projectile_Effects
 volumeMultiplier=0.4
 concurrencyOverrides {maxCount=4}
}
