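# Sound cue for the Ancient Weapon's recall explosion, mixing a close branch and a distant branch, each with its own attenuation.
# Each branch plays a random zap, followed after a short delay by a pitched-up metal layer mixed with a close or distant explosion.
# Quality-level nodes switch those delayed layers between full and reduced random wave pools, and concurrency caps playback at two instances, stopping the oldest.
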SoundCue /AncientWeapon/Assets/Audio/AncientWeapon_Recall_Explosion_Cue
root=Mixer_2
nodes [
 {
  id=Attenuation_0
  children [Mixer_5]
  properties {attenuationSettings=/AncientWeapon/Assets/Audio/Audio_Settings/Ancient_Weapon_Explode_Distant_Att}
 }
 {id=Attenuation_1, children [Mixer_4], properties {attenuationSettings=/Game/Sounds/Attenuation/Fort_Grenade_Smoke_Att}}
 {id=Delay_0, children [Mixer_3], properties {delayMax=0.1, delayMin=0.1}}
 {id=Delay_1, children [Mixer_0], properties {delayMax=0.1, delayMin=0.1}}
 {id=Mixer_0, children [Modulator_2,Modulator_6], properties {inputVolume [0.5,2]}}
 {id=Mixer_2, children [Attenuation_1, Attenuation_0], properties {inputVolume [1,1]}}
 {id=Mixer_3, children [Modulator_3,Modulator_4], properties {inputVolume [0.5,1]}}
 {id=Mixer_4, children [Delay_0,Modulator_7], properties {inputVolume [0.9,5]}}
 {id=Mixer_5, children [Delay_1,Modulator_8], properties {inputVolume [0.9,5]}}
 {id=Modulator_2, children [QualityLevel_1], properties {pitchMax=2, pitchMin=2}}
 {id=Modulator_3, children [QualityLevel_2], properties {pitchMax=2, pitchMin=2}}
 {id=Modulator_4, children [QualityLevel_3], properties {pitchMax=1.1, pitchMin=0.9}}
 {id=Modulator_6, children [QualityLevel_7], properties {pitchMax=1, pitchMin=1}}
 {id=Modulator_7, children [Random_16], properties {pitchMax=1.1, pitchMin=0.9}}
 {id=Modulator_8, children [Random_17], properties {pitchMax=1.1, pitchMin=0.9}}
 {id=QualityLevel_1, children [Random_3,Random_3,Random_4,Random_4]}
 {id=QualityLevel_2, children [Random_6,Random_6,Random_9,Random_9]}
 {id=QualityLevel_3, children [Random_10,Random_10,Random_11,Random_11]}
 {id=QualityLevel_7, children [Random_14,Random_14,Random_15,Random_15]}
 {
  id=Random_10
  children [WavePlayer_30, WavePlayer_31, WavePlayer_32, WavePlayer_33, WavePlayer_34, WavePlayer_35, WavePlayer_36, WavePlayer_37]
  properties {weights [1,1,1,1,1,1,1,1]}
 }
 {id=Random_11, children [WavePlayer_35, WavePlayer_36, WavePlayer_37], properties {bShouldExcludeFromBranchCulling=true, weights [1,1,1]}}
 {
  id=Random_14
  children [WavePlayer_38, WavePlayer_39, WavePlayer_40, WavePlayer_41, WavePlayer_42, WavePlayer_43, WavePlayer_44, WavePlayer_45]
  properties {weights [1,1,1,1,1,1,1,1]}
 }
 {id=Random_15, children [WavePlayer_43, WavePlayer_44, WavePlayer_45], properties {bShouldExcludeFromBranchCulling=true, weights [1,1,1]}}
 {id=Random_16, children [WavePlayer_46, WavePlayer_47], properties {bShouldExcludeFromBranchCulling=true, weights [1,1]}}
 {id=Random_17, children [WavePlayer_48, WavePlayer_49], properties {bShouldExcludeFromBranchCulling=true, weights [1,1]}}
 {
  id=Random_3
  children [WavePlayer_11, WavePlayer_12, WavePlayer_13, WavePlayer_17, WavePlayer_18, WavePlayer_19, WavePlayer_20, WavePlayer_21]
  properties {weights [1,1,1,1,1,1,1,1]}
 }
 {id=Random_4, children [WavePlayer_19, WavePlayer_20, null], properties {bShouldExcludeFromBranchCulling=true, weights [1,1,1]}}
 {
  id=Random_6
  children [WavePlayer_22, WavePlayer_23, WavePlayer_24, WavePlayer_25, WavePlayer_26, WavePlayer_27, WavePlayer_28, WavePlayer_29]
  properties {weights [1,1,1,1,1,1,1,1]}
 }
 {id=Random_9, children [WavePlayer_27, WavePlayer_28, WavePlayer_29], properties {bShouldExcludeFromBranchCulling=true, weights [1,1,1]}}
 {id=WavePlayer_11, properties {soundWaveAssetPtr=&/Amcient_Weapon_Explode_Metal_01}}
 {id=WavePlayer_12, properties {soundWaveAssetPtr=&/Amcient_Weapon_Explode_Metal_02}}
 {id=WavePlayer_13, properties {soundWaveAssetPtr=&/Amcient_Weapon_Explode_Metal_03}}
 {id=WavePlayer_17, properties {soundWaveAssetPtr=&/Amcient_Weapon_Explode_Metal_04}}
 {id=WavePlayer_18, properties {soundWaveAssetPtr=&/Amcient_Weapon_Explode_Metal_05}}
 {id=WavePlayer_19, properties {soundWaveAssetPtr=&/Amcient_Weapon_Explode_Metal_06}}
 {id=WavePlayer_20, properties {soundWaveAssetPtr=&/Amcient_Weapon_Explode_Metal_07}}
 {id=WavePlayer_21, properties {soundWaveAssetPtr=&/Amcient_Weapon_Explode_Metal_08}}
 {id=WavePlayer_22, properties {soundWaveAssetPtr=&Amcient_Weapon_Explode_Metal_01}}
 {id=WavePlayer_23, properties {soundWaveAssetPtr=&Amcient_Weapon_Explode_Metal_02}}
 {id=WavePlayer_24, properties {soundWaveAssetPtr=&Amcient_Weapon_Explode_Metal_03}}
 {id=WavePlayer_25, properties {soundWaveAssetPtr=&Amcient_Weapon_Explode_Metal_04}}
 {id=WavePlayer_26, properties {soundWaveAssetPtr=&Amcient_Weapon_Explode_Metal_05}}
 {id=WavePlayer_27, properties {soundWaveAssetPtr=&Amcient_Weapon_Explode_Metal_06}}
 {id=WavePlayer_28, properties {soundWaveAssetPtr=&Amcient_Weapon_Explode_Metal_07}}
 {id=WavePlayer_29, properties {soundWaveAssetPtr=&Amcient_Weapon_Explode_Metal_08}}
 {id=WavePlayer_30, properties {soundWaveAssetPtr=&/Amcient_Weapon_Explode_Close_01}}
 {id=WavePlayer_31, properties {soundWaveAssetPtr=&/Amcient_Weapon_Explode_Close_02}}
 {id=WavePlayer_32, properties {soundWaveAssetPtr=&/Amcient_Weapon_Explode_Close_03}}
 {id=WavePlayer_33, properties {soundWaveAssetPtr=&/Amcient_Weapon_Explode_Close_04}}
 {id=WavePlayer_34, properties {soundWaveAssetPtr=&/Amcient_Weapon_Explode_Close_05}}
 {id=WavePlayer_35, properties {soundWaveAssetPtr=&/Amcient_Weapon_Explode_Close_06}}
 {id=WavePlayer_36, properties {soundWaveAssetPtr=&/Amcient_Weapon_Explode_Close_07}}
 {id=WavePlayer_37, properties {soundWaveAssetPtr=&/Amcient_Weapon_Explode_Close_08}}
 {id=WavePlayer_38, properties {soundWaveAssetPtr=&/Amcient_Weapon_Explode_Distant_01}}
 {id=WavePlayer_39, properties {soundWaveAssetPtr=&/Amcient_Weapon_Explode_Distant_02}}
 {id=WavePlayer_40, properties {soundWaveAssetPtr=&/Amcient_Weapon_Explode_Distant_03}}
 {id=WavePlayer_41, properties {soundWaveAssetPtr=&/Amcient_Weapon_Explode_Distant_04}}
 {id=WavePlayer_42, properties {soundWaveAssetPtr=&/Amcient_Weapon_Explode_Distant_05}}
 {id=WavePlayer_43, properties {soundWaveAssetPtr=&/Amcient_Weapon_Explode_Distant_06}}
 {id=WavePlayer_44, properties {soundWaveAssetPtr=&/Amcient_Weapon_Explode_Distant_07}}
 {id=WavePlayer_45, properties {soundWaveAssetPtr=&/Amcient_Weapon_Explode_Distant_08}}
 {id=WavePlayer_46, properties {soundWaveAssetPtr=&/Ancient_Weapon_Explode_Zap_01}}
 {id=WavePlayer_47, properties {soundWaveAssetPtr=&/Ancient_Weapon_Explode_Zap_02}}
 {id=WavePlayer_48, properties {soundWaveAssetPtr=&Ancient_Weapon_Explode_Zap_01}}
 {id=WavePlayer_49, properties {soundWaveAssetPtr=&Ancient_Weapon_Explode_Zap_02}}
]
properties {
 bOverrideConcurrency=true
 maxDistance=30000
 pitchMultiplier=0.8
 soundClassObject=&Fort_Gadgets_Items
 volumeMultiplier=1.5
 attenuationOverrides {distanceAlgorithm=NaturalSound, falloffDistance=3800, lPFRadiusMax=3000, lPFRadiusMin=2000, attenuationShapeExtents {x=2200}}
 concurrencyOverrides {maxCount=2, resolutionRule=StopOldest}
}
