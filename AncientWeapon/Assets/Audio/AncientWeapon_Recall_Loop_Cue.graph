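# Sound cue for the Ancient Weapon's recall loop, using the looping physics sound class and a dedicated attenuation asset.
# A mixer layers three modulated branches: a looping recall wave, a reused ADS-out one-shot at half input volume, and a random pick of three ADS-in zap one-shots.
# The two one-shot branches are pitched down; only the recall wave loops, so they act as an onset over the sustained loop.
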SoundCue /AncientWeapon/Assets/Audio/AncientWeapon_Recall_Loop_Cue
root=Mixer_0
nodes [
 {id=Mixer_0, children [Modulator_0,Modulator_1,Modulator_9], properties {inputVolume [1,0.5,1]}}
 {id=Modulator_0, children [WavePlayer_0], properties {pitchMax=1}}
 {id=Modulator_1, children [WavePlayer_18], properties {pitchMax=0.9, pitchMin=0.8}}
 {id=Modulator_9, children [Random_7], properties {pitchMax=0.9, pitchMin=0.8, volumeMax=1, volumeMin=1}}
 {id=Random_7, children [WavePlayer_21, WavePlayer_23, WavePlayer_24], properties {weights [1,1,1]}}
 {id=WavePlayer_0, properties {bLooping=true, soundWaveAssetPtr=&/Ancient_Weapon_Recall_Loop_01}}
 {id=WavePlayer_18, properties {soundWaveAssetPtr=&/Ancient_Weapon_ADS_Out}}
 {id=WavePlayer_21, properties {soundWaveAssetPtr=&/Ancient_Weapon_ADS_In_Zaps_01}}
 {id=WavePlayer_23, properties {soundWaveAssetPtr=&/Ancient_Weapon_ADS_In_Zaps_02}}
 {id=WavePlayer_24, properties {soundWaveAssetPtr=&/Ancient_Weapon_ADS_In_Zaps_03}}
]
properties {
 attenuationSettings=/AncientWeapon/Assets/Audio/Audio_Settings/Ancient_Weapon_Recall_Loop_Att
 maxDistance=1800
 soundClassObject=/Game/Sounds/FortSoundClassesAndModes/Fort_Physics_Looping
 volumeMultiplier=1.5
}
