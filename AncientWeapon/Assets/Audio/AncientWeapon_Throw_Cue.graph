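# Sound cue for the Ancient Weapon throw, mixing a throw sound with a metal layer in the melee weapon sound class.
# An IsLocal branch selects first-person or standard attenuation for the throw, both drawing randomly from the same wave set.
# Quality level nodes shrink that random pool from six variations to three on the last two tiers.
# The metal layer plays a random variation after a short fixed delay, with its own attenuation and boosted mix volume.
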SoundCue /AncientWeapon/Assets/Audio/AncientWeapon_Throw_Cue
root=Mixer_0
nodes [
 {
  id=Attenuation_0
  children [Modulator_0]
  properties {attenuationSettings=/AncientWeapon/Assets/Audio/Audio_Settings/Ancient_Weapon_Throw_1P_Att}
 }
 {
  id=Attenuation_1
  children [Modulator_1]
  properties {attenuationSettings=/AncientWeapon/Assets/Audio/Audio_Settings/Ancient_Weapon_Throw_Metal_Att}
 }
 {
  id=Attenuation_2
  children [Modulator_2]
  properties {attenuationSettings=/AncientWeapon/Assets/Audio/Audio_Settings/Ancient_Weapon_Throw_Att}
 }
 {id=Branch_0, children [Attenuation_0, Attenuation_2, Attenuation_2], properties {boolParameterName=IsLocal}}
 {id=Delay_0, children [Random_2], properties {delayMax=0.1, delayMin=0.1}}
 {id=Mixer_0, children [Branch_0, Attenuation_1], properties {inputVolume [1,3.5]}}
 {id=Modulator_0, children [QualityLevel_0], properties {volumeMin=0.8}}
 {id=Modulator_1, children [Delay_0]}
 {id=Modulator_2, children [QualityLevel_1]}
 {id=QualityLevel_0, children [Random_1,Random_1,Random_0,Random_0]}
 {id=QualityLevel_1, children [Random_3,Random_3,Random_4,Random_4]}
 {id=Random_0, children [WavePlayer_9, WavePlayer_10, WavePlayer_11], properties {weights [1,1,1]}}
 {
  id=Random_1
  children [WavePlayer_0, WavePlayer_7, WavePlayer_8, WavePlayer_9, WavePlayer_10, WavePlayer_11]
  properties {weights [1,1,1,1,1,1]}
 }
 {id=Random_2, children [WavePlayer_1,WavePlayer_2,WavePlayer_3,WavePlayer_4], properties {weights [1,1,1,1]}}
 {
  id=Random_3
  children [WavePlayer_5, WavePlayer_6, WavePlayer_12, WavePlayer_13, WavePlayer_14, WavePlayer_15]
  properties {weights [1,1,1,1,1,1]}
 }
 {id=Random_4, children [WavePlayer_13, WavePlayer_14, WavePlayer_15], properties {weights [1,1,1]}}
 {id=WavePlayer_0, properties {soundWaveAssetPtr=&/AncientWeapon_Throw_06}}
 {id=WavePlayer_1, properties {soundWaveAssetPtr=&/Ancient_Weapon_Throw_Metal_04}}
 {id=WavePlayer_10, properties {soundWaveAssetPtr=&/AncientWeapon_Throw_04}}
 {id=WavePlayer_11, properties {soundWaveAssetPtr=&/AncientWeapon_Throw_05}}
 {id=WavePlayer_12, properties {soundWaveAssetPtr=&/AncientWeapon_Throw_02}}
 {id=WavePlayer_13, properties {soundWaveAssetPtr=&/AncientWeapon_Throw_03}}
 {id=WavePlayer_14, properties {soundWaveAssetPtr=&AncientWeapon_Throw_04}}
 {id=WavePlayer_15, properties {soundWaveAssetPtr=&AncientWeapon_Throw_05}}
 {id=WavePlayer_2, properties {soundWaveAssetPtr=&/Ancient_Weapon_Throw_Metal_03}}
 {id=WavePlayer_3, properties {soundWaveAssetPtr=&/Ancient_Weapon_Throw_Metal_02}}
 {id=WavePlayer_4, properties {soundWaveAssetPtr=&/Ancient_Weapon_Throw_Metal_01}}
 {id=WavePlayer_5, properties {soundWaveAssetPtr=&AncientWeapon_Throw_06}}
 {id=WavePlayer_6, properties {soundWaveAssetPtr=&/AncientWeapon_Throw_01}}
 {id=WavePlayer_7, properties {soundWaveAssetPtr=&AncientWeapon_Throw_01}}
 {id=WavePlayer_8, properties {soundWaveAssetPtr=&AncientWeapon_Throw_02}}
 {id=WavePlayer_9, properties {soundWaveAssetPtr=&AncientWeapon_Throw_03}}
]
properties {maxDistance=15000, soundClassObject=&Fort_WeaponMelee, volumeMultiplier=0.4}
