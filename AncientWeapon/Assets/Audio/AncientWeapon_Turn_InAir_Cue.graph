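# Sound cue for the Ancient Weapon's turn-in-air sound, played in the melee weapon sound class with its own attenuation asset and boosted volume.
# A Modulator varies the output of a quality-level switch that feeds two equally weighted Random nodes.
# The first two quality slots pick from all eight TurnInAir waves.
# The last two share a smaller pool of variations 06 to 08, reusing the same wave players.
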SoundCue /AncientWeapon/Assets/Audio/AncientWeapon_Turn_InAir_Cue
root=Modulator_0
nodes [
 {id=Modulator_0, children [QualityLevel_0]}
 {id=QualityLevel_0, children [Random_1,Random_1,Random_0,Random_0]}
 {id=Random_0, children [WavePlayer_6, WavePlayer_12, WavePlayer_13], properties {weights [1,1,1]}}
 {
  id=Random_1
  children [WavePlayer_1, WavePlayer_2, WavePlayer_3, WavePlayer_4, WavePlayer_5, WavePlayer_6, WavePlayer_12, WavePlayer_13]
  properties {weights [1,1,1,1,1,1,1,1]}
 }
 {id=WavePlayer_1, properties {soundWaveAssetPtr=&/AncientWeapon_TurnInAir_01}}
 {id=WavePlayer_12, properties {soundWaveAssetPtr=&/AncientWeapon_TurnInAir_07}}
 {id=WavePlayer_13, properties {soundWaveAssetPtr=&/AncientWeapon_TurnInAir_08}}
 {id=WavePlayer_2, properties {soundWaveAssetPtr=&/AncientWeapon_TurnInAir_02}}
 {id=WavePlayer_3, properties {soundWaveAssetPtr=&/AncientWeapon_TurnInAir_03}}
 {id=WavePlayer_4, properties {soundWaveAssetPtr=&/AncientWeapon_TurnInAir_04}}
 {id=WavePlayer_5, properties {soundWaveAssetPtr=&/AncientWeapon_TurnInAir_05}}
 {id=WavePlayer_6, properties {soundWaveAssetPtr=&/AncientWeapon_TurnInAir_06}}
]
properties {
 attenuationSettings=/AncientWeapon/Assets/Audio/Audio_Settings/Ancient_Weapon_Turn_InAir_Att
 maxDistance=15000
 soundClassObject=&Fort_WeaponMelee
 volumeMultiplier=5
}
