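# Character material for the Ancient Weapon that lights its emissive and gem according to charge count.
# Charges and MaxNumCharges from custom primitive data threshold SkinFX mask channels, and the resulting glow, scaled by time of day, adds to the character FX emissive.
# Gem detail uses bump-offset parallax for cracked crystal normals and a rotating, pulsing inner glow.
# Quality switches simplify the crystal and emissive work on lower settings.
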Material /AncientWeapon/Assets/Materials/M_Ancient_Weapon
graph {
 properties {
  bAutomaticallySetUsageInEditor=false
  bUseMaterialAttributes=true
  bUsedWithClothing=true
  bUsedWithMeshDeformer=true
  bUsedWithMorphTargets=true
  bUsedWithNanite=true
  bUsedWithSkeletalMesh=true
  pixelDepthOffsetMode=PDOM_Legacy
  refractionMethod=RM_IndexOfRefraction
  shadingModel=MSM_FromMaterialExpression
  subsurfaceProfile=&SS_HeroSkin_02
  usedShadingModels=Default Lit | Subsurface | Clear Coat | Subsurface Profile | Cloth | Eye
 }
 nodes [
  n1 = MaterialFunctionCall() {materialFunction=&BaseFN_CharacterFunction_PreFX}
  n2 = MaterialFunctionCall(
   PreskinnedPosition=@n1.Pre skinned Local Position
   Gmap=@n1.Pre skinned Local Position
   MaterialAttributes=@n1.Pre skinned Local Position
   AO=@n1.Pre skinned Local Position
  ) {materialFunction=&CharacterSkinFX}
  n3 = MaterialFunctionCall(AO=@n1.AO, Distance Blend=@n1.AO, Fresnel=@n1.AO, Gmap=@n1.AO, Pre Skinned Local Position=@n1.AO, Material=@n2.Result) {materialFunction=&BaseFN_CharacterFunction_PostFX}
  n4 = GetMaterialAttributes(None=@n3.Result) {attributeGetTypes [#1, #2]}
  n5 = Power(Base=NamedRerouteUsage()) {constExponent=1.3}
  n6 = NamedRerouteUsage()
  n7 = Max(A=@n5, B=@n6)
  n8 = ScalarParameter() {bUseCustomPrimitiveData=true, parameterName=Charges}
  n9 = ScalarParameter() {bUseCustomPrimitiveData=true, defaultValue=3, parameterName=MaxNumCharges, primitiveDataIndex=1}
  n10 = Abs(None=Sign(None=Subtract(A=@n9))) {desc="handle max charges = 1"}
  n11 = Subtract(A=@n8, B=@n10)
  n12 = Subtract(A=@n9, B=@n10)
  n13 = Add(A=Subtract(A=@n5) {constB=0.001}, B=Saturate(None=Divide(A=@n11, B=@n12)))
  n14 = Add(A=Multiply(A=@n6, B=Saturate(None=@n8)), B=Saturate(None=Floor(None=@n13)))
  n15 = NamedRerouteUsage()
  n16 = ScalarParameter() {defaultValue=2, group=Emissive, parameterName=Charged Gem Boost}
  n17 = Multiply(
   A=Add(A=ScalarParameter() {defaultValue=1, group=Emissive, parameterName=Charged Intensity}, B=Multiply(A=@n15, B=@n16))
   B=NamedRerouteUsage()
  )
  n18 = NamedRerouteUsage()
  n19 = ScalarParameter() {defaultValue=0.8, group=Emissive, parameterName=Charged TODFactor Morning}
  n20 = Constant() {r=1}
  n21 = ScalarParameter() {defaultValue=0.8, group=Emissive, parameterName=Charged TODFactor Evening}
  n22 = ScalarParameter() {defaultValue=0.2, group=Emissive, parameterName=Charged TODFactor Night}
  n23 = Multiply(
   A=VectorParameter() {group=Emissive, parameterName=Emissive Color, defaultValue=<rgba 1 1 1 1>}.RGB
   B=Multiply(
    A=Multiply(A=@n17, B=@n18)
    B=MaterialFunctionCall(Value_Morning=@n19, Value_Day=@n20, Value_Evening=@n21, Value_Night=@n22) {materialFunction=&Scalar_Time_of_Day}.Result
   )
  )
  n24 = TextureSampleParameter2D() {
   group=Emissive
   parameterName=Emissive
   samplerSource=SSM_Wrap_WorldGroupSettings
   texture=/AncientWeapon/Assets/Textures/T_Gauntlet_Ancient_Weapon_E
  }
  n25 = ScalarParameter() {defaultValue=0.2, group=Emissive, parameterName=Gem Contour Contrast}
  n26 = Multiply(
   A=@n23
   B=Power(
    Base=MaterialFunctionCall(In=@n24.RGB, Contrast=@n25) {materialFunction=&CheapContrast}.Result
    Exp=ScalarParameter() {defaultValue=4, group=Emissive, parameterName=Gem Contour Power}
   ) {constExponent=4}
  )
  n27 = NamedRerouteUsage()
  n28 = Multiply(
   A=MaterialFunctionCall(In=Multiply(A=@n7, B=@n14), Contrast=Constant() {r=0.05}) {materialFunction=&CheapContrast}.Result
   B=QualitySwitch(Default=LinearInterpolate(A=@n23, B=@n26, Alpha=@n27), Low=@n23, Medium=@n23)
  )
  MaterialFunctionCall(In=SetMaterialAttributes!(MaterialAttributes=@n4.MaterialAttributes, Emissive Color=Add(A=@n4.Emissive Color, B=@n28) {constB=0}, Normal=NamedRerouteUsage()) {attributeSetTypes [#1, #2]}) {materialFunction=&CINE_MF_CharPostFX}
  n29 = Multiply(A=PreSkinnedPosition(), B=Constant3Vector() {constant=<rgba 1 1 0.8 0>})
  Multiply(
   A=SphereMask(
    A=VertexInterpolator(VS=@n29).PS
    B=Constant3Vector() {constant=<rgba 0 0 -0.37 0>}
    Radius=ScalarParameter() {defaultValue=2.8, parameterName=Gem_Glow_Radius}
    Hardness=ScalarParameter() {parameterName=Gem_Glow_Hardness}
   )
   B=ScalarParameter() {defaultValue=4, parameterName=Gem_Glow_Intensity}
  )
  NamedRerouteDeclaration(None=ComponentMask(None=@n2.SkinFX_Mask) {g=true}) {name=EmissiveRamp, nodeColor=<rgba 0.325 0 1 1>}
  NamedRerouteDeclaration(None=ComponentMask(None=@n2.SkinFX_Mask) {b=true}) {name=GemMask, nodeColor=<rgba 1 0 0.1556 1>}
  NamedRerouteDeclaration(None=@n4.Normal) {name=InNormal, nodeColor=<rgba 0 0.1388 1 1>}
  n30 = NamedRerouteUsage()
  n31 = TextureSample(UVs=Multiply(A=TextureCoordinate(), B=ScalarParameter() {defaultValue=6, group=Gem, parameterName=Gem_Crystal_CracksSize})) {
   samplerSource=SSM_Wrap_WorldGroupSettings
   samplerType=SAMPLERTYPE_Normal
   texture=/AncientWeapon/Assets/Textures/T_Sharp_Ancient_Crytal_Macro_N
  }
  n32 = Power(Base=@n31.B) {constExponent=4}
  n33 = TextureCoordinate()
  n34 = Add(
   A=Multiply(
    A=ConstantBiasScale(None=@n32) {bias=-0.5, scale=2}
    B=ScalarParameter() {defaultValue=0.05, group=Gem, parameterName=Gem_Crystal_CracksBelowOffset}
   )
   B=Multiply(A=@n33, B=ScalarParameter() {defaultValue=3.7, group=Gem, parameterName=Gem_Crystal_Tiling})
  )
  n35 = MaterialFunctionCall!(
   HeightTexture=Power(Base=@n31.B, Exp=ScalarParameter() {defaultValue=2, group=Gem, parameterName=Gem_Crystal_DepthPower})
   TextureCoordinates
   ReferencePlane=ScalarParameter() {defaultValue=1, group=Gem, parameterName=Gem_Crystal_Height}
   HeightRatioInput=ScalarParameter() {defaultValue=0.25, group=Gem, parameterName=Gem_Crystal_HeightRatio}
   CameraVector
  ) {materialFunction=/Engine/Functions/Engine_MaterialFunctions02/Utility/BumpOffset_Advanced.BumpOffset_advanced}
  n36 = ScalarParameter() {defaultValue=3.7, group=Gem, parameterName=Gem_Crystal_Tiling_LOW}
  n37 = TextureSample(UVs=QualitySwitch(Default=Add(A=@n34, B=@n35.Distortion), Low=Multiply(A=@n33, B=@n36))) {
   samplerSource=SSM_Wrap_WorldGroupSettings
   samplerType=SAMPLERTYPE_LinearColor
   texture=/AncientWeapon/Assets/Textures/T_Sharp_Ancient_Crytal_Micro_N
  }
  n38 = MaterialFunctionCall(
   BaseNormal=@n30
   AdditionalNormal=MaterialFunctionCall!(XY vector=ComponentMask(None=@n37.RGB) {g=true, r=true}, Z sign, Default Normalization vector) {materialFunction=/Engine/Functions/Engine_MaterialFunctions02/Math/DeriveNormalZ_Function}.Result
  ) {materialFunction=&BlendAngleCorrectedNormals}
  n39 = NamedRerouteUsage()
  NamedRerouteDeclaration(None=QualitySwitch(Default=LinearInterpolate(A=@n30, B=@n38.Result, Alpha=@n39), Low=@n30)) {name=OutNormal, nodeColor=<rgba 0 0.0292 1 1>}
  n40 = NamedRerouteUsage()
  n41 = Multiply(A=NamedRerouteUsage(), B=ScalarParameter() {defaultValue=0.8, group=Gem, parameterName=Gem_LinesInfluence})
  NamedRerouteDeclaration(None=LinearInterpolate(A=Constant() {r=1}, B=Multiply(A=@n37.R, B=@n37.R), Alpha=Max(A=@n40, B=@n41))) {name=CrystalEmissiveInfluence, nodeColor=<rgba 1 0.75 0 1>}
  n42 = Constant2Vector() {g=-1, r=1}
  n43 = ComponentMask(None=NamedRerouteUsage()) {g=true, r=true}
  n44 = VectorParameter() {group=Gem, parameterName=GemGlow_Params, channelNames {a=(None), b=Size, g=PositionY, r=PositionX}, defaultValue=<rgba 0.5 0.5 12 0>}
  n45 = ComponentMask(None=@n44) {g=true, r=true}
  n46 = MaterialFunctionCall(UVs=Add(A=TextureCoordinate(), B=@n45), Texture Scale=@n44.Size) {materialFunction=&ScaleUVsByCenter}
  n47 = StaticSwitchParameter(True=Add(A=Divide(A=@n43, B=@n44.Size) {constB=7}, B=@n45), False=@n46.UVs) {defaultValue=true, group=Gem, parameterName=UseGemLocalCoords}
  n48 = QualitySwitch(Default=Add(A=Multiply(A=@n35.Distortion, B=@n42), B=@n47), Low=@n47)
  n49 = Multiply(A=Time(), B=ScalarParameter() {defaultValue=1, group=Gem, parameterName=Gem_GlowInternalDetail_Speed})
  n50 = ConstantBiasScale(None=Sine(None=Time()))
  n51 = ScalarParameter() {defaultValue=1, group=Gem, parameterName=Gem_GlowInternalDetail_Pulse}
  n52 = Power(
   Base=TextureSample(UVs=Rotator(Coordinate=@n48, Time=@n49) {speed=-1}) {samplerSource=SSM_Clamp_WorldGroupSettings, texture=/AncientWeapon/Assets/Textures/T_Sharp_Ancient_Glow}.RGB
   Exp=Add(A=Multiply(A=@n50, B=@n51) {constB=0.3}, B=ScalarParameter() {defaultValue=3, group=Gem, parameterName=Gem_GlowInternalDetail_Power}) {constB=2}
  )
  n53 = ScalarParameter() {defaultValue=8, group=Gem, parameterName=Gem_GlowInternalDetail_Intensity}
  n54 = NamedRerouteUsage()
  NamedRerouteDeclaration(None=LinearInterpolate(A=Constant() {r=1}, B=Multiply(A=@n52, B=@n53) {constB=4}, Alpha=Multiply(A=@n54, B=@n46.0-1 mask))) {name=GemGlow, nodeColor=<rgba 1 0 0.8203 1>}
  NamedRerouteDeclaration(None=@n1.Pre skinned Local Position) {name=PreSkinPos, nodeColor=<rgba 1 0 0 1>}
 ]
}
