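# Translucent Default Lit Niagara mesh-particle material for the AncientWeapon spinning motion blur, built on biplane propeller blur textures.
# Two rotating blur textures, driven by dynamic SpinSpeed, ease from normal to fast over a timestamped boost, then get hue shift, desaturation and panned color streaks.
# Sun/moon reflection dot products fake front and optional reverse highlights, tinted by time of day.
# Opacity combines blur, radial mask, view-angle fade, particle alpha and non-mobile depth fade.
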Material /AncientWeapon/Assets/VFX/M_AncientWeapon_Blur
graph {
 properties {bUsedWithNiagaraMeshParticles=true, blendMode=BLEND_Translucent, pixelDepthOffsetMode=PDOM_Legacy, usedShadingModels=Default Lit}
 nodes [
  n1 = Multiply(A=Add(A=MaterialFunctionCall() {materialFunction=&Color_Time_of_Day}.Result) {constB=0.25}) {constB=0.8}
  n2 = CollectionParameter() {
   collection=&FortniteMaterialParameters
   parameterName=SunAndMoonModelDirectionalVector
  }
  n3 = DotProduct(A=ComponentMask(None=@n2) {b=true, g=true, r=true}, B=ReflectionVectorWS() {bNormalizeCustomWorldNormal=true})
  n4 = ScalarParameter() {defaultValue=25, parameterName=Spec Highlight Power}
  n5 = Multiply(A=@n1, B=Clamp(None=Power(Base=@n3, Exp=@n4) {constExponent=20}))
  n6 = Multiply(A=@n2) {constB=-1}
  n7 = Power(
   Base=DotProduct(A=ComponentMask(None=@n6) {b=true, g=true, r=true}, B=ReflectionVectorWS() {bNormalizeCustomWorldNormal=true})
   Exp=ScalarParameter() {defaultValue=15, parameterName=Reverse Spec Highlight Power}
  ) {constExponent=20}
  n8 = Multiply(A=Multiply(A=@n1, B=Clamp(None=@n7)), B=ScalarParameter() {defaultValue=1, parameterName=ReverseHighlightOpacity})
  n9 = TextureCoordinate() {vTiling=8}
  n10 = Multiply(A=Time(), B=DynamicParameter() {defaultValue=<rgba 3.5 1 1 1>, paramNames [SpinSpeed,Param2,Param3,Param4]}.SpinSpeed)
  n11 = Power(Base=SphereMask(A=TextureCoordinate() {coordinateIndex=1}, B=Constant() {r=0.5}) {attenuationRadius=0.5, hardnessPercent=0}) {constExponent=0.8}
  n12 = Multiply(
   A=TextureSample(UVs=Panner(Coordinate=@n9, Time=@n10) {speedX=-7}) {texture=&T_Smooth_tile}.G
   B=OneMinus(None=@n11)
  )
  n13 = Rotator(Coordinate=TextureCoordinate() {coordinateIndex=1}, Time=@n10) {speed=-25}
  n14 = TextureSampleParameter2D(UVs=@n13) {parameterName=Texture01, texture=/Game/Athena/DrivableVehicles/FX/Plane/T_Biplane_Propeller_Blur}
  n15 = Multiply(A=TextureSampleParameter2D(UVs=@n13) {parameterName=Texture02, texture=/Game/Athena/DrivableVehicles/FX/Plane/T_Biplane_Propeller_Blur_Fast}.RGB) {constB=1.5}
  n16 = Subtract(A=Time() {period=5}, B=ScalarParameter() {parameterName=BoostStartTime})
  n17 = ScalarParameter() {defaultValue=3, group=Timestamp, parameterName=BoostDuration}
  n18 = MaterialFunctionCall!(In=Clamp(None=Divide(A=@n16, B=@n17))) {materialFunction=/Game/Effects/Fort_Effects/Materials/_Functions/MF_Smoothstep_EaseOut}
  n19 = Desaturation(
   None=MaterialFunctionCall(
    Hue Shift Percentage=ScalarParameter() {parameterName=HueShift}
    Texture=LinearInterpolate(A=@n14.RGB, B=@n15, Alpha=@n18.Result) {constAlpha=1}
   ) {materialFunction=&HueShift}.Result
   Fraction=ScalarParameter() {parameterName=Desaturate}
  )
  n20 = Add(A=Multiply(A=VectorParameter() {parameterName=ColorAddStreaks, defaultValue=<rgba 0.0647 0.0076 0.5 1>}.RGB, B=@n12), B=@n19)
  n21 = Multiply(A=StaticSwitchParameter(True=Add(A=@n5, B=@n8), False=@n5) {defaultValue=true, parameterName=AddReversedFakeLightVector}, B=@n20)
  n22 = Multiply(A=@n19) {constB=0.1}
  n23 = ParticleColor()
  MaterialFunctionCall(Emissive=Multiply(A=Add(A=@n21, B=@n22), B=@n23.RGB)) {materialFunction=&MF_FX_EmissiveAdjustments}
  n24 = Add(
   A=Multiply(A=ComponentMask(None=@n19) {g=true}, B=LinearInterpolate(Alpha=@n12) {constA=0.7, constB=1.5}) {constB=1.5}
   B=LinearInterpolate(Alpha=@n11) {constA=0.7, constB=0}
  )
  n25 = MaterialFunctionCall("FadeSensitivity (optional)"=ScalarParameter() {defaultValue=1.5, parameterName=AngleFade}) {materialFunction=&FadeBasedOnViewAngle}
  n26 = ScalarParameter() {defaultValue=25, parameterName=DepthFadeDistance}
  Multiply(
   A=Multiply(A=Multiply(A=@n24, B=@n25.Result), B=@n23.A)
   B=ShadingPathSwitch(Default=DepthFade(FadeDistance=@n26) {fadeDistanceDefault=25}, Mobile=Constant() {r=1}) {desc=Migrated from FeatureLevelSwitch}
  )
  Multiply(A=@n20, B=@n23.RGB)
  ScalarParameter() {defaultValue=3.5, parameterName=SpinSpeed}
 ]
}
