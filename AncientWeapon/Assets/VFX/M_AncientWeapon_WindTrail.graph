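# Translucent unlit, two-sided material for the Ancient Weapon wind trail VFX on Niagara ribbons and mesh particles.
# A noise-distorted, panning trail texture is tinted by a curve atlas and particle color for emissive.
# Opacity combines masks, swirl-noise dissolve, near-camera, depth and optional view-angle fades, with a simpler forward/mobile path.
# Niagara dynamic parameters drive randomization and world position offset: sine waves along normals, plus an optional clamped trailing bend.
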Material /AncientWeapon/Assets/VFX/M_AncientWeapon_WindTrail
graph {
 properties {
  bUseTranslucencyVertexFog=false
  bUsedWithNiagaraMeshParticles=true
  bUsedWithNiagaraRibbons=true
  blendMode=BLEND_Translucent
  pixelDepthOffsetMode=PDOM_Legacy
  shadingModel=MSM_Unlit
  twoSided=true
  usedShadingModels=Unlit
 }
 nodes [
  n1 = Constant3Vector()
  n2 = DynamicParameter() {defaultValue=<rgba 0 0 1 0>, paramNames [URandom, TimeOffsetRandom, WPOStrength, TrailingModeON]}
  n3 = Add(A=Time(), B=@n2.TimeOffsetRandom)
  n4 = VectorParameter() {group=WPO, parameterName=WPO Sine, channelNames {a="", b=Sine Strength, g=Sine Tile, r=Sine Speed}, defaultValue=<rgba 2 2 1 1>}
  n5 = DynamicParameter() {parameterIndex=2, defaultValue=<rgba 1 250 0 0>, paramNames [NoiseRandom, Max, Sine Strength, TrailSpeed]}
  n6 = Add(A=@n4.Sine Speed, B=@n5.TrailSpeed)
  n7 = ComponentMask(None=TextureCoordinate()) {r=true}
  n8 = Add(A=@n2.URandom, B=@n7)
  n9 = Sine(None=Add(A=Multiply(A=@n3, B=@n6), B=Multiply(A=@n8) {constB=2}))
  n10 = Multiply(
   A=MaterialFunctionCall(X=@n9, Y=@n9, Z=@n9) {materialFunction=&MakeFloat3}.Result
   B=Add(A=@n4.Sine Strength, B=@n5.Sine Strength)
  ) {constB=0.2}
  n11 = OneMinus(None=ComponentMask(None=TextureCoordinate()) {r=true})
  n12 = Multiply(A=@n11, B=@n11)
  n13 = VectorParameter() {group=WPO, parameterName=WPO Control, channelNames {a=WPOStrength, b=BendPower, g=Max, r=Min}, defaultValue=<rgba -50 50 3 1>}
  n14 = DynamicParameter() {defaultValue=<rgba 0 0 1 0>, paramNames [URandom, TimeOffsetRandom, WPOStrength, TrailingModeON]}
  n15 = Multiply(A=Multiply(A=VertexNormalWS(), B=LinearInterpolate(A=@n1, B=@n10, Alpha=@n12)), B=Multiply(A=@n13.WPOStrength, B=@n14.WPOStrength)) {constB=10}
  n16 = DynamicParameter() {desc=Parameters set by Niagara, parameterIndex=1, paramNames [X,Y,Z,WPOIntensity]}
  n17 = DynamicParameter() {parameterIndex=2, defaultValue=<rgba 1 250 0 0>, paramNames [NoiseRandom, Max, Sine Strength, TrailSpeed]}
  n18 = Clamp(
   None=Subtract(
    A=MaterialFunctionCall(X=@n16.X, Y=@n16.X, Z=@n16.X) {materialFunction=&MakeFloat3}.Result
    B=ObjectPositionWS()
   )
   Min=Multiply(A=@n17.Max) {constA=-1, constB=-1}
   Max=@n17.Max
  ) {maxDefault=50, minDefault=-50}
  n19 = ComponentMask(None=TextureCoordinate()) {r=true}
  n20 = Multiply(A=Power(Base=OneMinus(None=@n19), Exp=@n13.BendPower), B=@n16.WPOIntensity)
  LinearInterpolate(A=@n15, B=Add(A=LinearInterpolate(B=@n18, Alpha=@n20), B=@n15), Alpha=@n14.TrailingModeON)
  n21 = CurveAtlasRowParameter(CurveTime=ComponentMask(None=TextureCoordinate()) {r=true}) {
   atlas=/Game/Athena/LinearColorCurves/VFX_CurveAtlas
   curve=/Game/Athena/LinearColorCurves/Curve_CubeNinja_Contrail_Color
   defaultValue=9
   group=Trail
   parameterName=Curve
  }
  n22 = TextureCoordinate()
  n23 = VectorParameter() {group=Noise, parameterName=NoiseA UV, defaultValue=<rgba 1 1 0 0>}
  n24 = MaterialFunctionCall(X=@n23.R, Y=@n23.R) {materialFunction=&MakeFloat2}
  n25 = MaterialFunctionCall(X=@n23.B, Y=@n23.B) {materialFunction=&MakeFloat2}
  n26 = Add(
   A=DynamicParameter() {defaultValue=<rgba 0 0 1 0>, paramNames [URandom, TimeOffsetRandom, WPOStrength, TrailingModeON]}.TimeOffsetRandom
   B=Time()
  )
  n27 = TextureSample(UVs=Add(A=Multiply(A=@n22, B=@n24.Result), B=Multiply(A=@n25.Result, B=@n26))) {samplerType=SAMPLERTYPE_Grayscale, texture=&LowResBlurredNoise}
  n28 = DynamicParameter() {parameterIndex=2, defaultValue=<rgba 1 250 0 0>, paramNames [NoiseRandom, Max, Sine Strength, TrailSpeed]}
  n29 = Multiply(A=ScalarParameter() {group=Noise, parameterName=Noise Strength}, B=@n28.NoiseRandom)
  n30 = TextureCoordinate()
  n31 = VectorParameter() {group=Trail, parameterName=Trail UV, defaultValue=<rgba 1 1 0 0>}
  n32 = DynamicParameter() {defaultValue=<rgba 0 0 1 0>, paramNames [URandom, TimeOffsetRandom, WPOStrength, TrailingModeON]}
  n33 = MaterialFunctionCall(X=Add(A=@n31.R, B=@n32.URandom), Y=@n31.G) {materialFunction=&MakeFloat2}
  n34 = DynamicParameter() {parameterIndex=2, defaultValue=<rgba 1 250 0 0>, paramNames [NoiseRandom, Max, Sine Strength, TrailSpeed]}
  n35 = MaterialFunctionCall(X=Add(A=Add(A=@n31.B, B=@n32.URandom), B=@n34.TrailSpeed), Y=@n31.A) {materialFunction=&MakeFloat2}
  n36 = Multiply(A=@n35.Result, B=Add(A=@n32.TimeOffsetRandom, B=Time()))
  n37 = Constant2Vector() {g=1}
  n38 = Add(
   A=Multiply(A=Multiply(A=@n27.R, B=@n29), B=Constant2Vector() {r=1})
   B=Add(A=Multiply(A=@n30, B=@n33.Result), B=LinearInterpolate(A=@n36, B=@n35.Result, Alpha=@n37))
  )
  n39 = MaterialFunctionCall(Float2=Multiply(
    A=MaterialFunctionCall!(TexCoord=@n38, ClampChannel=@n37) {materialFunction=/Game/Effects/Fort_Effects/MaterialFunctions/MF_FX_TilingMethod}.Result
    B=TextureSampleParameter2D(UVs=@n38) {parameterName=Texture, samplerType=SAMPLERTYPE_LinearColor, texture=&/T_AncientWeapons_D_WindTrail}.RGB
   )) {materialFunction=&BreakOutFloat2Components}
  n40 = LinearInterpolate(A=@n39.G, B=@n39.G, Alpha=@n7)
  n41 = VectorParameter() {group=Trail, parameterName=Trail Refine, defaultValue=<rgba 1 1 1 0.5>}
  n42 = VectorParameter() {
   group=Trail
   parameterName=Trail Control
   channelNames {a=Near Distance, b=Depth Fade, g=Trail Add, r=Trail Mask}
   defaultValue=<rgba 1 0.5 100 256>
  }
  n43 = Add(
   A=MaterialFunctionCall(Texture=@n40, RGB=@n41.RGB, A=@n41.RGB) {materialFunction=&MF_FX_Refine}.Result
   B=@n42.Trail Add
  )
  n44 = ParticleColor()
  MaterialFunctionCall(Emissive=Multiply(A=Multiply(A=@n21, B=@n43), B=@n44.RGB)) {materialFunction=&MF_FX_EmissiveAdjustments}
  n45 = DynamicParameter() {parameterIndex=3, defaultValue=<rgba 1 1 0 1>, paramNames [Dissolve, Dissolve Tex Pow, 3, 4]}
  n46 = TextureSample(UVs=Add(A=TextureCoordinate() {uTiling=2, vTiling=2}, B=Constant2Vector())) {samplerType=SAMPLERTYPE_LinearGrayscale, texture=/Game/Global/Textures/Default/Noise/T_SwirlNoise_N_height}
  n47 = MaterialFunctionCall(Distance=@n42.Near Distance) {materialFunction=&Near_Camera_Fade}
  n48 = Multiply(A=@n40, B=@n42.Trail Mask)
  n49 = MaterialFunctionCall(Float2=TextureCoordinate()) {materialFunction=&BreakOutFloat2Components}
  n50 = Multiply(A=@n49.R, B=ScalarParameter() {defaultValue=1, parameterName=MaskU})
  n51 = Multiply(A=Time(), B=ScalarParameter() {defaultValue=1, parameterName=MaskSpeed})
  n52 = ScalarParameter() {defaultValue=1, parameterName=MaskV}
  n53 = LinearInterpolate(
   A=TextureCoordinate() {uTiling=-1}
   B=AppendVector(A=Add(A=@n50, B=@n51), B=Multiply(A=@n49.G, B=@n52))
   Alpha=ScalarParameter() {parameterName=MaskUV}
  )
  n54 = ComponentMask(None=TextureCoordinate()) {r=true}
  n55 = Power(Base=@n54, Exp=ScalarParameter() {defaultValue=2, parameterName=MaskPower}) {constExponent=3.4}
  n56 = OneMinus(None=@n54)
  n57 = Multiply(A=Multiply(A=@n55, B=Power(Base=@n56) {constExponent=0.9})) {constB=3.4}
  n58 = ScalarParameter() {parameterName=AddMask}
  n59 = Saturate(None=Multiply(
    A=TextureSampleParameter2D(UVs=@n53) {parameterName=Mask, texture=/Game/Effects/Fort_Effects/Textures/Smoke/T_SmokeTrailCoord1_Mask}.R
    B=LinearInterpolate(A=@n55, B=@n57, Alpha=@n58)
   ))
  n60 = TextureSampleParameter2D() {parameterName=Mask2, samplerType=SAMPLERTYPE_Masks, texture=/Game/Global/Textures/Default/Masks_and_Gradients/T_VertGradient_MASK_2}
  n61 = Multiply(A=@n60.R, B=@n60.R)
  n62 = Multiply(A=@n59, B=Multiply(A=@n61, B=@n61))
  n63 = Saturate(None=Add(A=Multiply(A=@n48, B=@n59), B=Multiply(A=@n62) {constB=0.5}))
  n64 = Multiply(A=@n44.A, B=Multiply(A=@n47.Result, B=DepthFade(Opacity=@n63, FadeDistance=@n42.Depth Fade)))
  n65 = Multiply(A=@n47.Result, B=@n63)
  n66 = ShadingPathSwitch(Default=@n64, Deferred=@n64, Forward=@n65, Mobile=@n65)
  n67 = MaterialFunctionCall("FadeSensitivity (optional)"=Constant() {r=1.4}) {materialFunction=&FadeBasedOnViewAngle}
  Multiply(
   A=Multiply(A=@n45.Dissolve, B=Power(Base=@n46.RGB, Exp=@n45.Dissolve Tex Pow))
   B=LinearInterpolate(A=@n66, B=Multiply(A=@n66, B=@n67.Result), Alpha=ScalarParameter() {parameterName=MeshFede})
  )
 ]
}
