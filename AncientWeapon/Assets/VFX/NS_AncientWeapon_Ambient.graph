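# Niagara system for the Ancient Weapon's ambient blue-violet energy: blade and edge glows, electric arcs and a glowing gem.
# Mesh emitters attach to a user skeletal mesh, while two randomly looping ribbon emitters run bone-flow arcs along sockets on two user meshes.
# A system Timeline drives the charge glow and gem flare, and ChargeAmount feeds the emissive line glows.
# BladeGlow uses distance-based scalability to raise opacity and material intensity with range before culling.
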NiagaraSystem /AncientWeapon/Assets/VFX/NS_AncientWeapon_Ambient
dependencies {
 usedDataInterfaces [&NiagaraDataInterfaceCurve, &NiagaraDataInterfaceArrayFloat3]
 usedDynamicInputs [&MakeVector2D, &NormalizeFloat]
 usedModules [
  &SystemState
  &self:SystemSpawnScript.SystemScriptSource.SystemScriptGraph.NiagaraNodeAssignment_0.SetVariables_#1
  &Timeline
 ]
}
emitters {
 BladeEdgeEnergy {
  emitterUpdateScript [
   &EmitterState(Life Cycle Mode=::NewEnumerator1, Loop Duration=1, UseLoopDelay=false)
   &SpawnRate(SpawnRate=2, Use Spawn Probability=false)
  ]
  particleSpawnScript [
   /SetVariables(Particles.RandBool:NiagaraBool=&RandomBool()) {isSetParameters=true, module=SetVariables_#2}
   &InitializeParticle(
    Color Maximum=<rgba 0.4834 0 1 1>
    Color Minimum=<rgba 0.0365 0 1 1>
    Color Mode=::NewEnumerator2
    Lifetime Max=1.25
    Lifetime Min=0.75
    Lifetime Mode=::NewEnumerator1
    Mesh Scale Mode=::NewEnumerator3
    Mesh Uniform Scale=1.01
    UsePositionOffset=false
    Write Material Random=true
   )
   &SkeletalMeshLocation(
    Bone Sampling Mode=::NewEnumerator4
    Interpolate Mesh Sampling=false
    MaskSpawnGroup=false
    Orientation Sampling=::NewEnumerator1
    Skeletal Mesh=$User.SkeletalMesh
   )
   &InitialMeshOrientation(Mesh Axis To Orient=<0 0 1>, Mesh Orientation Mode=::NewEnumerator2, Use Rotation Vector=false)
  ]
  particleUpdateScript [
   &ParticleState()
   &DynamicMaterialParameters(
    Index 0 Param 1=&FloatFromCurve(
     FloatCurve=DataInterface() {propertyValues {Curve {keys [{interpMode=RCIM_Cubic, tangentMode=RCTM_User, time=0, value=1}, {interpMode=RCIM_Cubic, tangentMode=RCTM_User, time=1, value=0}]}}}
     Scale Curve=&Multiply_Float(A=1, B=&MakeCustomFloatFromBool(Bool=$Particles.RandBool, True Float=1, False Float=-1))
    )
    Module.Param0WriteEnabled=true
    Module.Param1WriteEnabled=true
    Module.Param2WriteEnabled=true
    Module.Param3WriteEnabled=true
   )
   &ScaleColor(
    Scale RGB=&VectorFromFloat(Value=&Lerp_Float(
      A=1
      B=10
      Alpha=&FloatFromCurve(
       FloatCurve=DataInterface() {propertyValues {Curve {keys [
           {interpMode=RCIM_Cubic, tangentMode=RCTM_User, time=0, value=0}
           {arriveTangent=1.3375, interpMode=RCIM_Cubic, leaveTangent=1.3375, tangentMode=RCTM_User, time=0.8, value=0.25}
           {interpMode=RCIM_Cubic, tangentMode=RCTM_User, time=1, value=1}
          ]}}}
       CurveIndex=&RandomRangeFloat(Minimum=0, Maximum=1)
      )
     ))
    ScaleA=true
    ScaleRGB=true
   )
  ]
  rendererClasses [&NiagaraMeshRendererProperties]
  renderers [{rendererClass=&NiagaraMeshRendererProperties, rendererIndex=0}]
 }
 BladeEdgeGlow {
  emitterUpdateScript [
   &EmitterState(Life Cycle Mode=::NewEnumerator1, Loop Behavior=::NewEnumerator1, Loop Duration Mode=::NewEnumerator1, UseLoopDelay=false)
   &SpawnBurst_Instantaneous(Use Spawn Probability=false, UseLoopCountLimit=false)
  ]
  particleSpawnScript [
   &InitializeParticle(
    Color=<rgba 0.0822 0.0076 1 1>
    Mesh Scale Mode=::NewEnumerator3
    Mesh Uniform Scale=1.01
    UsePositionOffset=false
    Write Material Random=true
   )
   &SkeletalMeshLocation(
    Bone Sampling Mode=::NewEnumerator4
    Interpolate Mesh Sampling=false
    MaskSpawnGroup=false
    Orientation Sampling=::NewEnumerator1
    Skeletal Mesh=$User.SkeletalMesh
   )
   &InitialMeshOrientation(Mesh Axis To Orient=<0 0 1>, Mesh Orientation Mode=::NewEnumerator2, Use Rotation Vector=false)
  ]
  particleUpdateScript [
   &ParticleState(Kill Particles When Lifetime Has Elapsed=false, Let Infinitely Lived Particles Die When Emitter Deactivates=true)
   &ScaleColor(Scale RGB=&VectorFromFloat(Value=4), ScaleA=true, ScaleRGB=true)
  ]
  rendererClasses [&NiagaraMeshRendererProperties]
  renderers [{rendererClass=&NiagaraMeshRendererProperties, rendererIndex=0}]
 }
 BladeGlow {
  emitterUpdateScript [
   &EmitterState(
    Enable Distance Culling=true
    Life Cycle Mode=::NewEnumerator1
    Loop Behavior=::NewEnumerator1
    Loop Duration Mode=::NewEnumerator1
    MaxDistance=8000
    Module.bUseMaxDistance=true
    Module.bUseMinDistance=false
    Scalability Mode=::NewEnumerator1
    UseLoopDelay=false
   )
   &SpawnBurst_Instantaneous(Use Spawn Probability=false, UseLoopCountLimit=false)
  ]
  particleSpawnScript [
   /SetVariables(Particles.RandBool:NiagaraBool=&RandomBool()) {isSetParameters=true, module=SetVariables_#2}
   &InitializeParticle(Color=<rgba 0.3192 0 1 1>, Mesh Scale Mode=::NewEnumerator3, Mesh Uniform Scale=1.01, UsePositionOffset=false, Write Material Random=true)
   &SkeletalMeshLocation(
    Bone Sampling Mode=::NewEnumerator4
    Interpolate Mesh Sampling=false
    MaskSpawnGroup=false
    Orientation Sampling=::NewEnumerator1
    Skeletal Mesh=$User.SkeletalMesh
   )
   &InitialMeshOrientation(Mesh Axis To Orient=<0 0 1>, Mesh Orientation Mode=::NewEnumerator2, Use Rotation Vector=false)
  ]
  particleUpdateScript [
   &ParticleState(Kill Particles When Lifetime Has Elapsed=false, Let Infinitely Lived Particles Die When Emitter Deactivates=true)
   &DynamicMaterialParameters(
    Index 0 Param 1=&Lerp_Float(
     A=10
     B=35
     Alpha=/Niagara/DynamicInputs/Scalability/ScalabilityDistanceBasedFloat(
      Curve:NiagaraDataInterfaceCurve=DataInterface() {propertyValues {Curve {keys [{interpMode=RCIM_Cubic, tangentMode=RCTM_User, time=0, value=0}, {interpMode=RCIM_Cubic, tangentMode=RCTM_User, time=0.6, value=1}, <at 1 1>]}}}
      CurveIndex:NiagaraFloat=$Emitter.Scalability.DistanceFraction
      Bias:NiagaraFloat=0
      Scale:NiagaraFloat=1
     )
    )
    Module.Param0WriteEnabled=true
    Module.Param1WriteEnabled=true
    Module.Param2WriteEnabled=true
    Module.Param3WriteEnabled=true
   )
   &ScaleColor(
    Scale Alpha=&Multiply_Float(
     A=&Lerp_Float(
      A=0.02
      B=1
      Alpha=&ScalabilityDistanceBasedFloat(
       Curve:NiagaraDataInterfaceCurve=DataInterface() {propertyValues {Curve {keys [{interpMode=RCIM_Cubic, tangentMode=RCTM_User, time=0, value=0}, {interpMode=RCIM_Cubic, tangentMode=RCTM_User, time=0.6, value=1}, <at 1 1>]}}}
       CurveIndex:NiagaraFloat=$Emitter.Scalability.DistanceFraction
       Bias:NiagaraFloat=0
       Scale:NiagaraFloat=1
      )
     )
     B=1
    )
    Scale RGB=&VectorFromFloat(Value=5)
    ScaleA=true
    ScaleRGB=true
   )
  ]
  rendererClasses [&NiagaraMeshRendererProperties]
  renderers [{rendererClass=&NiagaraMeshRendererProperties, rendererIndex=0}]
 }
 BladeGlowCharge {
  emitterUpdateScript [
   &EmitterState(Life Cycle Mode=::NewEnumerator1, Loop Behavior=::NewEnumerator1, Loop Duration Mode=::NewEnumerator1, UseLoopDelay=false)
   &SpawnBurst_Instantaneous(Use Spawn Probability=false, UseLoopCountLimit=false)
  ]
  particleSpawnScript [
   /SetVariables(Particles.RandBool:NiagaraBool=&RandomBool()) {isSetParameters=true, module=SetVariables_#2}
   &InitializeParticle(
    Color=<rgba 0.1594 0.0328 1 1>
    Mesh Scale Mode=::NewEnumerator3
    Mesh Uniform Scale=1.01
    UsePositionOffset=false
    Write Material Random=true
   )
   &SkeletalMeshLocation(
    Bone Sampling Mode=::NewEnumerator4
    Interpolate Mesh Sampling=false
    MaskSpawnGroup=false
    Orientation Sampling=::NewEnumerator1
    Skeletal Mesh=$User.SkeletalMesh
   )
   &InitialMeshOrientation(Mesh Axis To Orient=<0 0 1>, Mesh Orientation Mode=::NewEnumerator2, Use Rotation Vector=false)
  ]
  particleUpdateScript [
   &ParticleState(Kill Particles When Lifetime Has Elapsed=false, Let Infinitely Lived Particles Die When Emitter Deactivates=true)
   &DynamicMaterialParameters(
    Index 0 Param 1=&Lerp_Float(
     A=1
     B=25
     Alpha=&FloatFromCurve(
      FloatCurve=DataInterface() {propertyValues {Curve {keys [
          {arriveTangent=3.0052, interpMode=RCIM_Cubic, leaveTangent=3.0052, tangentMode=RCTM_User, time=0, value=0}
          {interpMode=RCIM_Cubic, tangentMode=RCTM_User, time=0.7, value=1}
          {arriveTangent=-6.0778, interpMode=RCIM_Cubic, leaveTangent=-6.0778, tangentMode=RCTM_User, time=1, value=0}
         ]}}}
      CurveIndex=$System.Timeline.NormalizedTime
     )
    )
    Module.Param0WriteEnabled=true
    Module.Param1WriteEnabled=true
    Module.Param2WriteEnabled=true
    Module.Param3WriteEnabled=true
   )
   &ScaleColor(
    Scale Alpha=&FloatFromCurve(
     FloatCurve=DataInterface() {propertyValues {Curve {keys [
         {arriveTangent=3.0052, interpMode=RCIM_Cubic, leaveTangent=3.0052, tangentMode=RCTM_User, time=0, value=0}
         {interpMode=RCIM_Cubic, tangentMode=RCTM_User, time=0.7, value=1}
         {arriveTangent=-6.0778, interpMode=RCIM_Cubic, leaveTangent=-6.0778, tangentMode=RCTM_User, time=1, value=0}
        ]}}}
     CurveIndex=$System.Timeline.NormalizedTime
     Scale Curve=0.3
    )
    Scale RGB=&VectorFromFloat(Value=5)
    ScaleA=true
    ScaleRGB=true
   )
   /SetVariables(
    Particles.ColorEnergy:LinearColor=&MultiplyLinearColor(
     Linear Color=$Particles.Initial.Color
     Scale Factor=&MakeFloat4FromVectorAndFloat(
      "Vector (RGB)"=&VectorFromFloat(Value=10)
      "Float (Alpha)"=&FloatFromCurve(
       FloatCurve=DataInterface() {propertyValues {Curve {keys [
           {interpMode=RCIM_Cubic, tangentMode=RCTM_User, time=0, value=1}
           {arriveTangent=-2.7054, interpMode=RCIM_Cubic, leaveTangent=-2.7054, tangentMode=RCTM_User, time=1, value=0}
          ]}}}
       CurveIndex=$System.Timeline.NormalizedTime
      )
     )
    )
    Particles.DynMatEnergy:Vector4f=&MakeVector4(
     X=&FloatFromCurve(
      FloatCurve=DataInterface() {propertyValues {Curve {keys [
          {interpMode=RCIM_Cubic, leaveTangentWeight=0.9171, tangentMode=RCTM_User, time=0, value=2}
          {arriveTangentWeight=2.032, interpMode=RCIM_Cubic, tangentMode=RCTM_User, time=1, value=0.5}
         ]}}}
      CurveIndex=$System.Timeline.NormalizedTime
     )
     Y=0
     Z=0
     W=0
    )
   ) {isSetParameters=true}
   /SetVariables(Particles.VisibilityTag=&MakeCustomIntFromBool(Bool=$System.Timeline.IsPlaying, True INT=0, False INT=1)) {isSetParameters=true}
  ]
  rendererClasses [&NiagaraMeshRendererProperties]
  renderers [
   {rendererClass=&NiagaraMeshRendererProperties, rendererIndex=0}
   {rendererClass=&NiagaraMeshRendererProperties, rendererIndex=1}
  ]
 }
 ElectricityBoneFlow {
  emitterSpawnScript [
   /SetVariables(Emitter.RandomSeed:NiagaraInt32=/Niagara/DynamicInputs/Random/RandomInt(Randomness Mode:ENiagaraRandomnessMode=::NewEnumerator0, Random Seed:NiagaraInt32=0, Fixed Random Seed:NiagaraBool=false)) {isSetParameters=true, module=SetVariables_#3}
   /Game/Effects/Niagara/Modules/Partition/FN_PartitionAttributes(Number of Bone Sets=2, Particles Per Partition=2, Partitions=2)
  ]
  emitterUpdateScript [
   &EmitterState(
    Life Cycle Mode=::NewEnumerator1
    Loop Delay=&RandomRangeFloat(Minimum=0.1, Maximum=0.5)
    Loop Duration=0.5
    Recalculate Duration Each Loop=true
    UseLoopDelay=true
   )
   &SpawnRate(SpawnRate=&Multiply_Float(A=35, B=1), Use Spawn Probability=false)
   /SetVariables(Emitter.RandXSpeed:NiagaraFloat=&Multiply_Float(
     A=&RandomRangeFloat(Minimum=1.1, Maximum=2)
     B=&MakeCustomFloatFromBool(Bool=&RandomBool(), True Float=1, False Float=-1)
    )) {isSetParameters=true, module=SetVariables_#4}
  ]
  particleSpawnScript [
   /Game/Effects/Niagara/Modules/Partition/FN_Particle_partition_v5/FN_Particle_Partition_v5(Enable Ribbon ID=true, Pattern:ENiagaraPartitionWaveForm=::NewEnumerator1, Seed:NiagaraInt32=$Emitter.RandomSeed)
   &InitializeRibbon(
    Color=&LerpLinearColors(
     StartColor=<rgba 0.4781 0.0478 20 1>
     EndColor=<rgba 29.2434 0 50 1>
     LerpFactor=&FloatFromCurve(
      FloatCurve=DataInterface() {propertyValues {Curve {keys [
          {interpMode=RCIM_Cubic, tangentMode=RCTM_User, time=0, value=0}
          {arriveTangent=1.3375, interpMode=RCIM_Cubic, leaveTangent=1.3375, tangentMode=RCTM_User, time=0.8, value=0.25}
          {interpMode=RCIM_Cubic, tangentMode=RCTM_User, time=1, value=1}
         ]}}}
      CurveIndex=&RandomRangeFloat(Minimum=0, Maximum=1)
     )
    )
    Lifetime Max=0.4
    Lifetime Min=0.15
    Lifetime Mode:ENiagara_LifetimeMode=::NewEnumerator1
    Ribbon Width=&RandomRangeFloat(Minimum=4, Maximum=8)
    UsePositionOffset=false
    Write Material Random=true
   )
   /Game/Effects/Niagara/Systems/Prototype/Tests/MusicDrivenCCPM/FN_Initialize_BoneFlow_Attributes(
    Number of Bone Sets=2
    Radius Multiplier:NiagaraFloat=&RandomRangeFloat(Minimum=-5, Maximum=5)
    RadiusMax=1
    RadiusMin=0.2
    Rate=1
   )
  ]
  particleUpdateScript [
   &ParticleState()
   &FN_BoneFlow(
    Angle:NiagaraFloat=$Particles.FN_Initialize_BoneFlow_Attributes.TScale
    BoneID:NiagaraInt32=$Particles.Partition.ID
    Height:NiagaraFloat=$Particles.Initial.Partition.NormalizedID
    Radius:NiagaraFloat=$Particles.Initial.FN_Initialize_BoneFlow_Attributes.Radius
    Set Material Dynamic=false
    SideVector1:Vector3f=<1 1 0>
    SideVector2:Vector3f=<-1 -1 0>
    SideVector3:Vector3f=<1 1 0>
    SkeletalMesh:NiagaraDataInterfaceSkeletalMesh=$User.SkeletalMesh
    Stabalize Twisting=true
    Use Sockets Instead=true
   )
   &ScaleColor(
    Scale RGB=&VectorFromFloat(Value=&Lerp_Float(
      A=10
      B=1
      Alpha=&FloatFromCurve(FloatCurve=DataInterface() {propertyValues {Curve {keys [
           {interpMode=RCIM_Cubic, tangentMode=RCTM_User, time=0, value=0}
           {arriveTangent=1.5977, interpMode=RCIM_Cubic, leaveTangent=1.5977, tangentMode=RCTM_User, time=0.3, value=0.65}
           {interpMode=RCIM_Cubic, tangentMode=RCTM_User, time=1, value=1}
          ]}}})
     ))
    ScaleA=true
    ScaleRGB=true
   )
   &ScaleRibbonWidth(Ribbon Width Scale=&Lerp_Float(
     A=3
     B=0.1
     Alpha=&FloatFromCurve(FloatCurve=DataInterface() {propertyValues {Curve {keys [{interpMode=RCIM_Cubic, tangentMode=RCTM_User, time=0, value=0}, {interpMode=RCIM_Cubic, tangentMode=RCTM_User, time=1, value=1}]}}})
    ))
   &DynamicMaterialParameters(
    Index 0 Param 1=1
    Index 0 Param 2=&FloatFromCurve(FloatCurve=DataInterface() {propertyValues {Curve {keys [<at 0 0>,<at 1 -1>]}}})
    Index 0 Param 3=&Multiply_Float(
     A=&RandomRangeFloat(Minimum=0.2, Maximum=0.3)
     B=&FloatFromCurve(FloatCurve=DataInterface() {propertyValues {Curve {keys [
          {interpMode=RCIM_Cubic, leaveTangentWeight=1.0373, tangentMode=RCTM_User, time=0, value=1}
          {interpMode=RCIM_Cubic, tangentMode=RCTM_User, time=1, value=-0.25}
         ]}}})
    )
    Module.Param0WriteEnabled=true
    Module.Param1WriteEnabled=true
    Module.Param2WriteEnabled=true
    Module.Param3WriteEnabled=true
   ) {module=DynamicMaterialParameters001}
  ]
  rendererClasses [&NiagaraRibbonRendererProperties]
  renderers [{rendererClass=&NiagaraRibbonRendererProperties}]
 }
 ElectricityBoneFlow001 {
  emitterSpawnScript [
   /SetVariables(Emitter.RandomSeed:NiagaraInt32=&RandomInt(Randomness Mode:ENiagaraRandomnessMode=::NewEnumerator0, Random Seed:NiagaraInt32=0, Fixed Random Seed:NiagaraBool=false)) {isSetParameters=true, module=SetVariables_#3}
   &FN_PartitionAttributes(Number of Bone Sets=2, Particles Per Partition=2, Partitions=2)
  ]
  emitterUpdateScript [
   &EmitterState(
    Life Cycle Mode=::NewEnumerator1
    Loop Delay=&RandomRangeFloat(Minimum=0.1, Maximum=0.5)
    Loop Duration=0.5
    Recalculate Duration Each Loop=true
    UseLoopDelay=true
   )
   &SpawnRate(SpawnRate=&Multiply_Float(A=35, B=1), Use Spawn Probability=false)
   /SetVariables(Emitter.RandXSpeed:NiagaraFloat=&Multiply_Float(
     A=&RandomRangeFloat(Minimum=1.1, Maximum=2)
     B=&MakeCustomFloatFromBool(Bool=&RandomBool(), True Float=1, False Float=-1)
    )) {isSetParameters=true, module=SetVariables_#4}
  ]
  particleSpawnScript [
   &FN_Particle_Partition_v5(Enable Ribbon ID=true, Pattern:ENiagaraPartitionWaveForm=::NewEnumerator1, Seed:NiagaraInt32=$Emitter.RandomSeed)
   &InitializeRibbon(
    Color=&LerpLinearColors(
     StartColor=<rgba 0.4781 0.0478 20 1>
     EndColor=<rgba 29.2434 0 50 1>
     LerpFactor=&FloatFromCurve(
      FloatCurve=DataInterface() {propertyValues {Curve {keys [
          {interpMode=RCIM_Cubic, tangentMode=RCTM_User, time=0, value=0}
          {arriveTangent=1.3375, interpMode=RCIM_Cubic, leaveTangent=1.3375, tangentMode=RCTM_User, time=0.8, value=0.25}
          {interpMode=RCIM_Cubic, tangentMode=RCTM_User, time=1, value=1}
         ]}}}
      CurveIndex=&RandomRangeFloat(Minimum=0, Maximum=1)
     )
    )
    Lifetime Max=0.4
    Lifetime Min=0.15
    Lifetime Mode:ENiagara_LifetimeMode=::NewEnumerator1
    Ribbon Width=&RandomRangeFloat(Minimum=4, Maximum=8)
    UsePositionOffset=false
    Write Material Random=true
   )
   &FN_Initialize_BoneFlow_Attributes(
    Number of Bone Sets=2
    Radius Multiplier:NiagaraFloat=&RandomRangeFloat(Minimum=-5, Maximum=5)
    RadiusMax=1
    RadiusMin=0.2
    Rate=1
   )
  ]
  particleUpdateScript [
   &ParticleState()
   &FN_BoneFlow(
    Angle:NiagaraFloat=$Particles.FN_Initialize_BoneFlow_Attributes.TScale
    BoneID:NiagaraInt32=$Particles.Partition.ID
    Enable Position Offsets=true
    Height:NiagaraFloat=$Particles.Initial.Partition.NormalizedID
    Radius:NiagaraFloat=$Particles.Initial.FN_Initialize_BoneFlow_Attributes.Radius
    Set Material Dynamic=false
    SideVector1:Vector3f=<1 1 0>
    SideVector2:Vector3f=<-1 -1 0>
    SideVector3:Vector3f=<1 1 0>
    SkeletalMesh:NiagaraDataInterfaceSkeletalMesh=$User.SkeletalMesh02
    Stabalize Twisting=true
    Use Sockets Instead=true
   )
   &ScaleColor(
    Scale RGB=&VectorFromFloat(Value=&Lerp_Float(
      A=10
      B=1
      Alpha=&FloatFromCurve(FloatCurve=DataInterface() {propertyValues {Curve {keys [
           {interpMode=RCIM_Cubic, tangentMode=RCTM_User, time=0, value=0}
           {arriveTangent=1.5977, interpMode=RCIM_Cubic, leaveTangent=1.5977, tangentMode=RCTM_User, time=0.3, value=0.65}
           {interpMode=RCIM_Cubic, tangentMode=RCTM_User, time=1, value=1}
          ]}}})
     ))
    ScaleA=true
    ScaleRGB=true
   )
   &ScaleRibbonWidth(Ribbon Width Scale=&Lerp_Float(
     A=3
     B=0.1
     Alpha=&FloatFromCurve(FloatCurve=DataInterface() {propertyValues {Curve {keys [{interpMode=RCIM_Cubic, tangentMode=RCTM_User, time=0, value=0}, {interpMode=RCIM_Cubic, tangentMode=RCTM_User, time=1, value=1}]}}})
    ))
   &DynamicMaterialParameters(
    Index 0 Param 1=1
    Index 0 Param 2=&FloatFromCurve(FloatCurve=DataInterface() {propertyValues {Curve {keys [<at 0 0>,<at 1 -1>]}}})
    Index 0 Param 3=&Multiply_Float(
     A=&RandomRangeFloat(Minimum=0.2, Maximum=0.3)
     B=&FloatFromCurve(FloatCurve=DataInterface() {propertyValues {Curve {keys [
          {interpMode=RCIM_Cubic, leaveTangentWeight=1.0373, tangentMode=RCTM_User, time=0, value=1}
          {interpMode=RCIM_Cubic, tangentMode=RCTM_User, time=1, value=-0.25}
         ]}}})
    )
    Module.Param0WriteEnabled=true
    Module.Param1WriteEnabled=true
    Module.Param2WriteEnabled=true
    Module.Param3WriteEnabled=true
   ) {module=DynamicMaterialParameters001}
  ]
  rendererClasses [&NiagaraRibbonRendererProperties]
  renderers [{rendererClass=&NiagaraRibbonRendererProperties}]
 }
 EmissiveLineGlow {
  emitterUpdateScript [
   &EmitterState(Life Cycle Mode=::NewEnumerator1, Loop Behavior=::NewEnumerator1, Loop Duration Mode=::NewEnumerator1, UseLoopDelay=false)
   &SpawnBurst_Instantaneous(Spawn Count=6, Use Spawn Probability=false, UseLoopCountLimit=false)
  ]
  particleSpawnScript [
   &PartitionParticles(Number Of Partitions=6)
   /SetVariables(Particles.ScaleMeshArray:Vector3f=&self:ScratchDynamicInput_01(
     ChargeAmount:NiagaraFloat=$User.ChargeAmount
     MeshScale:Vector3f=<0 0 0>
     MeshScaleArray:NiagaraDataInterfaceArrayFloat3=DataInterface() {propertyValues {FloatData [<0.16 0.4 1>, <0.16 0.4 1>, <0.16 0.3 1>, <0.16 0.3 1>, <0.26 0.2 1>, <0.26 0.2 1>]}}
     PartitionParticle:NiagaraInt32=$Particles.PartitionParticles.Partition
    )) {isSetParameters=true}
   &InitializeParticle(
    Color=<rgba 0.635 0 5 1>
    Mesh Scale=&SelectVectorFromArray(
     Direct Array Index=$Particles.PartitionParticles.Partition
     Vector Selection Array=DataInterface() {propertyValues {FloatData [<0.16 0.4 1>, <0.16 0.4 1>, <0.16 0.3 1>, <0.16 0.3 1>, <0.26 0.2 1>, <0.26 0.2 1>]}}
    )
    Mesh Scale Mode=::NewEnumerator4
    UsePositionOffset=false
    Write Material Random=true
   ) {module=InitializeParticle001}
   &SkeletalMeshLocation(
    Bone Sampling Mode=::NewEnumerator4
    Interpolate Mesh Sampling=false
    MaskSpawnGroup=false
    Orientation Sampling=::NewEnumerator1
    Sampled Position Offset=&SelectVectorFromArray(
     Direct Array Index=$Particles.PartitionParticles.Partition
     Vector Selection Array=DataInterface() {propertyValues {FloatData [<1.5 0 -12.5>, <1.5 0 12.5>, <3.5 0 -16>, <3.5 0 16>, <8 0 -23>, <8 0 23>]}}
    )
    Skeletal Mesh=$User.SkeletalMesh
   )
   &InitialMeshOrientation(
    Mesh Axis To Orient=<0 0 1>
    Mesh Orientation Mode=::NewEnumerator2
    Rotation=&SelectVectorFromArray(
     Direct Array Index=$Particles.PartitionParticles.Partition
     Vector Selection Array=DataInterface() {propertyValues {FloatData [<0 0 -0.06>, <0 0 0.06>, <0 0 0.3>, <0 0 -0.3>, <0 0 -0.06>, <0 0 0.06>]}}
    )
    Use Rotation Vector=true
   )
  ]
  particleUpdateScript [
   &ParticleState(Kill Particles When Lifetime Has Elapsed=false)
   &ScaleMeshSize(Scale Factor=&MakeVector(
     X=1
     Y=&Multiply_Float(
      A=0.7
      B=&Waveform(
       "Waveform [1]"=::NewEnumerator0
       "[1] Frequency"=0.6
       "[1] Amplitude Min/Max"=&RandomRangeVector2D(Minimum=<0.6 0.9>, Maximum=<0.8 1.2>)
      )
     )
     Z=1
    ))
   &ScaleColor(
    Scale Alpha=&Multiply_Float(
     A=0.3
     B=&FloatFromCurve(
      FloatCurve=DataInterface() {propertyValues {Curve {keys [
          {interpMode=RCIM_Cubic, tangentMode=RCTM_User, time=0, value=0}
          {arriveTangent=0.8916, interpMode=RCIM_Cubic, leaveTangent=0.8916, tangentMode=RCTM_User, time=0.8, value=0.2}
          {arriveTangentWeight=0.3137, interpMode=RCIM_Cubic, tangentMode=RCTM_User, time=0.9, value=1}
         ]}}}
      CurveIndex=$System.ChargeAmount
     )
    )
    Scale RGB=&VectorFromFloat(Value=5)
    ScaleA=true
    ScaleRGB=true
   ) {module=ScaleColor001}
  ]
  rendererClasses [&NiagaraMeshRendererProperties]
  renderers [{rendererClass=&NiagaraMeshRendererProperties, rendererIndex=0}]
 }
 GlowGem {
  emitterUpdateScript [
   &EmitterState(
    Inactive Response=::NewEnumerator1
    Life Cycle Mode=::NewEnumerator1
    Loop Behavior=::NewEnumerator1
    Loop Duration Mode=::NewEnumerator1
    UseLoopDelay=false
   )
   &SpawnBurst_Instantaneous(Use Spawn Probability=false, UseLoopCountLimit=false)
  ]
  particleSpawnScript [
   &InitializeParticle(
    Color=<rgba 0.1216 0.0683 1 1>
    Sprite Rotation Mode=::NewEnumerator2
    Sprite Size=<125 35>
    Sprite Size Mode=::NewEnumerator4
    UsePositionOffset=false
    Write Material Random=true
   )
   /SetVariables(Particles.GlowRound:Vector2f=<35 35>) {isSetParameters=true}
   &CameraOffset(Camera Offset Amount=5)
   /SetVariables(Particles.FlareCameraOffset=20) {isSetParameters=true}
  ]
  particleUpdateScript [
   &ParticleState(Kill Particles When Lifetime Has Elapsed=false, Let Infinitely Lived Particles Die When Emitter Deactivates=true)
   &SpriteRotationRate(Rotation Rate=&Lerp_Float(
     A=30
     B=100
     Alpha=&FloatFromCurve(
      FloatCurve=DataInterface() {propertyValues {Curve {keys [
          {arriveTangent=4.7187, leaveTangentWeight=0.1789, time=0, value=0}
          <at 0.2 1>
          <at 0.8 1>
          {arriveTangent=-8.2126, interpMode=RCIM_Cubic, leaveTangent=-8.2126, tangentMode=RCTM_User, time=1, value=0}
         ]}}}
      CurveIndex=$System.Timeline.NormalizedTime
      Scale Curve=&MakeFloatFromInt(Int=&MakeCustomIntFromBool(Bool=$System.Timeline.IsPlaying, True INT=1, False INT=0))
     )
    ))
   &ScaleSpriteSize(Scale Sprite Size Mode=::NewEnumerator0)
   &ScaleColor(
    Scale Alpha=&Lerp_Float(
     A=1
     B=1
     Alpha=&FloatFromCurve(
      FloatCurve=DataInterface() {propertyValues {Curve {keys [
          {interpMode=RCIM_Cubic, leaveTangentWeight=0.1789, tangentMode=RCTM_User, time=0, value=0}
          {interpMode=RCIM_Cubic, tangentMode=RCTM_User, time=0.5, value=1}
          {arriveTangent=-8.2126, interpMode=RCIM_Cubic, leaveTangent=-8.2126, tangentMode=RCTM_User, time=0.8, value=0}
         ]}}}
      CurveIndex=$System.Timeline.NormalizedTime
      Scale Curve=&MakeFloatFromInt(Int=&MakeCustomIntFromBool(Bool=$System.Timeline.IsPlaying, True INT=1, False INT=0))
     )
    )
    Scale RGB=&VectorFromFloat(Value=&Lerp_Float(
      A=1
      B=10
      Alpha=&FloatFromCurve(
       FloatCurve=DataInterface() {propertyValues {Curve {keys [
           {interpMode=RCIM_Cubic, leaveTangentWeight=0.1789, tangentMode=RCTM_User, time=0, value=0}
           {interpMode=RCIM_Cubic, tangentMode=RCTM_User, time=0.5, value=1}
           {arriveTangent=-8.2126, interpMode=RCIM_Cubic, leaveTangent=-8.2126, tangentMode=RCTM_User, time=0.8, value=0}
          ]}}}
       CurveIndex=$System.Timeline.NormalizedTime
       Scale Curve=&MakeFloatFromInt(Int=&MakeCustomIntFromBool(Bool=$System.Timeline.IsPlaying, True INT=1, False INT=0))
      )
     ))
    ScaleA=true
    ScaleRGB=true
   )
   /SetVariables(Particles.FlareVisibility:NiagaraInt32=&MakeCustomIntFromBool(Bool=$System.Timeline.IsPlaying, True INT=0, False INT=1)) {isSetParameters=true}
  ]
  rendererClasses [&NiagaraSpriteRendererProperties]
  renderers [
   {rendererClass=&NiagaraSpriteRendererProperties}
   {rendererClass=&NiagaraSpriteRendererProperties, rendererIndex=1}
  ]
 }
}
summary {
 systemName=NS_AncientWeapon_Ambient
 emitters [
  {emitterName=GlowGem, rendererClasses [&NiagaraSpriteRendererProperties]}
  {emitterName=ElectricityBoneFlow, rendererClasses [&NiagaraRibbonRendererProperties]}
  {emitterName=ElectricityBoneFlow001, rendererClasses [&NiagaraRibbonRendererProperties]}
  {emitterName=BladeEdgeEnergy, rendererClasses [&NiagaraMeshRendererProperties]}
  {emitterName=BladeEdgeGlow, rendererClasses [&NiagaraMeshRendererProperties]}
  {emitterName=EmissiveLineGlow, rendererClasses [&NiagaraMeshRendererProperties]}
  {emitterName=BladeGlow, rendererClasses [&NiagaraMeshRendererProperties]}
  {emitterName=BladeGlowCharge, rendererClasses [&NiagaraMeshRendererProperties]}
 ]
 userVariables [
  User.AmmoCount {type=NiagaraFloat, defaultValue=0, description=""}
  User.Charge {type=NiagaraBool}
  User.ChargeAmount {type=NiagaraFloat, defaultValue=3, description=""}
  User.Magnitude {type=NiagaraFloat, defaultValue=0, description=""}
  User.SkeletalMesh {
   type=NiagaraDataInterfaceSkeletalMesh
   defaultValue {
    struct=NiagaraExt_VariableValue_DataInterface
    value {dataInterface=&self:NiagaraDataInterfaceSkeletalMesh_0}
   }
  }
  User.SkeletalMesh02 {
   type=NiagaraDataInterfaceSkeletalMesh
   defaultValue {
    struct=NiagaraExt_VariableValue_DataInterface
    value {dataInterface=&self:NiagaraDataInterfaceSkeletalMesh_1}
   }
  }
 ]
}
