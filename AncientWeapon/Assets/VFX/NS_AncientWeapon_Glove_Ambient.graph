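# Niagara ambient effect for the Ancient Weapon glove, showing purple electricity, glows and static, though every emitter is saved disabled.
# Electricity spawns flickering mesh arcs, sprites and lights at five partition-indexed sockets on a user-supplied skeletal mesh.
# Two BoneFlow emitters stream partitioned ribbons along sockets of two skeletal meshes, with Overheat-gated bursts switched off.
# Sprite emitters add socket glows, waveform-pulsed glows and rotating static, which also plays persistent audio.
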NiagaraSystem /AncientWeapon/Assets/VFX/NS_AncientWeapon_Glove_Ambient
dependencies {
 usedDataInterfaces [
  &NiagaraDataInterfaceCurve
  &NiagaraDataInterfaceAudioPlayer
  &NiagaraDataInterfaceArrayFloat3
  &NiagaraDataInterfaceArrayFloat2
  &NiagaraDataInterfaceArrayFloat
  &NiagaraDataInterfaceVectorCurve
  &NiagaraDataInterfaceColorCurve
 ]
 usedDynamicInputs []
 usedModules [&SystemState]
}
emitters {
 Electricity {
  bEnabled=false
  emitterUpdateScript [
   &EmitterState(Life Cycle Mode=::NewEnumerator1, Loop Duration=0.5, Recalculate Duration Each Loop=true, UseLoopDelay=false)
   &SpawnBurst_Instantaneous(Spawn Count=5, Use Spawn Probability=false, UseLoopCountLimit=false) {enabled=false}
   &SpawnRate(Spawn Probability=&RandomRangeFloat(Minimum=0.2, Maximum=1), SpawnRate=6, Use Spawn Probability=true)
   /SetVariables(Emitter.PanSpeed:NiagaraFloat=&Multiply_Float(
     A=&RandomRangeFloat(Minimum=0.3, Maximum=0.7)
     B=&MakeCustomFloatFromBool(Bool=&RandomBool(), True Float=1, False Float=-1)
    )) {isSetParameters=true}
  ]
  particleSpawnScript [
   &PartitionParticles(Number Of Partitions=5)
   &InitializeParticle(
    Color Maximum=<rgba 0.7033 0 1 1>
    Color Minimum=<rgba 0.2814 0.0045 1 1>
    Color Mode=::NewEnumerator2
    Lifetime Max=0.4
    Lifetime Min=0.2
    Lifetime Mode=::NewEnumerator1
    Mesh Scale=&SelectVectorFromArray(
     Direct Array Index=$Particles.PartitionParticles.Partition
     Vector Selection Array=DataInterface() {propertyValues {FloatData [<0.45 1 0.3>, <0.43 1 0.3>, <0.43 1 0.3>, <0.45 1 0.3>, <0.4 1 0.3>]}}
    )
    Mesh Scale Mode=::NewEnumerator4
    Sprite Size Mode=::NewEnumerator1
    Uniform Sprite Size Max=6
    Uniform Sprite Size Min=3
    UsePositionOffset=false
    Write Material Random=true
   )
   &SkeletalMeshLocation(
    Bone / Socket Index=$Particles.PartitionParticles.Partition
    MaskSpawnGroup=false
    Mesh Sampling Type=::NewEnumerator5
    Orientation Sampling=::NewEnumerator1
    Skeletal Mesh=$User.SkeletalMesh
    Socket Sampling Mode=::NewEnumerator5
   )
   &SkeletalMeshLocation(
    Bone / Socket Index=5
    Initialize Particle Position To Sampled Position=false
    Interpolate Mesh Sampling=false
    MaskSpawnGroup=false
    Mesh Sampling Type=::NewEnumerator5
    Position Sampling=::NewEnumerator1
    Sampled Position Offset=<0.5 -1.5 1>
    Skeletal Mesh=$User.SkeletalMesh
    Socket Sampling Mode=::NewEnumerator5
    Write To Particle Position=false
   ) {enabled=false, module=SkeletalMeshLocation001}
   &DynamicMaterialParameters(
    Index 0 Param 1=&UniformRangedFloat(Minimum=4, Maximum=4)
    Index 0 Param 3=$Particles.MaterialRandom
    Index 0 Param 4=&UniformRangedFloat(Minimum=0.5, Maximum=0.8)
    Module.Param0WriteEnabled=true
    Module.Param1WriteEnabled=false
    Module.Param2WriteEnabled=true
    Module.Param3WriteEnabled=true
   ) {module=DynamicMaterialParameters001}
   &InitialMeshOrientation(
    Mesh Axis To Orient=&SelectVectorFromArray(
     Direct Array Index=$Particles.PartitionParticles.Partition
     Vector Selection Array=DataInterface() {propertyValues {FloatData [<1 0 0>, <1 0 0>, <1 0.05 0>, <1 -0.05 0>, <1 0.5 0>]}}
    )
    Mesh Orientation Mode=::NewEnumerator2
    Rotation=&SelectVectorFromArray(
     Direct Array Index=$Particles.PartitionParticles.Partition
     Vector Selection Array=DataInterface() {propertyValues {FloatData [<0 0 0.95>, <0 0 0.98>, <0.07 0 0>, <0.18 0 0>, <-0.2 0 0>]}}
    )
    Use Rotation Vector=true
   )
   &AddVelocity(Velocity=<0 15 50>) {enabled=false}
   /SetVariables(
    Particles.FlareColor=&MultiplyLinearColor(
     Linear Color=$Particles.Color
     Scale Factor=&MakeFloat4FromVectorAndFloat(
      "Vector (RGB)"=&VectorFromFloat(Value=&RandomRangeFloat(Minimum=0.5, Maximum=1))
      "Float (Alpha)"=&RandomRangeFloat(Minimum=0.1, Maximum=0.25)
     )
    )
    Particles.FlareSize:Vector2f=&RandomRangeVector2D(Minimum=<8 8>, Maximum=<12 12>)
    Particles.GlowColor=&MultiplyLinearColor(
     Linear Color=$Particles.Color
     Scale Factor=&MakeFloat4FromVectorAndFloat(
      "Vector (RGB)"=&VectorFromFloat(Value=&RandomRangeFloat(Minimum=2, Maximum=10))
      "Float (Alpha)"=&RandomRangeFloat(Minimum=0.5, Maximum=1)
     )
    )
    Particles.GlowPosition:NiagaraPosition=<-1 -1 0.2>
    Particles.GlowSize=&RandomRangeVector2D(Minimum=<3 3>, Maximum=<6 6>)
   ) {isSetParameters=true}
   &SpriteFacingAndAlignment(Sprite Facing=<-0.1 -0.3 1>, WriteAlignment=true, WriteFacing=true)
  ]
  particleUpdateScript [
   &ParticleState()
   /Niagara/Modules/Update/Orientation/V2/OrientMeshToVector(
    Look At Direction:Vector3f=&Multiply_Vector(
     A=<1 1 0>
     B=&Subtract_Vector(A=Linked() {linkedVariable {name=Particles.Position, type=NiagaraPosition}}, B=/Niagara/DynamicInputs/Helpers/SimulationPosition())
    )
    Override Quat=false
    UseOptionalSideAxis=false
    UseOptionalUpAxis=false
   ) {enabled=false}
   /SetVariables(Particles.Position=&MakeVector(
     X=&MakeFloatFromVector(Vector=Linked() {linkedVariable {name=Particles.Position, type=NiagaraPosition}}, Channel=::NewEnumerator0)
     Y=&MakeFloatFromVector(Vector=Linked() {linkedVariable {name=Particles.Position, type=NiagaraPosition}}, Channel=::NewEnumerator1)
     Z=0
    )) {isSetParameters=true}
   &DynamicMaterialParameters(
    Index 0 Param 1=&FloatFromCurve(
     FloatCurve=DataInterface() {propertyValues {Curve {keys [<at 0.5 1 RCIM_Cubic>, {arriveTangent=-2.5142, interpMode=RCIM_Cubic, leaveTangent=-2.5142, tangentMode=RCTM_User, time=1, value=0}]}}}
     Scale Curve=&Lerp_Float(A=0.7, B=1.5, Alpha=$Particles.Initial.MaterialRandom)
    )
    Index 0 Param 2=&FloatFromCurve(
     FloatCurve=DataInterface() {propertyValues {Curve {keys [
         {arriveTangent=10.9494, interpMode=RCIM_Cubic, leaveTangent=10.9494, tangentMode=RCTM_User, time=0, value=1}
         {arriveTangent=-8.5345, interpMode=RCIM_Cubic, leaveTangent=-8.5345, tangentMode=RCTM_User, time=0.2, value=2}
         {arriveTangent=-2.4364, interpMode=RCIM_Cubic, leaveTangent=-2.4364, time=0.3638, value=0.5858}
         {arriveTangent=-8.496, interpMode=RCIM_Cubic, leaveTangent=-8.496, tangentMode=RCTM_User, time=0.4639, value=1.357}
         {arriveTangent=8.1076, interpMode=RCIM_Cubic, leaveTangent=-1.1376, tangentMode=RCTM_User, time=1, value=1.5}
        ]}}}
     Scale Curve=0.75
    )
    Index 0 Param 3=1
    Index 0 Param 4=1
    Module.Param0WriteEnabled=true
    Module.Param1WriteEnabled=true
    Module.Param2WriteEnabled=true
    Module.Param3WriteEnabled=true
   ) {enabled=false}
   &DynamicMaterialParameters(
    Index 0 Param 1=&FloatFromCurve(
     FloatCurve=DataInterface() {propertyValues {Curve {keys [
         <at 0.5 1 RCIM_Cubic>
         {arriveTangent=-1.5423, arriveTangentWeight=0.451, interpMode=RCIM_Cubic, leaveTangent=-1.5423, tangentMode=RCTM_User, time=1, value=0}
        ]}}}
     Scale Curve=&Lerp_Float(A=0.7, B=1.5, Alpha=$Particles.Initial.MaterialRandom)
    )
    Index 0 Param 2=&FloatFromCurve(
     FloatCurve=DataInterface() {propertyValues {Curve {keys [
         {arriveTangent=10.9494, interpMode=RCIM_Cubic, leaveTangent=10.9494, tangentMode=RCTM_User, time=0, value=1}
         {arriveTangent=-8.5345, interpMode=RCIM_Cubic, leaveTangent=-8.5345, tangentMode=RCTM_User, time=0.2, value=2}
         {arriveTangent=-2.4364, interpMode=RCIM_Cubic, leaveTangent=-2.4364, time=0.3638, value=0.5858}
         {arriveTangent=-8.496, interpMode=RCIM_Cubic, leaveTangent=-8.496, tangentMode=RCTM_User, time=0.4639, value=1.357}
         {arriveTangent=8.1076, interpMode=RCIM_Cubic, leaveTangent=-1.1376, tangentMode=RCTM_User, time=1, value=1.5}
        ]}}}
     Scale Curve=&RandomRangeFloat(Minimum=0.05, Maximum=0.5)
    )
    Module.Param0WriteEnabled=true
    Module.Param1WriteEnabled=true
    Module.Param2WriteEnabled=true
    Module.Param3WriteEnabled=true
   ) {module=DynamicMaterialParameters002}
   /SetVariables(Particles.DynamicParamOverride:Vector4f=&MakeFloat4FromVectorAndFloat(
     "Vector (RGB)"=/Niagara/DynamicInputs/Transforms/TransformPosition(
      Position=Linked() {linkedVariable {name=Particles.Initial.Position, type=NiagaraPosition}}
      Source Space=::Simulation
      Destination Space=::World
     )
     "Float (Alpha)"=0
    )) {isSetParameters=true}
   &ScaleMeshSize(Scale Factor=&MakeVector(
     X=1
     Y=1
     Z=&Multiply_Float(
      A=&FloatFromCurve(FloatCurve=DataInterface() {propertyValues {Curve {keys [
           {interpMode=RCIM_Cubic, tangentMode=RCTM_User, time=0, value=0.2}
           {arriveTangent=1.3862, interpMode=RCIM_Cubic, leaveTangent=1.3862, tangentMode=RCTM_User, time=0.3, value=0.625}
           {arriveTangentWeight=1.7224, interpMode=RCIM_Cubic, tangentMode=RCTM_User, time=1, value=1}
          ]}}})
      B=1
     )
    ))
   &ScaleMeshSize(Scale Factor=&VectorFromCurve(
     VectorCurve=DataInterface() {propertyValues {
       XCurve {keys [
         {arriveTangent=2.9222, interpMode=RCIM_Cubic, leaveTangent=2.9222, tangentMode=RCTM_User, time=0, value=0}
         {interpMode=RCIM_Cubic, tangentMode=RCTM_User, time=0.65, value=1}
        ]}
       YCurve {keys [<at 0 1>]}
       ZCurve {keys [
         {arriveTangent=3.2926, interpMode=RCIM_Cubic, leaveTangent=3.2926, tangentMode=RCTM_User, time=0, value=0.1}
         {arriveTangent=-1.316, interpMode=RCIM_Cubic, leaveTangent=-1.316, tangentMode=RCTM_User, time=0.2588, value=1}
         {arriveTangent=0.348, interpMode=RCIM_Cubic, leaveTangent=0.348, tangentMode=RCTM_User, time=0.4597, value=0.4069}
         {interpMode=RCIM_Cubic, leaveTangent=3.5341, time=1, value=1.5}
        ]}
      }}
     CurveIndex=$Particles.Age
    )) {enabled=false, module=ScaleMeshSize002}
   &ScaleMeshSize() {enabled=false, module=ScaleMeshSize001}
   &SolveForcesAndVelocity(Clamp Velocity=false, Limit Acceleration=false)
   &ScaleColor(
    Scale Alpha=&self:Flicker(
     FlickerAmount:NiagaraFloat=&FloatFromCurve(FloatCurve=DataInterface() {propertyValues {Curve {keys [
          {arriveTangent=2.7223, interpMode=RCIM_Cubic, leaveTangent=2.7223, tangentMode=RCTM_User, time=0, value=1}
          {arriveTangent=0.0266, interpMode=RCIM_Cubic, leaveTangent=0.0266, time=0.2457, value=0.5}
          {arriveTangent=-0.3314, interpMode=RCIM_Cubic, leaveTangent=-0.3314, time=0.527, value=1.014}
          {arriveTangent=-3.2376, interpMode=RCIM_Cubic, leaveTangent=-11960.9277, tangentMode=RCTM_User, time=1, value=0.25}
         ]}}})
     FlickerInterval:NiagaraFloat=0.1
    )
    Scale RGB=&VectorFromFloat(Value=&Lerp_Float(
      A=0.4
      B=5
      Alpha=&FloatFromCurve(
       FloatCurve=DataInterface() {propertyValues {Curve {keys [
           {interpMode=RCIM_Cubic, tangentMode=RCTM_User, time=0, value=0}
           {arriveTangent=0.6368, interpMode=RCIM_Cubic, leaveTangent=0.6368, tangentMode=RCTM_User, time=0.8, value=0.2}
           {arriveTangent=5.7967, interpMode=RCIM_Cubic, leaveTangent=5.7967, tangentMode=RCTM_User, time=1, value=1}
          ]}}}
       CurveIndex=&RandomRangeFloat(Minimum=0, Maximum=1)
      )
     ))
    ScaleA=true
    ScaleRGB=true
   )
   &FN_Light_Attributes(
    Color=$Particles.Initial.Color
    Create Color Attribute=true
    Create Exponent Attribute=true
    Create Position Attribute=true
    Create Radius Attribute=true
    Position=&AddVectorToPosition(Position=Linked() {linkedVariable {name=Particles.Initial.GlowPosition, type=Vector3f}}, Vector=<-8 0 5>)
    Radius=100
    Scale Color=2
   )
  ]
  rendererClasses [
   &NiagaraSpriteRendererProperties
   &NiagaraMeshRendererProperties
   &NiagaraLightRendererProperties
  ]
  renderers [
   {rendererClass=&NiagaraSpriteRendererProperties}
   {rendererClass=&NiagaraMeshRendererProperties, rendererIndex=1}
   {rendererClass=&NiagaraSpriteRendererProperties, rendererIndex=2}
   {rendererClass=&NiagaraSpriteRendererProperties, rendererIndex=3}
   {rendererClass=&NiagaraLightRendererProperties, rendererIndex=4}
  ]
 }
 ElectricityBoneFlow {
  bEnabled=false
  emitterSpawnScript [
   /SetVariables(Emitter.RandomSeed:NiagaraInt32=/Niagara/DynamicInputs/Random/RandomInt(Randomness Mode:ENiagaraRandomnessMode=::NewEnumerator0, Random Seed:NiagaraInt32=0, Fixed Random Seed:NiagaraBool=false)) {isSetParameters=true, module=SetVariables_#1}
   /Game/Effects/Niagara/Modules/Partition/FN_PartitionAttributes(Number of Bone Sets=5, Particles Per Partition=5, Partitions=5)
  ]
  emitterUpdateScript [
   &EmitterState(
    Life Cycle Mode=::NewEnumerator1
    Loop Delay=&RandomRangeFloat(Minimum=0.1, Maximum=0.5)
    Loop Duration=0.5
    Recalculate Duration Each Loop=true
    UseLoopDelay=true
   )
   &SpawnBurst_Instantaneous(Spawn Count=5, Use Spawn Probability=false, UseLoopCountLimit=false) {enabled=false, module=SpawnBurst_Instantaneous002}
   &SpawnRate(SpawnRate=100, Use Spawn Probability=false)
   &SpawnBurst_Instantaneous(
    Spawn Count=$Emitter.FN_PartitionAttributes.Spawn Total
    Spawn Probability=&MakeCustomFloatFromBool(Bool=&SetBoolByFloatComparison(A=1, B=0.5), True Float=1, False Float=0)
    Spawn Time=&RandomRangeFloat(Minimum=0, Maximum=0.5)
    Use Spawn Probability=true
   ) {enabled=false}
   &SpawnBurst_Instantaneous(
    Spawn Count=$Emitter.FN_PartitionAttributes.Spawn Total
    Spawn Probability=&MakeCustomFloatFromBool(Bool=&SetBoolByFloatComparison(A=$User.Overheat, B=0.2), True Float=1, False Float=0)
    Spawn Time=&RandomRangeFloat(Minimum=0, Maximum=0.1)
    Use Spawn Probability=true
   ) {enabled=false, module=SpawnBurst_Instantaneous001}
   /SetVariables(Emitter.RandXSpeed:NiagaraFloat=&Multiply_Float(
     A=&RandomRangeFloat(Minimum=1.1, Maximum=2)
     B=&MakeCustomFloatFromBool(Bool=&RandomBool(), True Float=1, False Float=-1)
    )) {isSetParameters=true, module=SetVariables_#2}
  ]
  particleSpawnScript [
   /Game/Effects/Niagara/Modules/Partition/FN_Particle_partition_v5/FN_Particle_Partition_v5(Enable Ribbon ID=true, Pattern:ENiagaraPartitionWaveForm=::NewEnumerator1, Seed:NiagaraInt32=$Emitter.RandomSeed)
   /SetVariables(Particles.ColorStart:LinearColor=&MultiplyLinearColor(
     Linear Color=&RandomRangeLinearColor(Minimum=<rgba 0.0221 0 0.3 1>, Maximum=<rgba 1 0.0603 0.9063 1>)
     Scale Factor=&MakeFloat4FromVectorAndFloat("Vector (RGB)"=&VectorFromFloat(Value=1), "Float (Alpha)"=1)
    )) {enabled=false, isSetParameters=true, module=SetVariables_#3}
   &InitializeParticle(Color=<rgba 0.5849 0 1 1>, Lifetime=0.3, UsePositionOffset=false, Write Material Random=true) {enabled=false}
   &InitializeRibbon(
    Color=<rgba 29.2434 0 50 1>
    Lifetime Max=0.4
    Lifetime Min=0.15
    Lifetime Mode:ENiagara_LifetimeMode=::NewEnumerator1
    Ribbon Width=&RandomRangeFloat(Minimum=6, Maximum=13)
    UsePositionOffset=false
    Write Material Random=true
   )
   /Game/Effects/Niagara/Systems/Prototype/Tests/MusicDrivenCCPM/FN_Initialize_BoneFlow_Attributes(Angle Rate Max=100, Angle Rate Min=75, Number of Bone Sets=5, Rate=1)
   &DynamicMaterialParameters(
    Index 0 Param 1=1
    Index 0 Param 2=&FloatFromCurve(FloatCurve=DataInterface() {propertyValues {Curve {keys [<at 0 0>,<at 1 -1>]}}})
    Index 0 Param 3=&RandomRangeFloat(Minimum=0.2, Maximum=0.3)
    Module.Param0WriteEnabled=true
    Module.Param1WriteEnabled=true
    Module.Param2WriteEnabled=true
    Module.Param3WriteEnabled=false
   ) {enabled=false}
  ]
  particleUpdateScript [
   &ParticleState()
   &FN_BoneFlow(
    Angle:NiagaraFloat=$Particles.FN_Initialize_BoneFlow_Attributes.TScale
    BoneID:NiagaraInt32=$Particles.Partition.ID
    Height:NiagaraFloat=$Particles.Initial.Partition.NormalizedID
    Radius:NiagaraFloat=$Particles.Initial.FN_Initialize_BoneFlow_Attributes.Radius
    Set Material Dynamic=false
    SkeletalMesh:NiagaraDataInterfaceSkeletalMesh=$User.SkeletalMesh
    Stabalize Twisting=true
    Use Sockets Instead=true
   )
   &ScaleColor(
    Scale RGB=&VectorFromFloat(Value=&Lerp_Float(
      A=10
      B=1
      Alpha=&FloatFromCurve(FloatCurve=DataInterface() {propertyValues {Curve {keys [
           {interpMode=RCIM_Cubic, tangentMode=RCTM_User, time=0, value=0}
           {arriveTangent=1.5977, interpMode=RCIM_Cubic, leaveTangent=1.5977, tangentMode=RCTM_User, time=0.3, value=0.65}
           {interpMode=RCIM_Cubic, tangentMode=RCTM_User, time=1, value=1}
          ]}}})
     ))
    ScaleA=true
    ScaleRGB=true
   )
   &ScaleRibbonWidth(Ribbon Width Scale=&Lerp_Float(
     A=3
     B=0.1
     Alpha=&FloatFromCurve(FloatCurve=DataInterface() {propertyValues {Curve {keys [{interpMode=RCIM_Cubic, tangentMode=RCTM_User, time=0, value=0}, {interpMode=RCIM_Cubic, tangentMode=RCTM_User, time=1, value=1}]}}})
    ))
   &ScaleColor(
    Scale Alpha=&FloatFromCurve(
     FloatCurve=DataInterface() {propertyValues {Curve {keys [<at 0 0>,<at 0.1 1>,<at 0.25 1>,<at 0.3 0>,<at 0.7 0>,<at 0.75 1>,<at 0.9 1>,<at 1 0>]}}}
     CurveIndex=$Particles.Partition.NormalizedID
    )
    Scale RGB=&MakeVectorFromLinearColorRGB(Color=&LerpLinearColors(
      StartColor=<rgba 1 0 0.971 1>
      EndColor=<rgba 0.4668 0 1 1>
      LerpFactor=&FloatFromCurve(FloatCurve=DataInterface() {propertyValues {Curve {keys [{interpMode=RCIM_Cubic, tangentMode=RCTM_User, time=0, value=0}, {interpMode=RCIM_Cubic, tangentMode=RCTM_User, time=1, value=1}]}}})
     ))
    ScaleA=true
    ScaleRGB=true
   ) {enabled=false, module=ScaleColor001}
   &DynamicMaterialParameters(
    Index 0 Param 1=1
    Index 0 Param 2=&FloatFromCurve(FloatCurve=DataInterface() {propertyValues {Curve {keys [<at 0 0>,<at 1 -1>]}}})
    Index 0 Param 3=&Multiply_Float(
     A=&RandomRangeFloat(Minimum=0.2, Maximum=0.3)
     B=&FloatFromCurve(FloatCurve=DataInterface() {propertyValues {Curve {keys [
          {interpMode=RCIM_Cubic, leaveTangentWeight=1.0373, tangentMode=RCTM_User, time=0, value=1}
          {interpMode=RCIM_Cubic, tangentMode=RCTM_User, time=1, value=-0.25}
         ]}}})
    )
    Module.Param0WriteEnabled=true
    Module.Param1WriteEnabled=true
    Module.Param2WriteEnabled=true
    Module.Param3WriteEnabled=true
   ) {module=DynamicMaterialParameters001}
  ]
  rendererClasses [&NiagaraRibbonRendererProperties]
  renderers [{rendererClass=&NiagaraRibbonRendererProperties}]
 }
 ElectricityBoneFlow001 {
  bEnabled=false
  emitterSpawnScript [
   /SetVariables(Emitter.RandomSeed:NiagaraInt32=&RandomInt(Randomness Mode:ENiagaraRandomnessMode=::NewEnumerator0, Random Seed:NiagaraInt32=0, Fixed Random Seed:NiagaraBool=false)) {isSetParameters=true, module=SetVariables_#1}
   &FN_PartitionAttributes(Number of Bone Sets=2, Particles Per Partition=2, Partitions=2)
  ]
  emitterUpdateScript [
   &EmitterState(
    Life Cycle Mode=::NewEnumerator1
    Loop Delay=&RandomRangeFloat(Minimum=0.1, Maximum=0.5)
    Loop Duration=0.5
    Recalculate Duration Each Loop=true
    UseLoopDelay=true
   )
   &SpawnBurst_Instantaneous(Spawn Count=5, Use Spawn Probability=false, UseLoopCountLimit=false) {enabled=false, module=SpawnBurst_Instantaneous002}
   &SpawnRate(SpawnRate=25, Use Spawn Probability=false)
   &SpawnBurst_Instantaneous(
    Spawn Count=$Emitter.FN_PartitionAttributes.Spawn Total
    Spawn Probability=&MakeCustomFloatFromBool(Bool=&SetBoolByFloatComparison(A=1, B=0.5), True Float=1, False Float=0)
    Spawn Time=&RandomRangeFloat(Minimum=0, Maximum=0.5)
    Use Spawn Probability=true
   ) {enabled=false}
   &SpawnBurst_Instantaneous(
    Spawn Count=$Emitter.FN_PartitionAttributes.Spawn Total
    Spawn Probability=&MakeCustomFloatFromBool(Bool=&SetBoolByFloatComparison(A=$User.Overheat, B=0.2), True Float=1, False Float=0)
    Spawn Time=&RandomRangeFloat(Minimum=0, Maximum=0.1)
    Use Spawn Probability=true
   ) {enabled=false, module=SpawnBurst_Instantaneous001}
   /SetVariables(Emitter.RandXSpeed:NiagaraFloat=&Multiply_Float(
     A=&RandomRangeFloat(Minimum=1.1, Maximum=2)
     B=&MakeCustomFloatFromBool(Bool=&RandomBool(), True Float=1, False Float=-1)
    )) {isSetParameters=true, module=SetVariables_#2}
  ]
  particleSpawnScript [
   &FN_Particle_Partition_v5(Enable Ribbon ID=true, Pattern:ENiagaraPartitionWaveForm=::NewEnumerator1, Seed:NiagaraInt32=$Emitter.RandomSeed)
   /SetVariables(Particles.ColorStart:LinearColor=&MultiplyLinearColor(
     Linear Color=&RandomRangeLinearColor(Minimum=<rgba 0.0221 0 0.3 1>, Maximum=<rgba 1 0.0603 0.9063 1>)
     Scale Factor=&MakeFloat4FromVectorAndFloat("Vector (RGB)"=&VectorFromFloat(Value=1), "Float (Alpha)"=1)
    )) {enabled=false, isSetParameters=true, module=SetVariables_#3}
   &InitializeParticle(Color=<rgba 0.5849 0 1 1>, Lifetime=0.3, UsePositionOffset=false, Write Material Random=true) {enabled=false}
   &InitializeRibbon(
    Color=<rgba 29.2434 0 50 1>
    Lifetime Max=0.4
    Lifetime Min=0.15
    Lifetime Mode:ENiagara_LifetimeMode=::NewEnumerator1
    Ribbon Width=&RandomRangeFloat(Minimum=6, Maximum=13)
    UsePositionOffset=false
    Write Material Random=true
   )
   &FN_Initialize_BoneFlow_Attributes(
    Number of Bone Sets=2
    Radius Multiplier:NiagaraFloat=&RandomRangeFloat(Minimum=-5, Maximum=5)
    RadiusMax=1
    RadiusMin=0.2
    Rate=1
   )
   &DynamicMaterialParameters(
    Index 0 Param 1=1
    Index 0 Param 2=&FloatFromCurve(FloatCurve=DataInterface() {propertyValues {Curve {keys [<at 0 0>,<at 1 -1>]}}})
    Index 0 Param 3=&RandomRangeFloat(Minimum=0.2, Maximum=0.3)
    Module.Param0WriteEnabled=true
    Module.Param1WriteEnabled=true
    Module.Param2WriteEnabled=true
    Module.Param3WriteEnabled=false
   ) {enabled=false}
  ]
  particleUpdateScript [
   &ParticleState()
   &FN_BoneFlow(
    Angle:NiagaraFloat=$Particles.FN_Initialize_BoneFlow_Attributes.TScale
    BoneID:NiagaraInt32=$Particles.Partition.ID
    Enable Position Offsets=true
    Height:NiagaraFloat=$Particles.Initial.Partition.NormalizedID
    Radius:NiagaraFloat=$Particles.Initial.FN_Initialize_BoneFlow_Attributes.Radius
    Set Material Dynamic=false
    SideVector1:Vector3f=<1 1 0>
    SideVector2:Vector3f=<-1 -1 0>
    SideVector3:Vector3f=<1 1 0>
    SkeletalMesh:NiagaraDataInterfaceSkeletalMesh=$User.SkeletalMesh02
    Stabalize Twisting=true
    Use Sockets Instead=true
   )
   &ScaleColor(
    Scale RGB=&VectorFromFloat(Value=&Lerp_Float(
      A=10
      B=1
      Alpha=&FloatFromCurve(FloatCurve=DataInterface() {propertyValues {Curve {keys [
           {interpMode=RCIM_Cubic, tangentMode=RCTM_User, time=0, value=0}
           {arriveTangent=1.5977, interpMode=RCIM_Cubic, leaveTangent=1.5977, tangentMode=RCTM_User, time=0.3, value=0.65}
           {interpMode=RCIM_Cubic, tangentMode=RCTM_User, time=1, value=1}
          ]}}})
     ))
    ScaleA=true
    ScaleRGB=true
   )
   &ScaleRibbonWidth(Ribbon Width Scale=&Lerp_Float(
     A=3
     B=0.1
     Alpha=&FloatFromCurve(FloatCurve=DataInterface() {propertyValues {Curve {keys [{interpMode=RCIM_Cubic, tangentMode=RCTM_User, time=0, value=0}, {interpMode=RCIM_Cubic, tangentMode=RCTM_User, time=1, value=1}]}}})
    ))
   &ScaleColor(
    Scale Alpha=&FloatFromCurve(
     FloatCurve=DataInterface() {propertyValues {Curve {keys [<at 0 0>,<at 0.1 1>,<at 0.25 1>,<at 0.3 0>,<at 0.7 0>,<at 0.75 1>,<at 0.9 1>,<at 1 0>]}}}
     CurveIndex=$Particles.Partition.NormalizedID
    )
    Scale RGB=&MakeVectorFromLinearColorRGB(Color=&LerpLinearColors(
      StartColor=<rgba 1 0 0.971 1>
      EndColor=<rgba 0.4668 0 1 1>
      LerpFactor=&FloatFromCurve(FloatCurve=DataInterface() {propertyValues {Curve {keys [{interpMode=RCIM_Cubic, tangentMode=RCTM_User, time=0, value=0}, {interpMode=RCIM_Cubic, tangentMode=RCTM_User, time=1, value=1}]}}})
     ))
    ScaleA=true
    ScaleRGB=true
   ) {enabled=false, module=ScaleColor001}
   &DynamicMaterialParameters(
    Index 0 Param 1=1
    Index 0 Param 2=&FloatFromCurve(FloatCurve=DataInterface() {propertyValues {Curve {keys [<at 0 0>,<at 1 -1>]}}})
    Index 0 Param 3=&Multiply_Float(
     A=&RandomRangeFloat(Minimum=0.2, Maximum=0.3)
     B=&FloatFromCurve(FloatCurve=DataInterface() {propertyValues {Curve {keys [
          {interpMode=RCIM_Cubic, leaveTangentWeight=1.0373, tangentMode=RCTM_User, time=0, value=1}
          {interpMode=RCIM_Cubic, tangentMode=RCTM_User, time=1, value=-0.25}
         ]}}})
    )
    Module.Param0WriteEnabled=true
    Module.Param1WriteEnabled=true
    Module.Param2WriteEnabled=true
    Module.Param3WriteEnabled=true
   ) {module=DynamicMaterialParameters001}
  ]
  rendererClasses [&NiagaraRibbonRendererProperties]
  renderers [{rendererClass=&NiagaraRibbonRendererProperties}]
 }
 Empty {
  bEnabled=false
  emitterUpdateScript [
   &EmitterState(Life Cycle Mode=::NewEnumerator1, Loop Behavior=::NewEnumerator1, Loop Duration Mode=::NewEnumerator1, UseLoopDelay=false)
   &SpawnBurst_Instantaneous(Spawn Count=6, Use Spawn Probability=false, UseLoopCountLimit=false)
   &SpawnRate(SpawnRate=5, Use Spawn Probability=false) {enabled=false}
  ]
  particleSpawnScript [
   &PartitionParticles(Number Of Partitions=6)
   &InitializeParticle(Color=<rgba 0.9278 0 1 1>, Sprite Size Mode=::NewEnumerator3, Uniform Sprite Size=7, UsePositionOffset=false, Write Material Random=true)
   &SkeletalMeshLocation(
    Bone / Socket Index=$Particles.PartitionParticles.Partition
    Interpolate Mesh Sampling=false
    MaskSpawnGroup=false
    Mesh Sampling Type=::NewEnumerator5
    Skeletal Mesh=$User.SkeletalMesh
    Socket Sampling Mode=::NewEnumerator5
   )
  ]
  particleUpdateScript [&ParticleState(Kill Particles When Lifetime Has Elapsed=false)]
  rendererClasses [&NiagaraSpriteRendererProperties]
  renderers [{rendererClass=&NiagaraSpriteRendererProperties}]
 }
 Glow {
  bEnabled=false
  emitterUpdateScript [
   &EmitterState(
    Life Cycle Mode=::NewEnumerator1
    Loop Delay=&RandomRangeFloat(Minimum=0.1, Maximum=0.5)
    Loop Duration=0.3
    Recalculate Duration Each Loop=true
    UseLoopDelay=true
   )
   &SpawnBurst_Instantaneous(Spawn Count=2, Use Spawn Probability=false, UseLoopCountLimit=false)
  ]
  particleSpawnScript [
   &PartitionParticles(Number Of Partitions=3)
   &InitializeParticle(
    Color Maximum=<rgba 0.1629 0 1 1>
    Color Minimum=<rgba 0.7844 0 1 1>
    Color Mode=::NewEnumerator2
    Lifetime Max=0.15
    Lifetime Min=0.8
    Lifetime Mode=::NewEnumerator1
    Sprite Size Mode=::NewEnumerator1
    Uniform Sprite Size Max=20
    Uniform Sprite Size Min=8
    UsePositionOffset=false
    Write Material Random=true
   )
   &SkeletalMeshLocation(MaskSpawnGroup=false, Mesh Sampling Type=::NewEnumerator5, OverrideRandomSeed=false, Skeletal Mesh=$User.SkeletalMesh02)
  ]
  particleUpdateScript [
   &ParticleState()
   &ScaleColor(
    Scale Alpha=&FloatFromCurve(FloatCurve=DataInterface() {propertyValues {Curve {keys [
         {interpMode=RCIM_Cubic, tangentMode=RCTM_User, time=0, value=1}
         {arriveTangentWeight=0.8714, interpMode=RCIM_Cubic, tangentMode=RCTM_User, time=1, value=0}
        ]}}})
    Scale RGB=&VectorFromFloat(Value=15)
    ScaleA=true
    ScaleRGB=true
   )
  ]
  rendererClasses [&NiagaraSpriteRendererProperties]
  renderers [{rendererClass=&NiagaraSpriteRendererProperties}]
 }
 Glows {
  bEnabled=false
  emitterUpdateScript [
   &EmitterState(
    Inactive Response=::NewEnumerator1
    Life Cycle Mode=::NewEnumerator1
    Loop Behavior=::NewEnumerator1
    Loop Duration Mode=::NewEnumerator1
    UseLoopDelay=false
   )
   &SpawnBurst_Instantaneous(Spawn Count=3, Use Spawn Probability=false, UseLoopCountLimit=false)
   &SpawnRate(SpawnRate=10, Use Spawn Probability=false) {enabled=false}
  ]
  particleSpawnScript [
   &PartitionParticles(Number Of Partitions=3)
   &InitializeParticle(
    Color=<rgba 0.8452 0.2456 1 1>
    Position Offset=&SelectVectorFromArray(
     Direct Array Index=$Particles.PartitionParticles.Partition
     Vector Selection Array=DataInterface() {propertyValues {FloatData [<0 0 0>, <0 7 12>, <0 0 -12>]}}
    )
    Sprite Rotation Angle=0.25
    Sprite Rotation Mode=::NewEnumerator1
    Sprite Size=&SelectVector2DFromArray(
     Direct Array Index=$Particles.PartitionParticles.Partition
     Vector 2D Selection Array=DataInterface() {propertyValues {FloatData [<30 15>, <20 10>, <20 10>]}}
    )
    Sprite Size Mode=::NewEnumerator4
    UsePositionOffset=true
    Write Material Random=true
   )
   &ShapeLocation(
    Disc Coverage=1
    MaskSpawnGroup=false
    Non Uniform Scale=<13 5 1>
    Offset Mode=::NewEnumerator0
    Override Seed=false
    Ring Radius=1
    Rotation Angle=0.96
    Rotation Mode=::NewEnumerator1
    Shape Primitive=::NewEnumerator4
    UseOwnerScale=true
   ) {enabled=false}
   /SetVariables(Particles.GlowRound:Vector2f=&SelectVector2DFromArray(
     Direct Array Index=$Particles.PartitionParticles.Partition
     Vector 2D Selection Array=DataInterface() {propertyValues {FloatData [<22 22>, <15 15>, <15 15>]}}
    )) {isSetParameters=true}
   /SetVariables(Particles.ColorGlow:LinearColor=<rgba 0.5426 0 1 0.7>) {isSetParameters=true}
   &SpriteFacingAndAlignment(Sprite Alignment=<1 0 0>, Sprite Facing=<0 0 1>, WriteAlignment=true, WriteFacing=true)
  ]
  particleUpdateScript [
   /SetVariables(Particles.ScaleWave:NiagaraFloat=&Waveform("Waveform [1]"=::NewEnumerator0, "[1] Frequency"=0.7, "[1] Amplitude Min/Max"=<0.75 1>)) {isSetParameters=true}
   /SetVariables(Particles.GlowRound:Vector2f=&Multiply_Vector2D_ByFloat(Float=$Particles.ScaleWave, Vector2D=$Particles.GlowRound)) {isSetParameters=true}
   &ParticleState(Kill Particles When Lifetime Has Elapsed=false, Let Infinitely Lived Particles Die When Emitter Deactivates=true)
   &ScaleSpriteSize(Scale Factor=&MakeVector2D(X=$Particles.ScaleWave, Y=1), Scale Sprite Size Mode=::NewEnumerator2)
   &ScaleColor(
    Scale Alpha=&SelectFloatFromArray(
     Direct Array Index=$Particles.PartitionParticles.Partition
     Float Selection Array=DataInterface() {propertyValues {FloatData [0.35,0.2,0.2]}}
    )
    Scale RGB=&VectorFromFloat(Value=1)
    ScaleA=true
    ScaleRGB=true
   )
  ]
  rendererClasses [&NiagaraSpriteRendererProperties]
  renderers [
   {rendererClass=&NiagaraSpriteRendererProperties}
   {rendererClass=&NiagaraSpriteRendererProperties, rendererIndex=1}
   {rendererClass=&NiagaraSpriteRendererProperties, rendererIndex=2}
  ]
 }
 Static {
  bEnabled=false
  emitterUpdateScript [
   &EmitterState(Life Cycle Mode=::NewEnumerator1, Loop Behavior=::NewEnumerator1, Loop Duration Mode=::NewEnumerator1, UseLoopDelay=false)
   &SpawnBurst_Instantaneous(Spawn Count=5, Spawn Time=0.05, Use Spawn Probability=false, UseLoopCountLimit=false)
   &SpawnBurst_Instantaneous(Spawn Count=2, Spawn Time=0.1, Use Spawn Probability=false, UseLoopCountLimit=false) {module=SpawnBurst_Instantaneous001}
   &SpawnRate(Spawn Probability=0.25, SpawnRate=5, Use Spawn Probability=false)
  ]
  particleSpawnScript [
   /Niagara/Modules/Spawn/Initialization/InitializeParticle(
    Color=&RandomRangeLinearColor(Minimum=<rgba 4 0.7052 3.8857 1>, Maximum=<rgba 6 1.4004 4.6812 1>)
    Lifetime=&UniformRangedFloat(Minimum=0.1, Maximum=0.15)
    Module.Write Color=true
    Module.Write Lifetime=true
    Module.Write Mass=true
    Module.Write Material Random=true
    Module.Write Position=false
    Module.Write Scale=false
    Module.Write SpriteRotation=true
    Module.Write SpriteSize=true
    Position=/Niagara/Functions/Localspace/SimulationPosition()
    Sprite Rotation=&UniformRangedFloat(Minimum=0, Maximum=360)
    Sprite Size=&UniformRangedVector2D(Minimum=<12 12>, Maximum=<25 25>)
   )
   &DynamicMaterialParameters(
    Index 0 Param 1=$Particles.MaterialRandom
    Index 0 Param 2=&Lerp_Float(A=0.7, B=1.25, Alpha=$Particles.MaterialRandom)
    Index 0 Param 4=1
    Module.Param0WriteEnabled=true
    Module.Param1WriteEnabled=true
    Module.Param2WriteEnabled=false
    Module.Param3WriteEnabled=true
   )
   &CameraOffset(Camera Offset Amount=12)
   &ShapeLocation(
    Disc Coverage=1
    MaskSpawnGroup=false
    Non Uniform Scale=<13 5 1>
    Offset Mode=::NewEnumerator0
    Override Seed=false
    Ring Radius=1
    Rotation Angle=0.96
    Rotation Mode=::NewEnumerator1
    Shape Primitive=::NewEnumerator4
    UseOwnerScale=true
   )
  ]
  particleUpdateScript [
   &ParticleState()
   &SpriteRotationRate(Rotation Rate=&RandomRangeFloat(Minimum=-100, Maximum=100))
   &DynamicMaterialParameters(
    Index 0 Param 3=&FloatFromCurve(FloatCurve=DataInterface() {propertyValues {Curve {keys [<at 0.5 0 RCIM_Cubic>, {arriveTangent=3.7979, interpMode=RCIM_Cubic, leaveTangent=3.7979, tangentMode=RCTM_User, time=1, value=1}]}}})
    Index 1 Param 2=0.7
    Index 1 Param 4=500
    Module.Param0WriteEnabled=false
    Module.Param0WriteEnabled1=true
    Module.Param1WriteEnabled=false
    Module.Param1WriteEnabled1=true
    Module.Param2WriteEnabled=true
    Module.Param2WriteEnabled1=true
    Module.Param3WriteEnabled=false
    Module.Param3WriteEnabled1=true
    Write Parameter Index 1=true
   ) {module=DynamicMaterialParameters001}
   &ScaleSpriteSize(Scale Factor=&Vector2DFromFloat(Value=&FloatFromCurve(FloatCurve=DataInterface() {propertyValues {Curve {keys [{arriveTangent=0.3147, interpMode=RCIM_Cubic, leaveTangent=0.3147, tangentMode=RCTM_User, time=0, value=0.8}, <at 1 1>]}}})))
   /Niagara/Modules/Audio/PlayPersistentAudio(Enable Fade In / Out=true, Fade In Time=0, Fade Out Time=2)
  ]
  rendererClasses [&NiagaraSpriteRendererProperties]
  renderers [{rendererClass=&NiagaraSpriteRendererProperties}]
 }
}
summary {
 systemName=NS_AncientWeapon_Glove_Ambient
 emitters [
  {bEnabled=false, emitterName=Static, rendererClasses [&NiagaraSpriteRendererProperties]}
  {bEnabled=false, emitterName=Glows, rendererClasses [&NiagaraSpriteRendererProperties]}
  {
   bEnabled=false
   emitterName=Electricity
   rendererClasses [
    &NiagaraSpriteRendererProperties
    &NiagaraMeshRendererProperties
    &NiagaraLightRendererProperties
   ]
  }
  {bEnabled=false, emitterName=ElectricityBoneFlow, rendererClasses [&NiagaraRibbonRendererProperties]}
  {bEnabled=false, emitterName=ElectricityBoneFlow001, rendererClasses [&NiagaraRibbonRendererProperties]}
  {bEnabled=false, emitterName=Glow, rendererClasses [&NiagaraSpriteRendererProperties]}
  {bEnabled=false, emitterName=Empty, rendererClasses [&NiagaraSpriteRendererProperties]}
 ]
 userVariables [
  User.AmmoCount {type=NiagaraFloat, defaultValue=0, description=""}
  User.Magnitude {type=NiagaraFloat, defaultValue=0, description=""}
  User.Overheat {type=NiagaraFloat, defaultValue=0, description=""}
  User.SkeletalMesh {
   type=NiagaraDataInterfaceSkeletalMesh
   defaultValue {
    struct=NiagaraExt_VariableValue_DataInterface
    value {dataInterface=&self:NiagaraDataInterfaceSkeletalMesh_0}
   }
  }
  User.SkeletalMesh02 {
   type=NiagaraDataInterfaceSkeletalMesh
   defaultValue {
    struct=NiagaraExt_VariableValue_DataInterface
    value {dataInterface=&self:NiagaraDataInterfaceSkeletalMesh_1}
   }
  }
  User.Steam {type=NiagaraBool, defaultValue=true}
 ]
}
