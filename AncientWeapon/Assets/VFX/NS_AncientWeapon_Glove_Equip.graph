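# Niagara effect for equipping the Ancient Weapon glove, showing a brief blue-violet flare on a skeletal mesh.
# The only enabled emitter, Flare, bursts rotating sprites that follow the sampled mesh location, grow over their life, brighten and then fade.
# A second sprite renderer accompanies the flare, with separate glow size and color attributes.
# Disabled Electricity (sprite, mesh, light) and LightningArc (Hermite-spline ribbon) emitters remain in the system.
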NiagaraSystem /AncientWeapon/Assets/VFX/NS_AncientWeapon_Glove_Equip
dependencies {
 usedDataInterfaces [&NiagaraDataInterfaceCurve, &NiagaraDataInterfaceArrayFloat3, &NiagaraDataInterfaceSkeletalMesh]
 usedDynamicInputs [
  &NormalizeFloat
  &RotateVectorByQuaternion
  &Add_Float
  &Sine
  &Cosine
 ]
 usedModules [
  &self:SystemSpawnScript.SystemScriptSource.SystemScriptGraph.NiagaraNodeAssignment_5.SetVariables_#1
  &SystemState
  &self:SystemSpawnScript.SystemScriptSource.SystemScriptGraph.NiagaraNodeAssignment_1.SetVariables_#2
  &self:SystemSpawnScript.SystemScriptSource.SystemScriptGraph.NiagaraNodeAssignment_0.SetVariables_#3
  &Timeline
  &self:SystemSpawnScript.SystemScriptSource.SystemScriptGraph.NiagaraNodeAssignment_2.SetVariables_#4
  &self:SystemSpawnScript.SystemScriptSource.SystemScriptGraph.NiagaraNodeAssignment_4.SetVariables_#5
  &self:SystemSpawnScript.SystemScriptSource.SystemScriptGraph.NiagaraNodeAssignment_3.SetVariables_#6
 ]
}
emitters {
 Electricity {
  bEnabled=false
  emitterUpdateScript [
   &EmitterState(Life Cycle Mode=::NewEnumerator1, Loop Duration=0.5, Recalculate Duration Each Loop=true, UseLoopDelay=false)
   &SpawnRate(SpawnRate=6, Use Spawn Probability=true)
   /SetVariables(Emitter.PanSpeed:NiagaraFloat=&Multiply_Float(A=&RandomRangeFloat(Minimum=0.3, Maximum=0.7), B=-1)) {isSetParameters=true}
  ]
  particleSpawnScript [
   &PartitionParticles()
   &InitializeParticle(
    Color Maximum=<rgba 0.0889 0.0333 0.9323 1>
    Color Minimum=<rgba 0.175 0 1 1>
    Color Mode=::NewEnumerator2
    Lifetime Max=0.4
    Lifetime Min=0.2
    Lifetime Mode=::NewEnumerator1
    Mesh Scale=<0.5 1 1>
    Mesh Scale Mode=::NewEnumerator4
    Sprite Rotation Mode=::NewEnumerator2
    Sprite Size Mode=::NewEnumerator1
    Uniform Sprite Size Max=6
    Uniform Sprite Size Min=3
    UsePositionOffset=false
    Write Material Random=true
   )
   &DynamicMaterialParameters(
    Index 0 Param 1=&UniformRangedFloat(Minimum=3, Maximum=5)
    Index 0 Param 3=$Particles.MaterialRandom
    Index 0 Param 4=&UniformRangedFloat(Minimum=0.5, Maximum=0.8)
    Module.Param0WriteEnabled=true
    Module.Param1WriteEnabled=false
    Module.Param2WriteEnabled=true
    Module.Param3WriteEnabled=true
   ) {module=DynamicMaterialParameters001}
   &InitialMeshOrientation(Mesh Orientation Mode=::NewEnumerator2, Orientation Coordinate Space=::Local, Use Rotation Vector=false)
   &InitialMeshOrientation(
    Mesh Axis To Orient=<-0.3 0 1>
    Mesh Orientation Mode=::NewEnumerator2
    Orientation Coordinate Space=::Local
    Rotation=<0 0 0.7>
    Rotation Coordinate Space=::NewEnumerator2
    Use Rotation Vector=true
   ) {enabled=false, module=InitialMeshOrientation002}
   &InitialMeshOrientation(Mesh Axis To Orient=<-0.4 0 1>, Mesh Orientation Mode=::NewEnumerator2, Rotation=<0 0 0.8>, Use Rotation Vector=true) {enabled=false, module=InitialMeshOrientation001}
   /SetVariables(
    Particles.FlareColor=&MultiplyLinearColor(
     Linear Color=<rgba 1 0 0.2377 1>
     Scale Factor=&MakeFloat4FromVectorAndFloat(
      "Vector (RGB)"=&VectorFromFloat(Value=&RandomRangeFloat(Minimum=0.5, Maximum=1))
      "Float (Alpha)"=&RandomRangeFloat(Minimum=0.1, Maximum=0.25)
     )
    )
    Particles.FlareSize:Vector2f=&RandomRangeVector2D(Minimum=<8 8>, Maximum=<12 12>)
    Particles.GlowColor=&MultiplyLinearColor(
     Linear Color=$Particles.Color
     Scale Factor=&MakeFloat4FromVectorAndFloat(
      "Vector (RGB)"=&VectorFromFloat(Value=&RandomRangeFloat(Minimum=2, Maximum=10))
      "Float (Alpha)"=&RandomRangeFloat(Minimum=0.5, Maximum=1)
     )
    )
    Particles.GlowPosition:NiagaraPosition=&SelectVectorFromArray(
     Direct Array Index=$Particles.PartitionParticles.Partition
     Vector Selection Array=DataInterface() {propertyValues {FloatData [<0 0 0>, <-14 -20 10>]}}
    )
    Particles.GlowSize=&RandomRangeVector2D(Minimum=<16 16>, Maximum=<25 25>)
   ) {isSetParameters=true}
   &SpriteFacingAndAlignment(Sprite Facing=<-0.1 -0.3 1>, WriteAlignment=true, WriteFacing=true)
  ]
  particleUpdateScript [
   &ParticleState()
   /SetVariables(Particles.Position=&MakeVector(
     X=&MakeFloatFromVector(Vector=Linked() {linkedVariable {name=Particles.Position, type=NiagaraPosition}}, Channel=::NewEnumerator0)
     Y=&MakeFloatFromVector(Vector=Linked() {linkedVariable {name=Particles.Position, type=NiagaraPosition}}, Channel=::NewEnumerator1)
     Z=0
    )) {enabled=false, isSetParameters=true}
   &DynamicMaterialParameters(
    Index 0 Param 1=&FloatFromCurve(
     FloatCurve=DataInterface() {propertyValues {Curve {keys [
         <at 0.5 1 RCIM_Cubic>
         {arriveTangent=-1.5423, arriveTangentWeight=0.451, interpMode=RCIM_Cubic, leaveTangent=-1.5423, tangentMode=RCTM_User, time=1, value=0}
        ]}}}
     Scale Curve=&Lerp_Float(A=0.7, B=1.5, Alpha=$Particles.Initial.MaterialRandom)
    )
    Index 0 Param 2=&FloatFromCurve(
     FloatCurve=DataInterface() {propertyValues {Curve {keys [
         {arriveTangent=10.9494, interpMode=RCIM_Cubic, leaveTangent=10.9494, tangentMode=RCTM_User, time=0, value=1}
         {arriveTangent=-8.5345, interpMode=RCIM_Cubic, leaveTangent=-8.5345, tangentMode=RCTM_User, time=0.2, value=2}
         {arriveTangent=-2.4364, interpMode=RCIM_Cubic, leaveTangent=-2.4364, time=0.3638, value=0.5858}
         {arriveTangent=-8.496, interpMode=RCIM_Cubic, leaveTangent=-8.496, tangentMode=RCTM_User, time=0.4639, value=1.357}
         {arriveTangent=8.1076, interpMode=RCIM_Cubic, leaveTangent=-1.1376, tangentMode=RCTM_User, time=1, value=1.5}
        ]}}}
     Scale Curve=&RandomRangeFloat(Minimum=0.05, Maximum=0.5)
    )
    Module.Param0WriteEnabled=true
    Module.Param1WriteEnabled=true
    Module.Param2WriteEnabled=true
    Module.Param3WriteEnabled=true
   ) {module=DynamicMaterialParameters002}
   &SkeletalMeshLocation(
    MaskSpawnGroup=false
    Mesh Sampling Type=::NewEnumerator5
    Sampled Orientation Coordinate Space=::Local
    Skeletal Mesh=$System.Skeletal Mesh
    Socket Sampling Mode=::NewEnumerator5
   )
   /SetVariables(Particles.DynamicParamOverride:Vector4f=&MakeFloat4FromVectorAndFloat(
     "Vector (RGB)"=/Niagara/DynamicInputs/Transforms/TransformPosition(
      Position=Linked() {linkedVariable {name=Particles.Initial.Position, type=NiagaraPosition}}
      Source Space=::Simulation
      Destination Space=::World
     )
     "Float (Alpha)"=0
    )) {enabled=false, isSetParameters=true}
   /SetVariables(Particles.DynamicParamOverride:Vector4f=&MakeFloat4FromVectorAndFloat(
     "Vector (RGB)"=&TransformPosition(
      Position=Linked() {linkedVariable {name=Particles.Initial.Position, type=NiagaraPosition}}
      Source Space=::Simulation
      Destination Space=::World
     )
     "Float (Alpha)"=0
    )) {enabled=false, isSetParameters=true}
   &ScaleMeshSize(Scale Factor=&MakeVector(
     X=1
     Y=1
     Z=&Multiply_Float(
      A=&FloatFromCurve(FloatCurve=DataInterface() {propertyValues {Curve {keys [
           {interpMode=RCIM_Cubic, tangentMode=RCTM_User, time=0, value=0.2}
           {arriveTangent=1.3862, interpMode=RCIM_Cubic, leaveTangent=1.3862, tangentMode=RCTM_User, time=0.3, value=0.625}
           {arriveTangentWeight=1.7224, interpMode=RCIM_Cubic, tangentMode=RCTM_User, time=1, value=1}
          ]}}})
      B=1
     )
    ))
   &SolveForcesAndVelocity(Clamp Velocity=false, Limit Acceleration=false)
   &ScaleColor(
    Scale Alpha=&self:Flicker_01(
     FlickerAmount:NiagaraFloat=&FloatFromCurve(FloatCurve=DataInterface() {propertyValues {Curve {keys [
          {arriveTangent=2.7223, interpMode=RCIM_Cubic, leaveTangent=2.7223, tangentMode=RCTM_User, time=0, value=1}
          {arriveTangent=0.0266, interpMode=RCIM_Cubic, leaveTangent=0.0266, time=0.2457, value=0.5}
          {arriveTangent=-0.3314, interpMode=RCIM_Cubic, leaveTangent=-0.3314, time=0.527, value=1.014}
          {arriveTangent=-3.2376, interpMode=RCIM_Cubic, leaveTangent=-11960.9277, tangentMode=RCTM_User, time=1, value=0.25}
         ]}}})
     FlickerInterval:NiagaraFloat=0.1
    )
    Scale RGB=&VectorFromFloat(Value=&Lerp_Float(
      A=0.4
      B=5
      Alpha=&FloatFromCurve(
       FloatCurve=DataInterface() {propertyValues {Curve {keys [
           {interpMode=RCIM_Cubic, tangentMode=RCTM_User, time=0, value=0}
           {arriveTangent=0.6368, interpMode=RCIM_Cubic, leaveTangent=0.6368, tangentMode=RCTM_User, time=0.8, value=0.2}
           {arriveTangent=5.7967, interpMode=RCIM_Cubic, leaveTangent=5.7967, tangentMode=RCTM_User, time=1, value=1}
          ]}}}
       CurveIndex=&RandomRangeFloat(Minimum=0, Maximum=1)
      )
     ))
    ScaleA=true
    ScaleRGB=true
   )
   &FN_Light_Attributes(
    Color=$Particles.Initial.Color
    Create Color Attribute=true
    Create Exponent Attribute=true
    Create Position Attribute=true
    Create Radius Attribute=true
    Position=&AddVectorToPosition(Position=Linked() {linkedVariable {name=Particles.Initial.GlowPosition, type=Vector3f}}, Vector=<-8 0 5>)
    Radius=100
    Scale Color=2
   )
  ]
  rendererClasses [
   &NiagaraSpriteRendererProperties
   &NiagaraMeshRendererProperties
   &NiagaraLightRendererProperties
  ]
  renderers [
   {rendererClass=&NiagaraSpriteRendererProperties}
   {rendererClass=&NiagaraMeshRendererProperties, rendererIndex=1}
   {rendererClass=&NiagaraSpriteRendererProperties, rendererIndex=2}
   {rendererClass=&NiagaraSpriteRendererProperties, rendererIndex=3}
   {rendererClass=&NiagaraLightRendererProperties, rendererIndex=4}
  ]
 }
 Flare {
  emitterUpdateScript [
   &EmitterState()
   &SpawnBurst_Instantaneous(Use Spawn Probability=false, UseLoopCountLimit=false)
  ]
  particleSpawnScript [
   &InitializeParticle(
    Color=<rgba 0.1216 0.0683 1 1>
    Lifetime=0.75
    Sprite Size=<75 40>
    Sprite Size Mode=::NewEnumerator4
    UsePositionOffset=false
    Write Material Random=true
   )
   &CameraOffset(Camera Offset Amount=6)
   /SetVariables(Particles.FlareGlowSize:Vector2f=<35 35>) {isSetParameters=true}
   /SetVariables() {isSetParameters=true}
  ]
  particleUpdateScript [
   &ParticleState()
   &SpriteRotationRate(Rotation Rate=100)
   &SkeletalMeshLocation(MaskSpawnGroup=false, Mesh Sampling Type=::NewEnumerator5, Skeletal Mesh=$System.Skeletal Mesh, Socket Sampling Mode=::NewEnumerator5)
   &ScaleSpriteSize(
    Uniform Curve Scale=1.1
    Uniform Curve Sprite Scale=DataInterface() {propertyValues {Curve {keys [
        {interpMode=RCIM_Cubic, leaveTangentWeight=0.4288, tangentMode=RCTM_User, time=0, value=0}
        {arriveTangent=1.4786, interpMode=RCIM_Cubic, leaveTangent=1.4786, tangentMode=RCTM_User, time=0.6, value=0.35}
        <at 1 1>
       ]}}}
   )
   &ScaleColor(
    Scale Alpha=&FloatFromCurve(FloatCurve=DataInterface() {propertyValues {Curve {keys [
         {interpMode=RCIM_Cubic, leaveTangentWeight=1.5094, tangentMode=RCTM_User, time=0, value=0}
         {interpMode=RCIM_Cubic, tangentMode=RCTM_User, time=0.7, value=1}
         {arriveTangent=-8.7416, interpMode=RCIM_Cubic, leaveTangent=-8.7416, tangentMode=RCTM_User, time=1, value=0}
        ]}}})
    Scale RGB=&VectorFromFloat(Value=10)
    ScaleA=true
    ScaleRGB=true
   )
   /SetVariables(Particles.FlareGlowColor:LinearColor=&MultiplyLinearColor(
     Linear Color=<rgba 0.1216 0.0683 1 1>
     Scale Factor=&MakeFloat4FromVectorAndFloat(
      "Vector (RGB)"=&VectorFromFloat(Value=3)
      "Float (Alpha)"=&FloatFromCurve(FloatCurve=DataInterface() {propertyValues {Curve {keys [
           {interpMode=RCIM_Cubic, leaveTangentWeight=1.5094, tangentMode=RCTM_User, time=0, value=0}
           {interpMode=RCIM_Cubic, tangentMode=RCTM_User, time=0.7, value=1}
           {arriveTangent=-8.7416, interpMode=RCIM_Cubic, leaveTangent=-8.7416, tangentMode=RCTM_User, time=1, value=0}
          ]}}})
     )
    )) {isSetParameters=true}
  ]
  rendererClasses [&NiagaraSpriteRendererProperties]
  renderers [
   {rendererClass=&NiagaraSpriteRendererProperties}
   {rendererClass=&NiagaraSpriteRendererProperties, rendererIndex=1}
  ]
 }
 LightningArc {
  bEnabled=false
  emitterSpawnScript [/Game/Effects/Niagara/Modules/Partition/FN_PartitionAttributes(Number of Bone Sets=1, Particles Per Partition=7)]
  emitterUpdateScript [
   &EmitterState(Life Cycle Mode=::NewEnumerator1, Loop Duration=0.25, UseLoopDelay=false)
   /SetVariables(Emitter.DirectionStart:Vector3f=&RandomRangeVector(Minimum=<0 0 0>, Maximum=<0 0 0>, Evaluation Type=::NewEnumerator1)) {isSetParameters=true}
   /SetVariables(Emitter.DirectionStart:Vector3f=&RandomRangeVector(Minimum=<-50 -100 -50>, Maximum=<50 100 50>, Evaluation Type=::NewEnumerator1)) {enabled=false, isSetParameters=true}
   &SpawnRate(SpawnRate=10, Use Spawn Probability=false) {enabled=false}
   &SpawnBurst_Instantaneous(Spawn Count=$Emitter.FN_PartitionAttributes.Spawn Total, Use Spawn Probability=false)
   /SetVariables(Emitter.VecEnd:Vector3f=<0 -25 7>) {isSetParameters=true}
  ]
  particleSpawnScript [
   /Niagara/Modules/Spawn/Initialization/InitializeParticle(
    Lifetime=5
    Module.Write Color=true
    Module.Write Lifetime=true
    Module.Write Mass=true
    Module.Write Material Random=true
    Module.Write Position=true
    Module.Write Scale=false
    Module.Write SpriteRotation=false
    Module.Write SpriteSize=true
    Sprite Size=<25 25>
   )
   /Game/Effects/Niagara/Modules/Partition/FN_Particle_partition_v5/FN_Particle_Partition_v5(Enable Ribbon ID=true)
   &InitializeRibbon(
    Lifetime=0.15
    Module.Write Color=true
    Module.Write Lifetime=true
    Module.Write Mass=true
    Module.Write Material Random=true
    Module.Write Position=false
    Module.Write RibbonTwist=false
    Module.Write RibbonWidth=true
    Ribbon Width=3
   )
   /SetVariables(Particles.VecOffset:Vector3f=&RandomRangeVector(Minimum=<0 0 0>, Maximum=<0 0 0>)) {enabled=false, isSetParameters=true}
  ]
  particleUpdateScript [
   &SkeletalMeshLocation(MaskSpawnGroup=false, Mesh Sampling Type=::NewEnumerator5, Skeletal Mesh=$System.Skeletal Mesh, Socket Sampling Mode=::NewEnumerator5)
   /SetVariables(Particles.VecEnd:Vector3f=&Add_Vector(A=$Particles.SkeletalMeshLocation.SampledPosition, B=$Emitter.VecEnd)) {isSetParameters=true}
   &ParticleState()
   /SetVariables(Particles.SplineEndPos:Vector3f=&Add_Vector(A=$Particles.VecEnd, B=$System.VecOffset)) {isSetParameters=true}
   /SetVariables(Particles.RibbonFacing=<1 0 0>) {isSetParameters=true}
   &SystemLocation(MaskSpawnGroup=false, Offset=$Particles.SkeletalMeshLocation.SampledPosition)
   /Game/Effects/Niagara/Modules/Curves/FN_HermiteCubicSpline_v1/FN_HermiteCubicSpline(
    Direction Start:Vector3f=$Emitter.DirectionStart
    Position Blend:NiagaraFloat=$Particles.Initial.Partition.NormalizedID
    Position End:Vector3f=$Particles.VecEnd
    Position Start:Vector3f=$Particles.SkeletalMeshLocation.SampledPosition
    Write Particles Position=true
   )
   &FN_HermiteCubicSpline(
    Direction End:Vector3f=&Lerp_Vector(A=$System.HermiteDirectionEnd, B=$System.HermiteDirectionEndAlt, Alpha=$User.InFlight)
    Direction Start:Vector3f=$Emitter.DirectionStart
    Position Blend:NiagaraFloat=$Particles.Initial.Partition.NormalizedID
    Position End:Vector3f=/Niagara/DynamicInputs/Helpers/SimulationPosition()
    Position Start:Vector3f=$Particles.SplineEndPos
    Write Particles Position=true
   ) {enabled=false, module=FN_HermiteCubicSpline001}
   &ScaleColor(
    Scale Alpha=&FloatFromCurve(
     FloatCurve=DataInterface() {propertyValues {Curve {keys [<at 0.01 0>,<at 0.05 1>,<at 1 1>]}}}
     CurveIndex=$Emitter.Scalability.DistanceFraction
    )
    ScaleA=true
    ScaleRGB=false
   ) {enabled=false, module=ScaleColor001}
   &FN_RibbonMaterial(
    Color=&RandomRangeLinearColor(Minimum=<rgba 0.5292 0.0834 0.75 1>, Maximum=<rgba 0 0.0337 1.5 1>)
    Color Pan U Speed=-2
    Erosion Amount=1
    Erosion Pan Speed U=-2
    Erossion Scale U=4
    Macro Pan Speed=-4
    Micro Distortion Amount=0.25
    Micro Pan Speed=-1
    Micro Scale U=0.25
    Opacity Erosion:NiagaraFloat=&RandomRangeFloat(Minimum=0, Maximum=0)
    Opacity Pan U Speed=-2
    Opacity Scale U=0.2
    Scale Alpha:NiagaraFloat=&FloatFromCurve(FloatCurve=DataInterface() {propertyValues {Curve {keys [{arriveTangent=2.1027, interpMode=RCIM_Cubic, leaveTangent=2.1027, tangentMode=RCTM_User, time=0, value=0}, <at 1 1 RCIM_Cubic>]}}})
    Scale Color:Vector3f=&VectorFromFloat(Value=&FloatFromCurve(
      FloatCurve=DataInterface() {propertyValues {Curve {keys [
          <at 0 10 RCIM_Cubic>
          {interpMode=RCIM_Cubic, leaveTangent=-5.4051, tangentMode=RCTM_Break, time=0.1165, value=6}
          {arriveTangent=5.5101, interpMode=RCIM_Cubic, tangentMode=RCTM_Break, time=0.9, value=6}
          {interpMode=RCIM_Cubic, tangentMode=RCTM_Break, time=1, value=10}
         ]}}}
      CurveIndex=$Particles.Partition.NormalizedID
      Scale Curve=2
     ))
   )
   &FN_RibbonMaterial(
    Color=<rgba 0.5292 0.0834 0.75 1>
    Color Pan U Speed=-2
    Erosion Amount=1
    Erosion Pan Speed U=-2
    Erossion Scale U=4
    Macro Pan Speed=-4
    Micro Distortion Amount=0.25
    Micro Pan Speed=-1
    Micro Scale U=0.25
    Opacity Erosion:NiagaraFloat=&RandomRangeFloat(Minimum=0, Maximum=0)
    Opacity Pan U Speed=-2
    Opacity Scale U=0.2
    Scale Alpha:NiagaraFloat=&FloatFromCurve(FloatCurve=DataInterface() {propertyValues {Curve {keys [{arriveTangent=2.1027, interpMode=RCIM_Cubic, leaveTangent=2.1027, tangentMode=RCTM_User, time=0, value=0}, <at 1 1 RCIM_Cubic>]}}})
    Scale Color:Vector3f=&VectorFromFloat(Value=&FloatFromCurve(
      FloatCurve=DataInterface() {propertyValues {Curve {keys [
          <at 0 10 RCIM_Cubic>
          {interpMode=RCIM_Cubic, leaveTangent=-13.0964, tangentMode=RCTM_Break, time=0.1165, value=6}
          {arriveTangent=11.7463, interpMode=RCIM_Cubic, tangentMode=RCTM_Break, time=0.9, value=6}
          {interpMode=RCIM_Cubic, tangentMode=RCTM_Break, time=1, value=10}
         ]}}}
      CurveIndex=$Particles.Partition.NormalizedID
      Scale Curve=2
     ))
   ) {enabled=false, module=FN_RibbonMaterial001}
   &ScaleRibbonWidth(Ribbon Width Scale=&Multiply_Float(
     A=&FloatFromCurve(
      FloatCurve=DataInterface() {propertyValues {Curve {keys [
          {interpMode=RCIM_Cubic, leaveTangentWeight=0.3333, tangentMode=RCTM_User, time=0, value=1}
          {arriveTangentWeight=5.2193, interpMode=RCIM_Cubic, tangentMode=RCTM_User, time=1, value=1}
         ]}}}
      CurveIndex=$Particles.Partition.NormalizedID
      Scale Curve=&Lerp_Float(A=1, B=1, Alpha=0)
     )
     B=2
    ))
  ]
  rendererClasses [&NiagaraRibbonRendererProperties]
  renderers [{rendererClass=&NiagaraRibbonRendererProperties}]
 }
}
summary {
 systemName=NS_AncientWeapon_Glove_Equip
 emitters [
  {emitterName=Flare, rendererClasses [&NiagaraSpriteRendererProperties]}
  {
   bEnabled=false
   emitterName=Electricity
   rendererClasses [
    &NiagaraSpriteRendererProperties
    &NiagaraMeshRendererProperties
    &NiagaraLightRendererProperties
   ]
  }
  {bEnabled=false, emitterName=LightningArc, rendererClasses [&NiagaraRibbonRendererProperties]}
 ]
 userVariables [
  User.Charge {type=NiagaraBool}
  User.ChargeAmount {type=NiagaraFloat, defaultValue=0, description=""}
  User.Delay {type=NiagaraBool}
  User.DustLocation {type=NiagaraPosition, description="", defaultValue {}}
  User.GlobalScale {type=NiagaraFloat, defaultValue=1, description=""}
  User.GroundDust {type=NiagaraBool}
  User.InFlight {type=NiagaraFloat, defaultValue=1, description=""}
 ]
}
