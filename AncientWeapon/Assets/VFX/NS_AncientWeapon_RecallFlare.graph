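# Niagara system for the Ancient Weapon's recall flare, a short-lived blue glowing sprite flare that tracks a skeletal mesh.
# A burst spawns particles that resample their location on User.SkeletalMesh every update, rotate, and grow along a curve scaled by User.GlobalScale.
# Alpha curves fade the flare in, then drop it out sharply at end of life, with boosted RGB.
# Two sprite renderers draw the single emitter, which also writes separate FlareGlowSize and FlareGlowColor attributes.
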NiagaraSystem /AncientWeapon/Assets/VFX/NS_AncientWeapon_RecallFlare
dependencies {
 usedDataInterfaces [&NiagaraDataInterfaceCurve]
 usedDynamicInputs [&NormalizeFloat]
 usedModules [
  &SystemState
  &self:SystemSpawnScript.SystemScriptSource.SystemScriptGraph.NiagaraNodeAssignment_0.SetVariables_#1
  &Timeline
 ]
}
emitters {Flare {
  emitterUpdateScript [
   &EmitterState()
   &SpawnBurst_Instantaneous(Use Spawn Probability=false, UseLoopCountLimit=false)
  ]
  particleSpawnScript [
   /Niagara/Modules/Spawn/Initialization/InitializeParticle(
    Color=<rgba 0.1216 0.0683 1 1>
    Lifetime=0.75
    Module.Write Color=true
    Module.Write Lifetime=true
    Module.Write Mass=false
    Module.Write Material Random=true
    Module.Write Position=false
    Module.Write Scale=false
    Module.Write SpriteRotation=false
    Module.Write SpriteSize=true
    Position=&SimulationPosition()
    Sprite Size=&Multiply_Vector2D_ByFloat(Float=3, Vector2D=<100 40>)
   )
   &CameraOffset(Camera Offset Amount=5)
   /SetVariables(Particles.FlareGlowSize:Vector2f=<75 75>) {isSetParameters=true}
   /SetVariables() {isSetParameters=true}
  ]
  particleUpdateScript [
   &ParticleState()
   &SpriteRotationRate(Rotation Rate=100)
   &SkeletalMeshLocation(MaskSpawnGroup=false, Mesh Sampling Type=::NewEnumerator5, Skeletal Mesh=$User.SkeletalMesh, Socket Sampling Mode=::NewEnumerator5)
   &ScaleSpriteSize(
    Uniform Curve Scale=&Multiply_Float(A=0.5, B=$User.GlobalScale)
    Uniform Curve Sprite Scale=DataInterface() {propertyValues {Curve {keys [
        {interpMode=RCIM_Cubic, leaveTangentWeight=0.4288, tangentMode=RCTM_User, time=0, value=0}
        {arriveTangent=1.4786, interpMode=RCIM_Cubic, leaveTangent=1.4786, tangentMode=RCTM_User, time=0.6, value=0.35}
        <at 1 1>
       ]}}}
   )
   &ScaleColor(
    Scale Alpha=&FloatFromCurve(FloatCurve=DataInterface() {propertyValues {Curve {keys [
         {interpMode=RCIM_Cubic, leaveTangentWeight=1.5094, tangentMode=RCTM_User, time=0, value=0}
         {interpMode=RCIM_Cubic, tangentMode=RCTM_User, time=0.7, value=1}
         {arriveTangent=-8.7416, interpMode=RCIM_Cubic, leaveTangent=-8.7416, tangentMode=RCTM_User, time=1, value=0}
        ]}}})
    Scale RGB=&VectorFromFloat(Value=10)
    ScaleA=true
    ScaleRGB=true
   )
   /SetVariables(Particles.FlareGlowColor:LinearColor=&MultiplyLinearColor(
     Linear Color=<rgba 0.1216 0.0683 1 1>
     Scale Factor=&MakeFloat4FromVectorAndFloat(
      "Vector (RGB)"=&VectorFromFloat(Value=3)
      "Float (Alpha)"=&FloatFromCurve(FloatCurve=DataInterface() {propertyValues {Curve {keys [
           {interpMode=RCIM_Cubic, leaveTangentWeight=1.5094, tangentMode=RCTM_User, time=0, value=0}
           {interpMode=RCIM_Cubic, tangentMode=RCTM_User, time=0.7, value=1}
           {arriveTangent=-8.7416, interpMode=RCIM_Cubic, leaveTangent=-8.7416, tangentMode=RCTM_User, time=1, value=0}
          ]}}})
     )
    )) {isSetParameters=true}
  ]
  rendererClasses [&NiagaraSpriteRendererProperties]
  renderers [
   {rendererClass=&NiagaraSpriteRendererProperties}
   {rendererClass=&NiagaraSpriteRendererProperties, rendererIndex=1}
  ]
 }}
summary {
 systemName=NS_AncientWeapon_RecallFlare
 emitters [{emitterName=Flare, rendererClasses [&NiagaraSpriteRendererProperties]}]
 userVariables [
  User.Charge {type=NiagaraBool}
  User.ChargeAmount {type=NiagaraFloat, defaultValue=0, description=""}
  User.Delay {type=NiagaraBool}
  User.GlobalScale {type=NiagaraFloat, defaultValue=1, description=""}
  User.SkeletalMesh {
   type=NiagaraDataInterfaceSkeletalMesh
   defaultValue {
    struct=NiagaraExt_VariableValue_DataInterface
    value {dataInterface=&self:NiagaraDataInterfaceSkeletalMesh_0}
   }
  }
 ]
}
